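# Translucent, unlit, two-sided material for Niagara mesh particles that draws a soft intersection effect where CloudyMuffin effect meshes meet scene geometry.
# A view-angle term blends inner and outer depth-fade distances and intersection values, and panning noise varies the power exponent to break up the result.
# That mask multiplies particle alpha, while particle color feeds the shared emissive adjustments function.
# A platform switch replaces the mask with zero on Switch, Android and PC performance mode.
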Material /JethroWeaponGameplay/Gameplay/FX/CloudyMuffin/M_CloudyMuffinIntersection
graph {
 properties {
  bAutomaticallySetUsageInEditor=false
  bUsedWithNiagaraMeshParticles=true
  blendMode=BLEND_Translucent
  shadingModel=MSM_Unlit
  twoSided=true
  usedShadingModels=Unlit
 }
 nodes [
  DistanceToNearestSurface!(World Position=WorldPosition().XYZ)
  MaterialFunctionCall() {materialFunction=&CameraDirectionVector}
  n1 = ParticleColor()
  MaterialFunctionCall(Emissive=@n1.RGB) {materialFunction=&MF_FX_EmissiveAdjustments}
  n2 = Constant()
  n3 = ScalarParameter() {parameterName=intersetion Value Low inner}
  n4 = ScalarParameter() {parameterName=intersetion Value high inner}
  n5 = DotProduct(A=CameraVectorWS(), B=VertexNormalWS())
  n6 = DepthFade(FadeDistance=LinearInterpolate(A=ScalarParameter() {parameterName=Fade Distance Outer}, B=ScalarParameter() {parameterName=Fade Distance inner}, Alpha=Abs(None=@n5)))
  n7 = ScalarParameter() {parameterName=intersetion Value Low Outer}
  n8 = ScalarParameter() {parameterName=intersetion Value high Outer}
  n9 = DotProduct(A=CameraVectorWS(), B=VertexNormalWS())
  n10 = VectorParameter() {parameterName=noise options, defaultValue=<rgba 1 1 0 0>}
  n11 = Panner(Coordinate=Multiply(A=TextureCoordinate(), B=AppendVector(A=@n10.R, B=@n10.R)), Speed=AppendVector(A=@n10.B, B=@n10.B))
  n12 = Power(
   Base=LinearInterpolate(A=LinearInterpolate(A=@n3, B=@n4, Alpha=@n6), B=LinearInterpolate(A=@n7, B=@n8, Alpha=@n6), Alpha=Abs(None=@n9))
   Exp=LinearInterpolate(
    A=ScalarParameter() {defaultValue=3, parameterName=Dpeth fade power}
    B=ScalarParameter() {defaultValue=3, parameterName=Dpeth fade power high}
    Alpha=TextureSample(UVs=@n11) {samplerType=SAMPLERTYPE_LinearColor, texture=&HighContrastNoise}.R
   )
  )
  Multiply(
   A=@n1.A
   B=MaterialFunctionCall("Xb1,PS4"=@n12, "PC-High&&Epic,XSX,PS5"=@n12, "PC-Perf,Android"=@n2, Switch=@n2, "PC-Low(NotPerf)&&Medium"=@n12) {materialFunction=&PlatformSwitch_1D}.Result
  )
 ]
}
