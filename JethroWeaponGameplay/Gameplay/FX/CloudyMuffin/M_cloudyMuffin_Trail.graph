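# Additive unlit material for Niagara ribbon trails of the CloudyMuffin weapon effect, colored by particle color through a shared emissive adjustment function.
# A tiled noise texture distorts the UVs feeding procedural GeneratedBand masks.
# The single band is shaped by power, multiply, add and contrast, then scaled by particle alpha as an opacity mask.
# A mirrored double-band variant ends at a reroute, separate from that chain.
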Material /JethroWeaponGameplay/Gameplay/FX/CloudyMuffin/M_cloudyMuffin_Trail
graph {
 properties {
  bAutomaticallySetUsageInEditor=false
  bUsedWithNiagaraRibbons=true
  blendMode=BLEND_Additive
  shadingModel=MSM_Unlit
  usedShadingModels=Unlit
 }
 nodes [
  n1 = ParticleColor()
  MaterialFunctionCall(Emissive=@n1.RGB) {materialFunction=&MF_FX_EmissiveAdjustments}
  n2 = ScalarParameter() {defaultValue=0.5, parameterName=Width single}
  n3 = ScalarParameter() {parameterName=Shaprness Single}
  n4 = TextureCoordinate()
  n5 = VectorParameter() {parameterName=DistortOptions, channelNames {a=Zero Center, b=Distort Str, g=Tile Y, r=tile X}, defaultValue=<rgba 1 1 0 -0.5>}
  n6 = ComponentMask(None=@n5) {g=true, r=true}
  n7 = ComponentMask(None=TextureSample(UVs=Multiply(A=@n4, B=@n6)) {samplerType=SAMPLERTYPE_LinearColor, texture=&HighContrastNoise}.RGB) {g=true, r=true}
  n8 = Add(A=TextureCoordinate(), B=Multiply(A=Subtract(A=@n7, B=@n5.Zero Center) {constB=0.5}, B=@n5.Distort Str))
  n9 = VectorParameter() {parameterName=Opacity Options, channelNames {a=contast, b=add, g=mul, r=pow}, defaultValue=<rgba 1 1 0 0>}
  n10 = Multiply(
   A=Power(
    Base=MaterialFunctionCall!(Width=@n2, Sharpness=@n3, Offset, Direction Switch, Compare, Input Coordinates=@n8) {materialFunction=/Engine/Functions/Engine_MaterialFunctions03/Procedurals/GeneratedBand}.Result
    Exp=@n9.pow
   )
   B=@n9.mul
  )
  Multiply(
   A=@n1.A
   B=MaterialFunctionCall(In=Add(A=@n10, B=@n9.add), Contrast=@n9.contast) {materialFunction=&CheapContrast}.Result
  )
  n11 = ScalarParameter() {defaultValue=0.25, parameterName=Width}
  n12 = ScalarParameter() {parameterName=Sharpness}
  n13 = ScalarParameter() {defaultValue=0.25, parameterName=Offset}
  n14 = Multiply(A=@n13) {constB=-1}
  Reroute(None=Add(
    A=MaterialFunctionCall(Width=@n11, Sharpness=@n12, Offset=@n13, Input Coordinates=@n8) {materialFunction=&GeneratedBand}.Result
    B=MaterialFunctionCall(Width=@n11, Sharpness=@n12, Offset=@n14, Input Coordinates=@n8) {materialFunction=&GeneratedBand}.Result
   ))
 ]
}
