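# Niagara explosion effect for the CloudyMuffin in Jethro weapon gameplay, layering flash, streaks, energy bursts, an expanding light-casting wave and fire smoke.
# Burst emitters share a system delay and mostly scale sizes and speeds by User.FXScale.
# System camera-adjust and player-indoors values blend brightness, opacity and size to tone down the blast.
# User.IsInWater adds a water mesh and drops one smoke burst, which also depends on platform quality.
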NiagaraSystem /JethroWeaponGameplay/Gameplay/FX/CloudyMuffin/NS_Jethro_CloudyMuffin_Exp
dependencies {
 usedDataInterfaces [
  &NiagaraDataInterfaceCurve
  &NiagaraDataInterfaceArrayInt32
  &NiagaraDataInterfacePlatformSet
  &NiagaraDataInterfaceCamera
  &NiagaraDataInterfaceVectorCurve
  &NiagaraDataInterfaceArrayFloat
  &NiagaraDataInterfaceColorCurve
  &NiagaraDataInterfaceParticleRead
  &NiagaraDataInterfaceSpriteRendererInfo
  &NiagaraDataInterfaceVector2DCurve
 ]
 usedDynamicInputs [/Niagara/DynamicInputs/Scalability/ScalabilityDistanceBasedFloat, &OneMinusFloat]
 usedModules [
  &self:SystemSpawnScript.SystemScriptSource.SystemScriptGraph.NiagaraNodeAssignment_0.SetVariables_#1
  &self:SystemSpawnScript.SystemScriptSource.SystemScriptGraph.NiagaraNodeAssignment_1.SetVariables_#2
  &self:SystemSpawnScript.SystemScriptSource.SystemScriptGraph.NiagaraNodeAssignment_5.SetVariables_#3
  &self:SystemSpawnScript.SystemScriptSource.SystemScriptGraph.NiagaraNodeAssignment_4.SetVariables_#4
  &SystemState
  &self:SystemSpawnScript.SystemScriptSource.SystemScriptGraph.NiagaraNodeAssignment_2.SetVariables_#5
  &self:SystemSpawnScript.SystemScriptSource.SystemScriptGraph.NiagaraNodeAssignment_3.SetVariables_#6
 ]
}
emitters {
 CenterFlash {
  emitterUpdateScript [
   &EmitterState()
   &SpawnBurst_Instantaneous(Spawn Group=1, Use Spawn Probability=false, UseLoopCountLimit=false) {enabled=false, module=SpawnBurst_Instantaneous001}
   &SpawnBurst_Instantaneous(Spawn Count=1, Spawn Time=&Add_Float(A=$System.Delay, B=0.01), Use Spawn Probability=false)
  ]
  particleSpawnScript [
   &InitializeParticle(
    Color=<rgba 12 2.087 0.6504 1>
    Lifetime=&MaskFloatBySpawnGroup(
     Masked Float=&RandomRangeFloat(Minimum=0.1, Maximum=0.15)
     Passthrough Float=&RandomRangeFloat(Minimum=0.125, Maximum=0.125)
    )
    Mesh Scale Mode=::NewEnumerator3
    Mesh Uniform Scale=&MaskFloatBySpawnGroup(
     Masked Float=&Multiply_Float(A=0.0016, B=$User.FXScale)
     Passthrough Float=&Multiply_Float(A=0.0008, B=$User.FXScale)
    )
    UsePositionOffset=false
    Write Material Random=true
   )
   &InitialMeshOrientation(Mesh Orientation Mode=::NewEnumerator2, Orientation Vector=<0 0 1>, Use Rotation Vector=false)
   &DynamicMaterialParameters(
    Index 0 Param 1=&UniformRangedFloat(Minimum=0, Maximum=20)
    Module.Param0WriteEnabled=true
    Module.Param1WriteEnabled=false
    Module.Param2WriteEnabled=false
    Module.Param3WriteEnabled=false
   )
   &CameraOffset(Camera Offset Amount=100)
  ]
  particleUpdateScript [
   &ParticleState()
   &SolveForcesAndVelocity(Clamp Velocity=false, Limit Acceleration=false)
   &ScaleMeshSize(Scale Factor=&VectorFromFloat(Value=&FloatFromCurve(
      FloatCurve=DataInterface() {propertyValues {Curve {keys [{arriveTangent=9.901, interpMode=RCIM_Cubic, leaveTangent=9.901, tangentMode=RCTM_User, time=0, value=0.4}, <at 1 6 RCIM_Cubic>]}}}
      Scale Curve=&Lerp_Float(A=0.65, B=1.5, Alpha=$System.Cam Adjust)
     )))
   &ScaleColor(
    Scale Alpha=&FloatFromCurve(
     FloatCurve=DataInterface() {propertyValues {Curve {keys [<at 0.4 1 RCIM_Cubic>,<at 1 0>]}}}
     Scale Curve=&Lerp_Float(
      A=&Lerp_Float(A=0.1, B=1, Alpha=$System.Cam Adjust)
      B=&Lerp_Float(A=0.01, B=1, Alpha=$System.Cam Adjust)
      Alpha=$System.IsPlayerIndoors
     )
    )
    ScaleA=true
    ScaleRGB=true
   )
  ]
  rendererClasses [&NiagaraMeshRendererProperties]
  renderers [{rendererClass=&NiagaraMeshRendererProperties, rendererIndex=0}]
 }
 Combined_Streaks {
  emitterUpdateScript [
   &EmitterState()
   &SpawnBurst_Instantaneous(Spawn Count=9, Spawn Time=$System.Delay, Use Spawn Probability=false, UseLoopCountLimit=false)
   &SpawnBurst_Instantaneous(Spawn Count=6, Spawn Group=1, Spawn Time=$System.Delay, Use Spawn Probability=false, UseLoopCountLimit=false) {enabled=false, module=SpawnBurst_Instantaneous001}
  ]
  particleSpawnScript [
   &InitializeParticle(
    Color=&MaskLinearColorBySpawnGroup(
     Masked Linear Color=<rgba 30 3.937 0 1>
     Passthrough Linear Color=&RandomRangeLinearColor(Minimum=<rgba 0.03 0.0259 0.0262 5>, Maximum=<rgba 0.09 0.0756 0.0729 3>)
    )
    Lifetime=&MaskFloatBySpawnGroup(
     Masked Float=&RandomRangeFloat(Minimum=0.15, Maximum=0.25)
     Passthrough Float=&RandomRangeFloat(Minimum=0.25, Maximum=0.4)
    )
    Sprite Size=&MaskVector2dBySpawnGroup(
     Masked Vector2D=&Multiply_Vector2D_ByFloat(
      Float=&Multiply_Float(A=0.007, B=$User.FXScale)
      Vector2D=&RandomRangeVector2D(Minimum=<20 100>, Maximum=<60 250>)
     )
     Passthrough Vector2D=&Multiply_Vector2D_ByFloat(
      Float=&Multiply_Float(A=0.005, B=$User.FXScale)
      Vector2D=&RandomRangeVector2D(Minimum=<300 300>, Maximum=<250 200>)
     )
     Spawn Group Mask=0
    )
    Sprite Size Mode=::NewEnumerator4
    UsePositionOffset=false
    Write Material Random=true
   )
   &ShapeLocation(
    MaskSpawnGroup=false
    Override Seed=false
    Sphere Radius=&Multiply_Float(A=&MaskFloatBySpawnGroup(Masked Float=0.05, Passthrough Float=0.08), B=$User.FXScale)
    UseOwnerScale=true
   )
   &AddVelocity(
    Module.Use Falloff Distance=false
    Override Seed=false
    Velocity Mode=::NewEnumerator1
    Velocity Speed=&MaskFloatBySpawnGroup(
     Masked Float=&Multiply_Float(A=&UniformRangedFloat(Minimum=0.4, Maximum=0.8), B=$User.FXScale)
     Passthrough Float=&Multiply_Float(
      A=&Lerp_Float(
       A=0.4
       B=2.5
       Alpha=&FloatFromCurve(
        FloatCurve=DataInterface() {propertyValues {Curve {keys [
            {interpMode=RCIM_Cubic, tangentMode=RCTM_Break, time=0, value=0}
            {arriveTangent=0.9428, interpMode=RCIM_Cubic, leaveTangent=0.9428, tangentMode=RCTM_User, time=0.3, value=0.65}
            {interpMode=RCIM_Cubic, tangentMode=RCTM_Break, time=1, value=1}
           ]}}}
        CurveIndex=&RandomRangeFloat(Minimum=0, Maximum=1)
       )
      )
      B=$User.FXScale
     )
    )
   )
   &SubUVAnimation(
    "Sprite Renderer "=DataInterface() {propertyValues {SpriteRenderer=&self:Combined_Streaks_0.NiagaraSpriteRendererProperties_0}}
    SubUV Animation Mode:ENiagara_SubUVLookupModeV2=::NewEnumerator2
    UseEndFrame=false
    UseStartFrame=false
   )
   &ApplyInitialForces()
  ]
  particleUpdateScript [
   &ParticleState()
   &ScaleSpriteSize(
    Scale Sprite Size Mode=::NewEnumerator0
    Uniform Scale Factor=&MaskFloatBySpawnGroup(Masked Float=1, Passthrough Float=0.4)
   ) {module=ScaleSpriteSize001}
   &ScaleSpriteSize(
    Scale Factor=&MaskVector2dBySpawnGroup(
     Masked Vector2D=&Vector2DFromCurve(
      Vector2Curve=DataInterface() {propertyValues {
        XCurve {defaultValue=1, keys [<at 0 0.75 RCIM_Cubic>,<at 1 1 RCIM_Cubic>]}
        YCurve {defaultValue=1, keys [<at 0 0 RCIM_Cubic>,<at 1 2 RCIM_Cubic>]}
       }}
      ScaleCurve=<2 0.6>
     )
     Passthrough Vector2D=&Vector2DFromCurve(
      Vector2Curve=DataInterface() {propertyValues {XCurve {keys [<at 0.0036 1.6897>,<at 0.9967 0.0582>,<at 1 0>]}, YCurve {defaultValue=1, keys [<at 0 1>,<at 1.002 3.7227>]}}}
      ScaleCurve=<1 1.25>
     )
     Spawn Group Mask=0
    )
    Scale Sprite Size Mode=::NewEnumerator2
   )
   &SolveForcesAndVelocity(Clamp Velocity=false, Limit Acceleration=false)
   &ScaleColor(
    Scale Alpha=&FloatFromCurve(
     FloatCurve=DataInterface() {propertyValues {Curve {defaultValue=1, keys [<at 0 1 RCIM_Cubic>,<at 1 0>]}}}
     Scale Curve=&Lerp_Float(
      A=&MaskFloatBySpawnGroup(Masked Float=&Lerp_Float(A=0.4, B=1, Alpha=$System.Cam Adjust), Passthrough Float=1)
      B=&MaskFloatBySpawnGroup(Masked Float=&Lerp_Float(A=0.001, B=1, Alpha=$System.Cam Adjust), Passthrough Float=1)
      Alpha=$System.IsPlayerIndoors
     )
    )
    Scale RGB=&MaskVectorBySpawnGroup(
     Masked Vector=&VectorFromFloat(Value=&FloatFromCurve(
       FloatCurve=DataInterface() {propertyValues {Curve {keys [<at 0 10 RCIM_Cubic>,<at 1 1 RCIM_Cubic>]}}}
       Scale Curve=&Lerp_Float(A=0.5, B=1, Alpha=$System.Cam Adjust)
      ))
     Passthrough Vector=<1 1 1>
    )
    ScaleA=true
    ScaleRGB=true
   )
   &DynamicMaterialParameters(
    Index 0 Param 2=1
    Index 0 Param 3=1
    Index 0 Param 4=&MaskFloatBySpawnGroup(
     Masked Float=1
     Passthrough Float=&FloatFromCurve(FloatCurve=DataInterface() {propertyValues {Curve {keys [<at 0.3 1>,<at 1 0.25>]}}})
    )
    Module.Index0Float0Write=true
    Module.Index0Float1Write=true
    Module.Index0Float2Write=true
    Module.Index0Float3Write=true
   )
  ]
  rendererClasses [&NiagaraSpriteRendererProperties]
  renderers [{rendererClass=&NiagaraSpriteRendererProperties}]
 }
 Corona_LW {rendererClasses [], renderers []}
 Energy {
  emitterUpdateScript [
   /SetVariables(Emitter.saturation=1) {isSetParameters=true}
   &EmitterState()
   &SpawnBurst_Instantaneous(Spawn Count=1, Use Spawn Probability=false)
  ]
  particleSpawnScript [
   &InitializeParticle(
    Color=<rgba 1 1 1 0.7>
    Lifetime=0.5
    Mesh Scale Mode=::NewEnumerator3
    Mesh Uniform Scale=1.1
    Sprite Rotation Mode=::NewEnumerator2
    Sprite Size Mode=::NewEnumerator3
    Uniform Sprite Size=12
    UsePositionOffset=false
    Write Material Random=true
   )
   &InitialMeshOrientation(Use Rotation Vector=false)
   &CameraOffset(Camera Offset Amount=25)
  ]
  particleUpdateScript [
   &ParticleState()
   &ScaleSpriteSize(Uniform Curve Scale=&Lerp_Float(A=50, B=100, Alpha=$System.Cam Adjust))
   &ScaleMeshSize(Scale Factor=&VectorFromFloat(Value=&FloatFromCurve(
      FloatCurve=DataInterface() {propertyValues {Curve {defaultValue=1, keys [{tangentMode=RCTM_User, time=0, value=0.2}, {tangentMode=RCTM_User, time=1, value=1}]}}}
      Scale Curve=&Lerp_Float(A=1, B=2, Alpha=$System.Cam Adjust)
     )))
   &ScaleColor(
    Color Value To Scale=&MultiplyLinearColor(
     Linear Color=$Particles.Initial.Color
     Scale Factor=&MakeVector4FromVectorAndFloat(
      "Vector (RGB)"=&VectorFromFloat(Value=1000)
      "Float (Alpha)"=&Lerp_Float(
       A=&Lerp_Float(A=0.5, B=1, Alpha=$System.Cam Adjust)
       B=&Lerp_Float(
        A=0.25
        B=1
        Alpha=&FloatFromCurve(
         FloatCurve=DataInterface() {propertyValues {Curve {keys [
             {arriveTangent=0.2133, interpMode=RCIM_Cubic, leaveTangent=0.2133, leaveTangentWeight=0.3333, tangentMode=RCTM_User, time=0, value=0}
             {arriveTangent=2.8541, arriveTangentWeight=0.3333, interpMode=RCIM_Cubic, leaveTangent=2.8541, tangentMode=RCTM_User, time=1, value=1}
            ]}}}
         CurveIndex=$System.Cam Adjust
        )
       )
       Alpha=$System.IsPlayerIndoors
      )
     )
    )
    Linear Color Curve=DataInterface() {propertyValues {
      AlphaCurve {keys [{arriveTangent=-2.6598, interpMode=RCIM_Cubic, leaveTangent=-2.6598, tangentMode=RCTM_User, time=0, value=1}, <at 1 0>]}
      BlueCurve {keys [<at 0.2206 0>]}
      GreenCurve {keys [<at 0.2206 0.2601>]}
      RedCurve {keys [<at 0.2206 1>]}
     }}
    Scale Mode=::NewEnumerator2
   )
   &FN_Radial_Material(Base Sprite Tile U=0.5, Base Sprite Tile V=3, Depth Fade=250, Radial Mask Offset X=0.1, Sprite Extrusion Speed=1.25)
   /SetVariables(Particles.Linear Color=&MultiplyLinearColor(
     Linear Color=$Particles.Color
     Scale Factor=&MakeVector4FromVectorAndFloat("Vector (RGB)"=&VectorFromFloat(Value=1.5), "Float (Alpha)"=0.35)
    )) {isSetParameters=true}
   /SetVariables(Particles.MeshDynam0:Vector4f=&MakeVector4(
     X=&FloatFromCurve(FloatCurve=DataInterface() {propertyValues {Curve {keys [<at 0 1.5 RCIM_Cubic>, {arriveTangent=-1.2989, interpMode=RCIM_Cubic, leaveTangent=-1.2989, tangentMode=RCTM_User, time=1, value=1}]}}})
     Y=0
     Z=0
     W=0
    )) {isSetParameters=true}
  ]
  rendererClasses [&NiagaraMeshRendererProperties, &NiagaraSpriteRendererProperties]
  renderers [
   {rendererClass=&NiagaraMeshRendererProperties, rendererIndex=0}
   {rendererClass=&NiagaraSpriteRendererProperties, rendererIndex=1}
  ]
 }
 EnergyWave {
  emitterUpdateScript [
   &EmitterState()
   &SpawnBurst_Instantaneous(Spawn Group=1, Use Spawn Probability=false, UseLoopCountLimit=false) {enabled=false, module=SpawnBurst_Instantaneous001}
   &SpawnBurst_Instantaneous(Spawn Count=2, Spawn Time=$System.Delay, Use Spawn Probability=false)
  ]
  particleSpawnScript [
   &InitializeParticle(
    Color Mode=::NewEnumerator0
    Lifetime=&SelectFloatFromArray(
     Direct Array Index=&ReturnParticleIDAttributes()
     Float Selection Array=DataInterface() {propertyValues {FloatData [0.35,0.53]}}
    )
    Mesh Scale Mode=::NewEnumerator3
    Mesh Uniform Scale=&MaskFloatBySpawnGroup(
     Masked Float=&Multiply_Float(A=0.011, B=$User.FXScale)
     Passthrough Float=&Multiply_Float(A=0.0072, B=$User.FXScale)
    )
    UsePositionOffset=false
    Write Material Random=true
   )
   &InitialMeshOrientation(Orientation Axis=<0 0 1>, Orientation Vector=<0 0 1>, Rotation=<0 0.25 0>, Use Orientation Vector=true, Use Rotation Vector=true)
   &CameraOffset(Camera Offset Amount=30)
   &ShapeLocation(MaskSpawnGroup=false, Offset Mode=::NewEnumerator0, Override Seed=false, Sphere Radius=1, UseOwnerScale=true)
   /SetVariables(Particles.LightPosOffset:NiagaraPosition=<0 0 50>) {isSetParameters=true}
   /SetVariables(Particles.VisibilityTag=&ReturnParticleIDAttributes()) {isSetParameters=true}
   &InitialMeshOrientation(
    Mesh Orientation Mode=::NewEnumerator7
    Rotation=&RandomRangeVector(Minimum=<0 0 0>, Maximum=<1 0 0>)
    Use Rotation Vector=true
   ) {module=InitialMeshOrientation001}
  ]
  particleUpdateScript [
   &ParticleState()
   &ScaleColor(
    Scale Mode=::NewEnumerator1
    Scale RGBA=&MakeVector4_FromLinearColor(LinearColor=&ColorFromCurve(ColorCurve=DataInterface() {propertyValues {
        AlphaCurve {keys [<at 0 1>,<at 0.3576 1>,<at 1 0>]}
        BlueCurve {keys [<at 0 0>,<at 0.1684 0>,<at 0.3654 0>,<at 1 0>]}
        GreenCurve {keys [<at 0 0.4051>,<at 0.1684 0.325>,<at 0.3654 0.2679>,<at 1 0.1355>]}
        RedCurve {keys [<at 0 1>,<at 0.1684 1>,<at 0.3654 1>,<at 1 1>]}
       }}))
    ScaleRGBA=true
   )
   &DynamicMaterialParameters(
    Index 0 Param 1=&FloatFromCurve(FloatCurve=DataInterface() {propertyValues {Curve {keys [{arriveTangent=-2.1015, interpMode=RCIM_Cubic, leaveTangent=-2.1015, tangentMode=RCTM_User, time=0, value=0.5}, <at 1 -0.15>]}}})
    Index 0 Param 2=&FloatFromCurve(FloatCurve=DataInterface() {propertyValues {Curve {keys [{arriveTangent=-0.3464, interpMode=RCIM_Cubic, leaveTangent=-0.3464, tangentMode=RCTM_User, time=0, value=1}, <at 1 0.9>]}}})
    Index 0 Param 3=&FloatFromCurve(FloatCurve=DataInterface() {propertyValues {Curve {keys [
         {interpMode=RCIM_Cubic, tangentMode=RCTM_Break, time=0, value=0}
         {arriveTangent=0.4478, interpMode=RCIM_Cubic, leaveTangent=0.4478, tangentMode=RCTM_User, time=0.15, value=1}
         {interpMode=RCIM_Cubic, tangentMode=RCTM_Break, time=1, value=1.2}
        ]}}})
    Module.Param0WriteEnabled=true
    Module.Param1WriteEnabled=true
    Module.Param2WriteEnabled=true
    Module.Param3WriteEnabled=false
   )
   &ScaleColor(
    Scale RGB=&Lerp_Vector(
     A=&VectorFromFloat(Value=&FloatFromCurve(
       FloatCurve=DataInterface() {propertyValues {Curve {keys [
           {interpMode=RCIM_Cubic, leaveTangent=-40.039, tangentMode=RCTM_Break, time=0, value=10}
           {arriveTangent=-6.6072, interpMode=RCIM_Cubic, leaveTangent=-6.6072, tangentMode=RCTM_User, time=0.15, value=3}
           {interpMode=RCIM_Cubic, tangentMode=RCTM_Break, time=1, value=1}
          ]}}}
       Scale Curve=&Lerp_Float(A=2, B=10, Alpha=$System.Cam Adjust)
      ))
     B=&VectorFromFloat(Value=&FloatFromCurve(
       FloatCurve=DataInterface() {propertyValues {Curve {keys [
           {interpMode=RCIM_Cubic, leaveTangent=-40.039, leaveTangentWeight=2.0026, tangentMode=RCTM_Break, time=0, value=10}
           {
            arriveTangent=-6.6072
            arriveTangentWeight=0.3341
            interpMode=RCIM_Cubic
            leaveTangent=-6.6072
            leaveTangentWeight=1.8934
            tangentMode=RCTM_User
            time=0.15
            value=3
           }
           {arriveTangentWeight=0.2833, interpMode=RCIM_Cubic, tangentMode=RCTM_Break, time=1, value=1}
          ]}}}
       Scale Curve=&Lerp_Float(A=0.1, B=10, Alpha=$System.Cam Adjust)
      ))
     Alpha=$System.IsPlayerIndoors
    )
    ScaleA=false
    ScaleRGB=true
   ) {module=ScaleColor001}
   &ScaleMeshSize(Scale Factor=&VectorFromFloat(Value=&FloatFromCurve(FloatCurve=DataInterface() {propertyValues {Curve {keys [
          {interpMode=RCIM_Cubic, tangentMode=RCTM_Break, time=0, value=0.2}
          {arriveTangent=2.9365, interpMode=RCIM_Cubic, leaveTangent=2.9365, tangentMode=RCTM_User, time=0.1, value=1}
          {interpMode=RCIM_Cubic, tangentMode=RCTM_Break, time=1, value=2}
         ]}}})))
   &FN_Light_Attributes(
    Color=<rgba 1 0.1939 0 1>
    Create Color Attribute=true
    Create Exponent Attribute=true
    Create Position Attribute=true
    Create Radius Attribute=true
    Position=&Add_Vector(A=Linked() {linkedVariable {name=Particles.Initial.Position, type=NiagaraPosition}}, B=<0 0 25>)
    Radius=&Multiply_Float(A=&Lerp_Float(A=4, B=3, Alpha=$System.IsPlayerIndoors), B=$User.FXScale)
    Scale Color=&Lerp_Float(
     A=&FloatFromCurve(
      FloatCurve=DataInterface() {propertyValues {Curve {keys [
          {interpMode=RCIM_Cubic, leaveTangentWeight=0.9776, tangentMode=RCTM_Break, time=0.1, value=1}
          {arriveTangent=-3.2716, interpMode=RCIM_Cubic, tangentMode=RCTM_Break, time=1, value=0}
         ]}}}
      Scale Curve=&Lerp_Float(A=15, B=25, Alpha=$System.Cam Adjust)
     )
     B=&FloatFromCurve(
      FloatCurve=DataInterface() {propertyValues {Curve {keys [
          {interpMode=RCIM_Cubic, leaveTangentWeight=0.9776, tangentMode=RCTM_Break, time=0.1, value=1}
          {arriveTangent=-1.8345, interpMode=RCIM_Cubic, leaveTangent=-1.8345, tangentMode=RCTM_User, time=1, value=0}
         ]}}}
      Scale Curve=&Lerp_Float(A=3, B=18, Alpha=$System.Cam Adjust)
     )
     Alpha=$System.IsPlayerIndoors
    )
   )
   /SetVariables(Particles.LightPos:NiagaraPosition=&Add_Vector(
     A=Linked() {linkedVariable {name=Particles.Position, type=NiagaraPosition}}
     B=Linked() {linkedVariable {name=Particles.LightPosOffset, type=NiagaraPosition}}
    )) {isSetParameters=true}
  ]
  rendererClasses [&NiagaraMeshRendererProperties, &NiagaraLightRendererProperties]
  renderers [
   {rendererClass=&NiagaraMeshRendererProperties, rendererIndex=0}
   {rendererClass=&NiagaraLightRendererProperties, rendererIndex=1}
  ]
 }
 FN_Mesh001_water {
  emitterUpdateScript [
   &EmitterState()
   &SpawnBurst_Instantaneous(
    Spawn Count=&MakeCustomIntFromBool(Bool=$User.IsInWater, True INT=1, False INT=0)
    Spawn Group=1
    Spawn Time=$System.Delay
    Use Spawn Probability=false
    UseLoopCountLimit=false
   ) {module=SpawnBurst_Instantaneous001}
  ]
  particleSpawnScript [
   &InitializeParticle(
    Color=<rgba 1 1 1 0.025>
    Lifetime=&UniformRangedFloat(Minimum=1.9, Maximum=2.4)
    Mesh Scale=&Multiply_VectorByFloat(
     Vector=&MaskVectorBySpawnGroup(Masked Vector=<5 5 2>, Passthrough Vector=<5 5 1>)
     Float=&MaskFloatBySpawnGroup(Masked Float=1.2, Passthrough Float=2)
    )
    Mesh Scale Mode=::NewEnumerator4
    UsePositionOffset=false
    Write Material Random=true
   ) {module=InitializeParticle001}
   &InitialMeshOrientation(
    Mesh Orientation Mode=::NewEnumerator7
    Rotation=&RandomRangeVector(Minimum=<0 0 0>, Maximum=<0 0 1>)
    Use Rotation Vector=true
   ) {module=InitialMeshOrientation001}
  ]
  particleUpdateScript [
   &ParticleState()
   &ScaleMeshSize(Scale Factor=&VectorFromCurve(VectorCurve=DataInterface() {propertyValues {
       XCurve {keys [
         {interpMode=RCIM_Cubic, leaveTangent=7.6613, tangentMode=RCTM_Break, time=0, value=0}
         {arriveTangent=2.4936, interpMode=RCIM_Cubic, leaveTangent=2.4936, tangentMode=RCTM_User, time=0.2392, value=1.3545}
         <at 1 2>
        ]}
       YCurve {keys [
         {interpMode=RCIM_Cubic, leaveTangent=7.7916, tangentMode=RCTM_Break, time=0, value=0}
         {arriveTangent=2.6017, interpMode=RCIM_Cubic, leaveTangent=2.6017, tangentMode=RCTM_User, time=0.2392, value=1.3545}
         <at 1 2>
        ]}
       ZCurve {keys [<at 0 3>,<at 1 1>]}
      }}))
   &DynamicMaterialParameters(
    Index 0 Param 1=&FloatFromCurve(FloatCurve=DataInterface() {propertyValues {Curve {keys [<at 0 0>,<at 1 1.5>]}}})
    Index 0 Param 2=0.25
    Index 0 Param 3=50
    Index 0 Param 4=1
    Module.Param0WriteEnabled=true
    Module.Param1WriteEnabled=true
    Module.Param2WriteEnabled=true
    Module.Param3WriteEnabled=true
   )
   &CameraOffset(Camera Offset Amount=50)
   &ScaleColor(
    Scale Alpha=&FloatFromCurve(FloatCurve=DataInterface() {propertyValues {Curve {keys [
         <at 0 0>
         {arriveTangent=5.1071, interpMode=RCIM_Cubic, leaveTangent=5.1071, tangentMode=RCTM_Break, time=0.1294, value=2.8567}
         {arriveTangent=-8.2025, interpMode=RCIM_Cubic, leaveTangent=-8.2025, tangentMode=RCTM_User, time=0.3115, value=1.8534}
         <at 0.4715 0.0687>
         <at 1 0>
        ]}}})
    Scale RGB=&VectorFromFloat(Value=&FloatFromCurve(FloatCurve=DataInterface() {propertyValues {Curve {keys [<at 0 2>,<at 1 0.2>]}}}))
    ScaleA=true
    ScaleRGB=true
   )
  ]
  rendererClasses [&NiagaraMeshRendererProperties]
  renderers [{rendererClass=&NiagaraMeshRendererProperties, rendererIndex=0}]
 }
 FireSmoke {
  emitterSpawnScript [/SetVariables(Emitter.Mat Emiss:NiagaraFloat=&Lerp_Float(A=0.025, B=0, Alpha=$System.IsPlayerIndoors)) {isSetParameters=true}]
  emitterUpdateScript [
   &EmitterState()
   &SpawnBurst_Instantaneous(Spawn Count=25, Spawn Time=&Add_Float(A=$System.Delay, B=0.1), Use Spawn Probability=false)
   &SpawnBurst_Instantaneous(
    Spawn Count=&MakeCustomIntFromBool(Bool=$User.IsInWater, True INT=0, False INT=25)
    Spawn Group=1
    Spawn Probability=&MakeCustomFloatFromBool(
     Bool=&IsPlatformSetActive(New Platform Set=DataInterface() {propertyValues {Platforms {qualityLevelMask=-4}}})
     True Float=1
     False Float=0
    )
    Spawn Time=0.05
    Use Spawn Probability=true
    UseLoopCountLimit=false
   ) {module=SpawnBurst_Instantaneous001}
   &SpawnBurst_Instantaneous(Spawn Group=2, Use Spawn Probability=false, UseLoopCountLimit=false) {module=SpawnBurst_Instantaneous002}
  ]
  particleSpawnScript [
   &InitializeParticle(
    Color=&LerpLinearColors(
     StartColor=<rgba 0.287 0.0717 0 1>
     EndColor=&MaskLinearColorBySpawnGroup(
      Masked Linear Color=&RandomRangeLinearColor(Minimum=<rgba 0.6435 0.6435 0.6435 1>, Maximum=<rgba 0.4957 0.4957 0.4957 1>)
      Passthrough Linear Color=&RandomRangeLinearColor(Minimum=<rgba 0.2696 0.2696 0.2696 0.35>, Maximum=<rgba 0.7043 0.7043 0.7043 0.5>)
     )
     LerpFactor=0.75
    )
    Lifetime=&self:float_sel_4(
     float:NiagaraFloat=&RandomRangeFloat(Minimum=0.8, Maximum=1.2)
     float001:NiagaraFloat=&RandomRangeFloat(Minimum=0.4, Maximum=0.6)
     float002:NiagaraFloat=&RandomRangeFloat(Minimum=0.8, Maximum=1.2)
     float003:NiagaraFloat=0
     int32:NiagaraInt32=&SpawnGroup()
    )
    Sprite Rotation Mode=::NewEnumerator2
    Sprite Size Mode=::NewEnumerator3
    Uniform Sprite Size=&Multiply_Float(
     A=&RandomRangeFloat(
      Minimum=&self:floatsel(float:NiagaraFloat=1.65, float001:NiagaraFloat=1.5, float002:NiagaraFloat=3, int32:NiagaraInt32=&SpawnGroup())
      Maximum=&self:floatsel(float:NiagaraFloat=1.95, float001:NiagaraFloat=1.8, float002:NiagaraFloat=3, int32:NiagaraInt32=&SpawnGroup())
     )
     B=$User.FXScale
    )
    UsePositionOffset=false
    Write Material Random=true
   )
   &ShapeLocation(
    Apply To Particle Position=::NewEnumerator1
    Cylinder Height=1
    Cylinder Radius=&Multiply_Float(A=0.3, B=$User.FXScale)
    MaskSpawnGroup=false
    Override Seed=false
    Shape Primitive=::NewEnumerator1
    Surface Only=true
    Use Endcaps In Surface Only Mode=false
    UseOwnerScale=true
   ) {module=ShapeLocation001}
   &ShapeLocation(
    Apply To Particle Position=::NewEnumerator1
    MaskSpawnGroup=false
    Override Seed=false
    Sphere Radius=&Multiply_Float(A=0.5, B=$User.FXScale)
    UseOwnerScale=true
   )
   /SetVariables(Particles.Position=&self:select_pos(
     Position:NiagaraPosition=$Particles.ShapeLocation.ShapePosition
     Position001:NiagaraPosition=$Particles.ShapeLocation001.ShapePosition
     Position002:NiagaraPosition=/Niagara/DynamicInputs/Helpers/SimulationPosition()
     int32:NiagaraInt32=&SpawnGroup()
    )) {isSetParameters=true}
   &AddVelocity(
    Velocity=&self:ScratchDynamicInput_01(
     Default:Vector3f=$Particles.ShapeLocation.ShapeNormal
     Vector:Vector3f=$Particles.ShapeLocation001.ShapeNormal
     Vector001:Vector3f=<0 0 0>
     int32:NiagaraInt32=&SpawnGroup()
    )
    Velocity Speed Scale=&self:floatsel(
     float:NiagaraFloat=&Multiply_Float(
      A=&Lerp_Float(
       A=6
       B=20
       Alpha=&FloatFromCurve(
        FloatCurve=DataInterface() {propertyValues {Curve {keys [
            {interpMode=RCIM_Cubic, tangentMode=RCTM_Break, time=0, value=0}
            {arriveTangent=1.9418, interpMode=RCIM_Cubic, leaveTangent=1.9418, tangentMode=RCTM_User, time=0.3, value=0.55}
            {arriveTangentWeight=0.3338, interpMode=RCIM_Cubic, tangentMode=RCTM_Break, time=1, value=1}
           ]}}}
        CurveIndex=&RandomRangeFloat(Minimum=0, Maximum=1)
       )
      )
      B=$User.FXScale
     )
     float001:NiagaraFloat=&Multiply_Float(A=&RandomRangeFloat(Minimum=10, Maximum=10), B=$User.FXScale)
     float002:NiagaraFloat=&Multiply_Float(
      A=&Lerp_Float(
       A=6
       B=20
       Alpha=&FloatFromCurve(
        FloatCurve=DataInterface() {propertyValues {Curve {keys [
            {interpMode=RCIM_Cubic, tangentMode=RCTM_Break, time=0, value=0}
            {arriveTangent=1.9418, interpMode=RCIM_Cubic, leaveTangent=1.9418, tangentMode=RCTM_User, time=0.3, value=0.55}
            {arriveTangentWeight=0.3338, interpMode=RCIM_Cubic, tangentMode=RCTM_Break, time=1, value=1}
           ]}}}
        CurveIndex=&RandomRangeFloat(Minimum=0, Maximum=1)
       )
      )
      B=$User.FXScale
     )
     int32:NiagaraInt32=&SpawnGroup()
    )
   )
   /SetVariables(Particles.VisibilityTag=&SelectIntFromArray(Direct Array Index=&SpawnGroup(), Int Selection Array=DataInterface() {propertyValues {IntData [0,0,1]}})) {isSetParameters=true}
  ]
  particleUpdateScript [
   &ParticleState()
   &SpriteRotationRate(Rotation Rate=&Multiply_Float(
     A=&RandomRangeFloat(Minimum=-3, Maximum=3)
     B=&FloatFromCurve(
      FloatCurve=DataInterface() {propertyValues {Curve {keys [
          {arriveTangent=-5.9085, interpMode=RCIM_Cubic, leaveTangent=-5.9085, tangentMode=RCTM_User, time=0, value=1}
          {arriveTangent=-0.2496, interpMode=RCIM_Cubic, leaveTangent=-0.2496, tangentMode=RCTM_User, time=0.5, value=0.1}
         ]}}}
      CurveIndex=$Emitter.Age
      Scale Curve=150
     )
    ))
   &ScaleColor(
    Scale Alpha=&self:floatsel(
     float:NiagaraFloat=&FloatFromCurve(FloatCurve=DataInterface() {propertyValues {Curve {keys [<at 0.1219 4>,<at 0.194 1.5>,<at 0.7612 1.5>,<at 1 0>]}}})
     float001:NiagaraFloat=&FloatFromCurve(FloatCurve=DataInterface() {propertyValues {Curve {keys [<at 0.1219 4>,<at 0.194 1.5>,<at 0.7612 1.5>,<at 1 0>]}}}, Scale Curve=0.5)
     float002:NiagaraFloat=&FloatFromCurve(FloatCurve=DataInterface() {propertyValues {Curve {keys [<at 0.1219 4>,<at 0.194 1.5>,<at 0.7612 1.5>,<at 1 0>]}}}, Scale Curve=0.5)
     int32:NiagaraInt32=&SpawnGroup()
    )
    Scale RGB=&VectorFromFloat(Value=&FloatFromCurve(FloatCurve=DataInterface() {propertyValues {Curve {keys [{arriveTangent=1.8145, interpMode=RCIM_Cubic, leaveTangent=4.5795, tangentMode=RCTM_Break, time=0, value=0.4}, <at 0.3387 0.9565>, <at 1 1>]}}}))
    ScaleA=true
    ScaleRGB=true
   )
   /Niagara/Modules/Update/Forces/CurlNoiseForce(Noise Frequency=0.0012, Noise Strength=800, Pan Noise=true, Pan Noise Field=<0 0 1>)
   &Drag(
    Drag=&self:floatsel(
     float:NiagaraFloat=&FloatFromCurve(
      FloatCurve=DataInterface() {propertyValues {Curve {keys [
          {interpMode=RCIM_Cubic, leaveTangent=7.1629, tangentMode=RCTM_Break, time=0, value=0.5}
          {interpMode=RCIM_Cubic, tangentMode=RCTM_Break, time=0.2, value=1}
         ]}}}
      Scale Curve=40
     )
     float001:NiagaraFloat=6
     float002:NiagaraFloat=&FloatFromCurve(
      FloatCurve=DataInterface() {propertyValues {Curve {keys [
          {interpMode=RCIM_Cubic, leaveTangent=7.1629, tangentMode=RCTM_Break, time=0, value=0.5}
          {interpMode=RCIM_Cubic, tangentMode=RCTM_Break, time=0.2, value=1}
         ]}}}
      Scale Curve=60
     )
     int32:NiagaraInt32=&SpawnGroup()
    )
    Use Linear Drag=true
    Use Rotational Drag=false
   )
   &SolveForcesAndVelocity(Clamp Velocity=false, Limit Acceleration=false)
   &ScaleSpriteSize(Scale Factor=&Vector2DFromFloat(Value=&self:floatsel(
      float:NiagaraFloat=&FloatFromCurve(
       FloatCurve=DataInterface() {propertyValues {Curve {keys [
           {arriveTangent=8.1876, interpMode=RCIM_Cubic, leaveTangent=5.3162, tangentMode=RCTM_User, time=0, value=0.15}
           {arriveTangent=1.0025, interpMode=RCIM_Cubic, leaveTangent=1.0025, tangentMode=RCTM_User, time=0.166, value=0.7682}
           {arriveTangent=0.2754, interpMode=RCIM_Cubic, leaveTangent=0.2754, tangentMode=RCTM_User, time=1.0463, value=1}
          ]}}}
       Scale Curve=0.6
      )
      float001:NiagaraFloat=&FloatFromCurve(
       FloatCurve=DataInterface() {propertyValues {Curve {keys [
           {arriveTangent=8.1876, interpMode=RCIM_Cubic, leaveTangent=5.3162, tangentMode=RCTM_User, time=0, value=0.15}
           {arriveTangent=1.0025, interpMode=RCIM_Cubic, leaveTangent=1.0025, tangentMode=RCTM_User, time=0.166, value=0.7682}
           {arriveTangent=0.2754, interpMode=RCIM_Cubic, leaveTangent=0.2754, tangentMode=RCTM_User, time=1.0463, value=1}
          ]}}}
       Scale Curve=0.6
      )
      float002:NiagaraFloat=&FloatFromCurve(
       FloatCurve=DataInterface() {propertyValues {Curve {keys [
           {arriveTangent=2.0695, interpMode=RCIM_Cubic, leaveTangent=2.0695, tangentMode=RCTM_User, time=0, value=0.7}
           {arriveTangent=0.2867, interpMode=RCIM_Cubic, leaveTangent=0.2867, time=0.166, value=0.9}
           {arriveTangent=0.2754, interpMode=RCIM_Cubic, leaveTangent=0.2754, tangentMode=RCTM_User, time=1.0463, value=1}
          ]}}}
       Scale Curve=&Lerp_Float(A=0.4, B=0.6, Alpha=$System.Cam Adjust)
      )
      int32:NiagaraInt32=&SpawnGroup()
     )))
   &DynamicMaterialParameters(
    Index 0 Param 1=&Add_Float(
     A=&RandomRangeFloat(Minimum=0, Maximum=1)
     B=&FloatFromCurve(
      FloatCurve=DataInterface() {propertyValues {Curve {keys [
          {arriveTangent=2.7533, interpMode=RCIM_Cubic, leaveTangent=1.8958, tangentMode=RCTM_User, time=0, value=1}
          {arriveTangent=0.8292, interpMode=RCIM_Cubic, leaveTangent=0.8292, tangentMode=RCTM_User, time=0.321, value=1.3453}
          {arriveTangent=0.0529, interpMode=RCIM_Cubic, leaveTangent=0.2021, tangentMode=RCTM_User, time=1, value=1.5}
         ]}}}
      CurveIndex=&Add_Float(A=&Multiply_Float(A=$Particles.Initial.MaterialRandom, B=0.5), B=$Particles.NormalizedAge)
     )
    )
    Index 0 Param 2=&FloatFromCurve(FloatCurve=DataInterface() {propertyValues {Curve {keys [
         {arriveTangent=0.0453, interpMode=RCIM_Cubic, tangentMode=RCTM_Break, time=0, value=-1}
         {arriveTangent=0.0604, interpMode=RCIM_Cubic, leaveTangent=0.0604, tangentMode=RCTM_User, time=1, value=-1}
        ]}}})
    Index 0 Param 3=&FloatFromCurve(
     FloatCurve=DataInterface() {propertyValues {Curve {keys [
         {arriveTangent=0.0632, interpMode=RCIM_Cubic, leaveTangent=0.0632, tangentMode=RCTM_User, time=0, value=1}
         {
          arriveTangent=-0.6137
          arriveTangentWeight=0.0333
          interpMode=RCIM_Cubic
          leaveTangent=-0.6137
          leaveTangentWeight=0.3137
          tangentMode=RCTM_User
          time=0.1
          value=0.1
         }
         {arriveTangentWeight=0.1027, interpMode=RCIM_Cubic, tangentMode=RCTM_Break, time=0.4, value=0}
        ]}}}
     Scale Curve=&Lerp_Float(
      A=&RandomRangeFloat(Minimum=2, Maximum=4)
      B=&RandomRangeFloat(Minimum=0.35, Maximum=0.5)
      Alpha=$System.IsPlayerIndoors
     )
    )
    Index 0 Param 4=&FloatFromCurve(FloatCurve=DataInterface() {propertyValues {Curve {keys [<at 0.5 1>,<at 1 0>]}}})
    Index 1 Param 1=&FloatFromCurve(
     FloatCurve=DataInterface() {propertyValues {Curve {keys [
         {interpMode=RCIM_Cubic, tangentMode=RCTM_Break, time=0, value=1}
         {arriveTangentWeight=0.0833, interpMode=RCIM_Cubic, tangentMode=RCTM_Break, time=0.25, value=0.3}
        ]}}}
     Scale Curve=&RandomRangeFloat(Minimum=1.5, Maximum=3)
    )
    Index 1 Param 2=&FloatFromCurve(
     FloatCurve=DataInterface() {propertyValues {Curve {keys [
         {interpMode=RCIM_Cubic, leaveTangent=-2.5489, tangentMode=RCTM_Break, time=0, value=1}
         {interpMode=RCIM_Cubic, leaveTangent=0.9066, tangentMode=RCTM_Break, time=1, value=0}
        ]}}}
     Scale Curve=&RandomRangeFloat(Minimum=0.01, Maximum=0.1)
    )
    Index 1 Param 3=&FloatFromCurve(FloatCurve=DataInterface() {propertyValues {Curve {keys [
         {interpMode=RCIM_Cubic, leaveTangent=6.3917, tangentMode=RCTM_Break, time=0, value=0.2}
         {arriveTangent=1.7549, interpMode=RCIM_Cubic, leaveTangent=10.6767, tangentMode=RCTM_User, time=0.25, value=1}
        ]}}})
    Index 1 Param 4=&FloatFromCurve(
     FloatCurve=DataInterface() {propertyValues {Curve {keys [{arriveTangent=-0.136, interpMode=RCIM_Cubic, leaveTangent=-0.136, tangentMode=RCTM_User, time=0, value=1}, <at 1 0.95>]}}}
     Scale Curve=&RandomRangeFloat(Minimum=1.3, Maximum=2)
    )
    Index 2 Param 1=&FloatFromCurve(FloatCurve=DataInterface() {propertyValues {Curve {keys [
         {interpMode=RCIM_Cubic, leaveTangent=4.2485, tangentMode=RCTM_Break, time=0, value=0}
         {arriveTangent=1.0548, interpMode=RCIM_Cubic, leaveTangent=1.0548, tangentMode=RCTM_User, time=0.17, value=0.5715}
         <at 1 1>
        ]}}})
    Index 2 Param 2=&FloatFromCurve(
     FloatCurve=DataInterface() {propertyValues {Curve {keys [
         {arriveTangent=1.9056, interpMode=RCIM_Cubic, leaveTangent=1.9056, tangentMode=RCTM_User, time=0, value=0.2}
         {arriveTangent=0.0259, interpMode=RCIM_Cubic, leaveTangent=0.0259, tangentMode=RCTM_User, time=1, value=1.4}
        ]}}}
     Scale Curve=&RandomRangeFloat(Minimum=0.1, Maximum=0.2)
    )
    Index 2 Param 3=&FloatFromCurve(FloatCurve=DataInterface() {propertyValues {Curve {keys [
         {interpMode=RCIM_Cubic, leaveTangent=2.3064, tangentMode=RCTM_Break, time=0, value=1}
         {interpMode=RCIM_Cubic, leaveTangent=11.1233, tangentMode=RCTM_Break, time=1, value=2}
        ]}}})
    Index 2 Param 4=&FloatFromCurve(
     FloatCurve=DataInterface() {propertyValues {Curve {keys [
         {arriveTangent=-0.6867, interpMode=RCIM_Cubic, leaveTangent=-0.6867, tangentMode=RCTM_User, time=0, value=1}
         {interpMode=RCIM_Cubic, leaveTangent=-1.2956, tangentMode=RCTM_Break, time=1, value=0.75}
        ]}}}
     Scale Curve=1.5
    )
    Module.Index0Float0Write=true
    Module.Index0Float1Write=true
    Module.Index0Float2Write=true
    Module.Index0Float3Write=true
    Module.Index1Float0Write=true
    Module.Index1Float1Write=true
    Module.Index1Float2Write=true
    Module.Index1Float3Write=true
    Module.Index2Float0Write=true
    Module.Index2Float1Write=true
    Module.Index2Float2Write=true
    Module.Index2Float3Write=true
    Write Parameter Index 1=::NewEnumerator1
    Write Parameter Index 2=::NewEnumerator1
   )
   /SetVariables(
    Particles.Capture Dynamic A1:Vector4f=$Particles.DynamicMaterialParameter
    Particles.Capture Dynamic A2:Vector4f=$Particles.DynamicMaterialParameter1
    Particles.Capture Dynamic A3:Vector4f=$Particles.DynamicMaterialParameter2
   ) {isSetParameters=true}
   &DynamicMaterialParameters(
    Index 0 Param 1=&Add_Float(
     A=&RandomRangeFloat(Minimum=0, Maximum=1)
     B=&FloatFromCurve(
      FloatCurve=DataInterface() {propertyValues {Curve {keys [
          {arriveTangent=2.7533, interpMode=RCIM_Cubic, leaveTangent=1.8958, tangentMode=RCTM_User, time=0, value=1}
          {arriveTangent=0.8292, interpMode=RCIM_Cubic, leaveTangent=0.8292, tangentMode=RCTM_User, time=0.321, value=1.3453}
          {arriveTangent=0.0529, interpMode=RCIM_Cubic, leaveTangent=0.2021, tangentMode=RCTM_User, time=1, value=1.5}
         ]}}}
      CurveIndex=&Add_Float(A=&Multiply_Float(A=$Particles.Initial.MaterialRandom, B=0.5), B=$Particles.NormalizedAge)
     )
    )
    Index 0 Param 2=&FloatFromCurve(FloatCurve=DataInterface() {propertyValues {Curve {keys [
         {arriveTangent=0.0453, interpMode=RCIM_Cubic, tangentMode=RCTM_Break, time=0, value=-1}
         {arriveTangent=0.0604, interpMode=RCIM_Cubic, leaveTangent=0.0604, tangentMode=RCTM_User, time=1, value=-1}
        ]}}})
    Index 0 Param 4=&FloatFromCurve(FloatCurve=DataInterface() {propertyValues {Curve {keys [<at 0.5 1>,<at 1 0>]}}})
    Index 1 Param 1=&FloatFromCurve(
     FloatCurve=DataInterface() {propertyValues {Curve {keys [
         {interpMode=RCIM_Cubic, tangentMode=RCTM_Break, time=0, value=1}
         {arriveTangentWeight=0.0833, interpMode=RCIM_Cubic, tangentMode=RCTM_Break, time=0.25, value=0.3}
        ]}}}
     Scale Curve=&RandomRangeFloat(Minimum=1.5, Maximum=3)
    )
    Index 1 Param 2=&FloatFromCurve(
     FloatCurve=DataInterface() {propertyValues {Curve {keys [
         {interpMode=RCIM_Cubic, leaveTangent=-2.5489, tangentMode=RCTM_Break, time=0, value=1}
         {interpMode=RCIM_Cubic, leaveTangent=0.9066, tangentMode=RCTM_Break, time=1, value=0}
        ]}}}
     Scale Curve=&RandomRangeFloat(Minimum=0.01, Maximum=0.1)
    )
    Index 1 Param 3=&FloatFromCurve(FloatCurve=DataInterface() {propertyValues {Curve {keys [
         {interpMode=RCIM_Cubic, leaveTangent=6.3917, tangentMode=RCTM_Break, time=0, value=0.2}
         {arriveTangent=1.7549, interpMode=RCIM_Cubic, leaveTangent=10.6767, tangentMode=RCTM_User, time=0.25, value=1}
        ]}}})
    Index 1 Param 4=&FloatFromCurve(
     FloatCurve=DataInterface() {propertyValues {Curve {keys [{arriveTangent=-0.136, interpMode=RCIM_Cubic, leaveTangent=-0.136, tangentMode=RCTM_User, time=0, value=1}, <at 1 0.95>]}}}
     Scale Curve=&RandomRangeFloat(Minimum=1.3, Maximum=2)
    )
    Index 2 Param 1=&FloatFromCurve(FloatCurve=DataInterface() {propertyValues {Curve {keys [
         {interpMode=RCIM_Cubic, leaveTangent=4.2485, tangentMode=RCTM_Break, time=0, value=0}
         {arriveTangent=1.0548, interpMode=RCIM_Cubic, leaveTangent=1.0548, tangentMode=RCTM_User, time=0.17, value=0.5715}
         <at 1 1>
        ]}}})
    Index 2 Param 2=&FloatFromCurve(
     FloatCurve=DataInterface() {propertyValues {Curve {keys [
         {arriveTangent=1.9056, interpMode=RCIM_Cubic, leaveTangent=1.9056, tangentMode=RCTM_User, time=0, value=0.2}
         {arriveTangent=0.0259, interpMode=RCIM_Cubic, leaveTangent=0.0259, tangentMode=RCTM_User, time=1, value=1.4}
        ]}}}
     Scale Curve=&RandomRangeFloat(Minimum=0.1, Maximum=0.2)
    )
    Index 2 Param 3=&FloatFromCurve(FloatCurve=DataInterface() {propertyValues {Curve {keys [
         {interpMode=RCIM_Cubic, leaveTangent=2.3064, tangentMode=RCTM_Break, time=0, value=1}
         {interpMode=RCIM_Cubic, leaveTangent=11.1233, tangentMode=RCTM_Break, time=1, value=2}
        ]}}})
    Index 2 Param 4=&FloatFromCurve(
     FloatCurve=DataInterface() {propertyValues {Curve {keys [
         {arriveTangent=-0.6867, interpMode=RCIM_Cubic, leaveTangent=-0.6867, tangentMode=RCTM_User, time=0, value=1}
         {interpMode=RCIM_Cubic, leaveTangent=-1.2956, tangentMode=RCTM_Break, time=1, value=0.75}
        ]}}}
     Scale Curve=1.5
    )
    Module.Index0Float0Write=true
    Module.Index0Float1Write=true
    Module.Index0Float2Write=true
    Module.Index0Float3Write=true
    Module.Index1Float0Write=true
    Module.Index1Float1Write=true
    Module.Index1Float2Write=true
    Module.Index1Float3Write=true
    Module.Index2Float0Write=true
    Module.Index2Float1Write=true
    Module.Index2Float2Write=true
    Module.Index2Float3Write=true
    Write Parameter Index 1=::NewEnumerator1
    Write Parameter Index 2=::NewEnumerator1
   ) {module=DynamicMaterialParameters001}
   /SetVariables(
    Particles.Capture Dynamic B1:Vector4f=$Particles.DynamicMaterialParameter
    Particles.Capture Dynamic B2:Vector4f=$Particles.DynamicMaterialParameter1
    Particles.Capture Dynamic B3:Vector4f=$Particles.DynamicMaterialParameter2
   ) {isSetParameters=true}
   &DynamicMaterialParameters(
    Index 0 Param 1=&Add_Float(
     A=&RandomRangeFloat(Minimum=0, Maximum=1)
     B=&FloatFromCurve(
      FloatCurve=DataInterface() {propertyValues {Curve {keys [
          {arriveTangent=2.7533, interpMode=RCIM_Cubic, leaveTangent=1.8958, tangentMode=RCTM_User, time=0, value=1}
          {arriveTangent=0.8292, interpMode=RCIM_Cubic, leaveTangent=0.8292, tangentMode=RCTM_User, time=0.321, value=1.3453}
          {arriveTangent=0.0529, interpMode=RCIM_Cubic, leaveTangent=0.2021, tangentMode=RCTM_User, time=1, value=1.5}
         ]}}}
      CurveIndex=&Add_Float(A=&Multiply_Float(A=$Particles.Initial.MaterialRandom, B=0.5), B=$Particles.NormalizedAge)
     )
    )
    Index 0 Param 2=&FloatFromCurve(FloatCurve=DataInterface() {propertyValues {Curve {keys [
         {interpMode=RCIM_Cubic, leaveTangentWeight=0.3333, tangentMode=RCTM_Break, time=0, value=-1}
         {arriveTangentWeight=0.3339, interpMode=RCIM_Cubic, tangentMode=RCTM_User, time=1, value=-1}
        ]}}})
    Index 0 Param 3=&FloatFromCurve(
     FloatCurve=DataInterface() {propertyValues {Curve {keys [
         <at 0 1 RCIM_Cubic>
         {arriveTangentWeight=0.0333, interpMode=RCIM_Cubic, leaveTangent=-0.6137, leaveTangentWeight=0.3137, time=0.15, value=0}
        ]}}}
     Scale Curve=&Lerp_Float(
      A=&Lerp_Float(
       A=&RandomRangeFloat(Minimum=2, Maximum=1)
       B=&RandomRangeFloat(Minimum=250, Maximum=500)
       Alpha=$System.Cam Adjust
      )
      B=&Lerp_Float(
       A=&RandomRangeFloat(Minimum=0.15, Maximum=0.2)
       B=&RandomRangeFloat(Minimum=25, Maximum=50)
       Alpha=$System.Cam Adjust
      )
      Alpha=$System.IsPlayerIndoors
     )
    )
    Index 0 Param 4=&FloatFromCurve(FloatCurve=DataInterface() {propertyValues {Curve {keys [<at 0.5 1>,<at 1 0>]}}})
    Index 1 Param 1=&FloatFromCurve(
     FloatCurve=DataInterface() {propertyValues {Curve {keys [
         {interpMode=RCIM_Cubic, tangentMode=RCTM_Break, time=0, value=1}
         {arriveTangentWeight=0.0833, interpMode=RCIM_Cubic, tangentMode=RCTM_Break, time=0.25, value=0.3}
        ]}}}
     Scale Curve=&RandomRangeFloat(Minimum=1.5, Maximum=3)
    )
    Index 1 Param 2=&FloatFromCurve(
     FloatCurve=DataInterface() {propertyValues {Curve {keys [
         {interpMode=RCIM_Cubic, leaveTangent=-2.5489, tangentMode=RCTM_Break, time=0, value=1}
         {interpMode=RCIM_Cubic, leaveTangent=0.9066, tangentMode=RCTM_Break, time=1, value=0}
        ]}}}
     Scale Curve=&RandomRangeFloat(Minimum=0.01, Maximum=0.1)
    )
    Index 1 Param 3=&FloatFromCurve(FloatCurve=DataInterface() {propertyValues {Curve {keys [
         {interpMode=RCIM_Cubic, leaveTangent=6.3917, tangentMode=RCTM_Break, time=0, value=0.2}
         {arriveTangent=1.7549, interpMode=RCIM_Cubic, leaveTangent=10.6767, tangentMode=RCTM_User, time=0.25, value=1}
        ]}}})
    Index 1 Param 4=&FloatFromCurve(
     FloatCurve=DataInterface() {propertyValues {Curve {keys [{arriveTangent=-0.136, interpMode=RCIM_Cubic, leaveTangent=-0.136, tangentMode=RCTM_User, time=0, value=1}, <at 1 0.95>]}}}
     Scale Curve=&RandomRangeFloat(Minimum=1.3, Maximum=2)
    )
    Index 2 Param 1=&FloatFromCurve(FloatCurve=DataInterface() {propertyValues {Curve {keys [
         {interpMode=RCIM_Cubic, leaveTangent=4.2485, tangentMode=RCTM_Break, time=0, value=0}
         {arriveTangent=1.0548, interpMode=RCIM_Cubic, leaveTangent=1.0548, tangentMode=RCTM_User, time=0.17, value=0.5715}
         <at 1 1>
        ]}}})
    Index 2 Param 2=&FloatFromCurve(
     FloatCurve=DataInterface() {propertyValues {Curve {keys [
         {arriveTangent=1.9056, interpMode=RCIM_Cubic, leaveTangent=1.9056, tangentMode=RCTM_User, time=0, value=0.2}
         {arriveTangent=0.0259, interpMode=RCIM_Cubic, leaveTangent=0.0259, tangentMode=RCTM_User, time=1, value=1.4}
        ]}}}
     Scale Curve=&RandomRangeFloat(Minimum=0.1, Maximum=0.2)
    )
    Index 2 Param 3=&FloatFromCurve(FloatCurve=DataInterface() {propertyValues {Curve {keys [
         {interpMode=RCIM_Cubic, leaveTangent=2.3064, tangentMode=RCTM_Break, time=0, value=1}
         {interpMode=RCIM_Cubic, leaveTangent=11.1233, tangentMode=RCTM_Break, time=1, value=2}
        ]}}})
    Index 2 Param 4=&FloatFromCurve(
     FloatCurve=DataInterface() {propertyValues {Curve {keys [
         {arriveTangent=-0.6867, interpMode=RCIM_Cubic, leaveTangent=-0.6867, tangentMode=RCTM_User, time=0, value=1}
         {interpMode=RCIM_Cubic, leaveTangent=-1.2956, tangentMode=RCTM_Break, time=1, value=0.75}
        ]}}}
     Scale Curve=1.5
    )
    Module.Index0Float0Write=true
    Module.Index0Float1Write=true
    Module.Index0Float2Write=true
    Module.Index0Float3Write=true
    Module.Index1Float0Write=true
    Module.Index1Float1Write=true
    Module.Index1Float2Write=true
    Module.Index1Float3Write=true
    Module.Index2Float0Write=true
    Module.Index2Float1Write=true
    Module.Index2Float2Write=true
    Module.Index2Float3Write=true
    Write Parameter Index 1=::NewEnumerator1
    Write Parameter Index 2=::NewEnumerator1
   ) {module=DynamicMaterialParameters003}
   /SetVariables(
    Particles.Capture Dynamic C1:Vector4f=$Particles.DynamicMaterialParameter
    Particles.Capture Dynamic C2:Vector4f=$Particles.DynamicMaterialParameter1
    Particles.Capture Dynamic C3:Vector4f=$Particles.DynamicMaterialParameter2
   ) {isSetParameters=true}
   &DynamicMaterialParameters(
    Index 0 Vector 4:Vector4f=&self:V4_select(
     Vector 4:Vector4f=$Particles.Capture Dynamic A1
     Vector 4001:Vector4f=$Particles.Capture Dynamic B1
     Vector 4002:Vector4f=$Particles.Capture Dynamic C1
     int32:NiagaraInt32=&SpawnGroup()
    )
    Index 1 Vector 4:Vector4f=&self:V4_select(
     Vector 4:Vector4f=$Particles.Capture Dynamic A2
     Vector 4001:Vector4f=$Particles.Capture Dynamic B2
     Vector 4002:Vector4f=$Particles.Capture Dynamic C2
     int32:NiagaraInt32=&SpawnGroup()
    )
    Index 2 Vector 4:Vector4f=&self:V4_select(
     Vector 4:Vector4f=$Particles.Capture Dynamic A3
     Vector 4001:Vector4f=$Particles.Capture Dynamic B3
     Vector 4002:Vector4f=$Particles.Capture Dynamic C3
     int32:NiagaraInt32=&SpawnGroup()
    )
    Write Parameter Index 0=::NewEnumerator4
    Write Parameter Index 1=::NewEnumerator4
    Write Parameter Index 2=::NewEnumerator4
   ) {module=DynamicMaterialParameters002}
   &CameraOffset(Camera Offset Amount=&self:floatsel(
     float:NiagaraFloat=&Lerp_Float(
      A=0
      B=250
      Alpha=&FloatFromCurve(
       FloatCurve=DataInterface() {propertyValues {Curve {
          defaultValue=1
          keys [
           {arriveTangent=2.2026, interpMode=RCIM_Cubic, leaveTangent=2.2026, tangentMode=RCTM_User, time=0, value=0}
           {arriveTangent=0.7842, interpMode=RCIM_Cubic, leaveTangent=0.7842, tangentMode=RCTM_User, time=1, value=1}
          ]
         }}}
       Scale Curve=2
      )
     )
     float001:NiagaraFloat=&Lerp_Float(
      A=0
      B=500
      Alpha=&FloatFromCurve(FloatCurve=DataInterface() {propertyValues {Curve {
          defaultValue=1
          keys [
           {arriveTangent=2.2026, interpMode=RCIM_Cubic, leaveTangent=2.2026, tangentMode=RCTM_User, time=0, value=0}
           {arriveTangent=0.3024, interpMode=RCIM_Cubic, leaveTangent=0.3024, tangentMode=RCTM_User, time=1, value=1}
          ]
         }}})
     )
     float002:NiagaraFloat=&Lerp_Float(
      A=0
      B=1000
      Alpha=&FloatFromCurve(
       FloatCurve=DataInterface() {propertyValues {Curve {
          defaultValue=1
          keys [
           {arriveTangent=2.2026, interpMode=RCIM_Cubic, leaveTangent=2.2026, tangentMode=RCTM_User, time=0, value=0}
           {tangentMode=RCTM_User, time=1, value=1}
          ]
         }}}
       Scale Curve=0.5
      )
     )
     int32:NiagaraInt32=&SpawnGroup()
    ))
   /SetVariables(Particles.LerpPosition:NiagaraPosition=&LerpPosition(
     A=&SimulationPosition()
     B=$Particles.Position
     Alpha=&Lerp_Float(
      A=0.5
      B=1
      Alpha=&FloatFromCurve(FloatCurve=DataInterface() {propertyValues {Curve {
          defaultValue=1
          keys [{arriveTangent=1.6948, interpMode=RCIM_Cubic, leaveTangent=1.6948, tangentMode=RCTM_User, time=0, value=0}, <at 1 1 RCIM_Cubic>]
         }}})
     )
    )) {isSetParameters=true}
   &ScaleColor(
    Scale Alpha=&self:float_sel_4(
     float:NiagaraFloat=&Lerp_Float(A=&Lerp_Float(A=0.01, B=0.2, Alpha=$System.IsPlayerIndoors), B=1, Alpha=$System.Cam Adjust)
     float001:NiagaraFloat=0.2
     float002:NiagaraFloat=1
     float003:NiagaraFloat=1
     int32:NiagaraInt32=&SpawnGroup()
    )
    Scale RGB=&VectorFromFloat(Value=1)
    ScaleA=true
    ScaleRGB=true
   ) {module=ScaleColor001}
  ]
  rendererClasses [&NiagaraSpriteRendererProperties]
  renderers [
   {rendererClass=&NiagaraSpriteRendererProperties}
   {rendererClass=&NiagaraSpriteRendererProperties, rendererIndex=1}
   {rendererClass=&NiagaraSpriteRendererProperties, rendererIndex=2}
   {rendererClass=&NiagaraSpriteRendererProperties, rendererIndex=3}
  ]
 }
 Pressure_waveLW {rendererClasses [], renderers []}
 Rocks {
  emitterUpdateScript [
   &EmitterState(Enable Distance Culling=true, Module.bUseMaxDistance=true, Module.bUseMinDistance=false, Scalability Mode=::NewEnumerator1)
   &SpawnBurst_Instantaneous(
    Spawn Count=45
    Spawn Group=1
    Spawn Probability=&MakeCustomFloatFromBool(Bool=$User.IsInWater, True Float=0, False Float=1)
    Use Spawn Probability=true
    UseLoopCountLimit=false
   ) {module=SpawnBurst_Instantaneous001}
   &SpawnBurst_Instantaneous(
    Spawn Count=17
    Spawn Probability=&MakeCustomFloatFromBool(Bool=$User.IsInWater, True Float=0, False Float=1)
    Spawn Time=$System.Delay
    Use Spawn Probability=true
   )
  ]
  particleSpawnScript [
   /SetVariables(Particles.Size Scale:NiagaraFloat=&RandomRangeFloat(Minimum=5, Maximum=1)) {isSetParameters=true}
   &InitializeParticle(
    Color=<rgba 0 0 0 5>
    Lifetime Max=1.4
    Lifetime Min=0.8
    Lifetime Mode=::NewEnumerator1
    Mesh Scale Mode=::NewEnumerator3
    Mesh Uniform Scale=&RandomRangeFloat(Minimum=0.3, Maximum=1)
    Sprite Rotation Mode=::NewEnumerator2
    Sprite Size=&Multiply_Vector2D_ByFloat(Float=27.2202, Vector2D=&RandomRangeVector2D(Minimum=<3 3>, Maximum=<1 1>))
    Sprite Size Mode=::NewEnumerator4
    Sprite UV Mode=::NewEnumerator3
    UsePositionOffset=false
    Write Material Random=true
   )
   /Niagara/Modules/Spawn/Orientation/V2/InitialMeshOrientation(
    Rotation=&UniformRangedVector(Minimum=<0 0 0>, Maximum=<1 1 1>)
    Use Orientation Vector=true
    Use Rotation Vector=true
   )
   &ShapeLocation(
    MaskSpawnGroup=false
    Override Seed=false
    Sphere Radius=&Multiply_Float(A=0.5, B=$User.FXScale)
    UseOwnerScale=true
   )
   &AddVelocity(Velocity=&Multiply_VectorByFloat(
     Vector=&UniformRangedVector(Minimum=<-800 -800 200>, Maximum=<800 800 600>)
     Float=&Multiply_Float(A=0.003, B=$User.FXScale)
    ))
   &AddVelocity(Velocity=&Multiply_Vector(
     A=&Lerp_Vector(
      A=<0 0 600>
      B=<0 0 2000>
      Alpha=&FloatFromCurve(
       FloatCurve=DataInterface() {propertyValues {Curve {keys [
           {interpMode=RCIM_Cubic, tangentMode=RCTM_User, time=0, value=0}
           {arriveTangent=3.1561, interpMode=RCIM_Cubic, leaveTangent=3.1561, tangentMode=RCTM_User, time=1, value=1}
          ]}}}
       CurveIndex=&RandomRangeFloat(Minimum=0, Maximum=1)
      )
     )
     B=<1 1 1>
    )) {module=AddVelocity001}
   /SetVariables(Particles.VisibilityTag=&SelectIntFromArray(Direct Array Index=&SpawnGroup(), Int Selection Array=DataInterface() {propertyValues {IntData [0,1]}})) {isSetParameters=true}
  ]
  particleUpdateScript [
   &ParticleState()
   &SpriteRotationRate(Rotation Rate=&Multiply_Float(A=120, B=&RandomRangeFloat(Minimum=-1, Maximum=1)))
   &ScaleSpriteSize(Uniform Curve Sprite Scale=DataInterface() {propertyValues {Curve {
       defaultValue=0
       keys [
        {tangentMode=RCTM_User, time=0, value=1}
        {arriveTangent=-0.0405, interpMode=RCIM_Cubic, leaveTangent=-0.0405, tangentMode=RCTM_User, time=0.8, value=1}
        {arriveTangent=-8.0089, interpMode=RCIM_Cubic, leaveTangent=-8.0089, tangentMode=RCTM_User, time=1, value=0}
       ]
      }}})
   &AccelerationForce(Acceleration=<0 0 -2200>)
   &Collision(
    Age Colliding Particles=false
    Collision Enabled=&Boolean_AndOperation(
     Bool1=&IsPlatformSetActive(New Platform Set=DataInterface() {propertyValues {Platforms {qualityLevelMask=-4}}})
     Bool2=&SetBoolByFloatComparison(A=$Particles.Age, B=0.2)
    )
    Friction During a Bounce=0.4
    Module.Control Rest State Via Surface Normal=false
    Module.EnableMaxCollisionCount=false
    Module.Randomize Collision Normal Vector=false
    Particle Radius Scale=0.8
    Restitution=0.2
   )
   &MeshRotationForce(Rotation=&RandomVector(Vector Scale=&RandomRangeFloat(Minimum=0.5, Maximum=1)))
   &SolveForcesAndVelocity(Clamp Velocity=false, Limit Acceleration=false)
   &ScaleMeshSize(Scale Factor=&VectorFromFloat(Value=&FloatFromCurve(FloatCurve=DataInterface() {propertyValues {Curve {keys [<at 0.6 1 RCIM_Cubic>,<at 1 0 RCIM_Cubic>]}}})))
   &ScaleColor(
    Scale Alpha=&FloatFromCurve(FloatCurve=DataInterface() {propertyValues {Curve {keys [<at 0.8 1>,<at 1 0>]}}})
    Scale RGB=<0.1 0.1 0.1>
    ScaleA=true
    ScaleRGB=true
   )
   /Niagara/Modules/Update/Orientation/V2/UpdateMeshOrientation(
    Rotation Rate:NiagaraFloat=&Lerp_Float(A=-3, B=3, Alpha=$Particles.Initial.MaterialRandom)
    Rotation Vector:Vector3f=/Niagara/DynamicInputs/Math/SlerpVector(Vector A:Vector3f=<-1 0 1>, Vector B:Vector3f=<0 1 -1>, Alpha:NiagaraFloat=$Particles.Initial.MaterialRandom, Scale:NiagaraFloat=1)
   )
  ]
  rendererClasses [&NiagaraMeshRendererProperties, &NiagaraSpriteRendererProperties]
  renderers [
   {rendererClass=&NiagaraMeshRendererProperties, rendererIndex=0}
   {rendererClass=&NiagaraSpriteRendererProperties, rendererIndex=1}
  ]
 }
 Secondary_Gloops {
  emitterUpdateScript [
   &EmitterState(
    Enable Distance Culling=true
    MaxDistance=12000
    Module.bUseMaxDistance=true
    Module.bUseMinDistance=false
    Scalability Mode=::NewEnumerator1
   )
   &SpawnBurst_Instantaneous(
    Spawn Count=&RandomRangeInt(Minimum=45, Maximum=20)
    Spawn Group=1
    Spawn Probability=&MakeCustomFloatFromBool(Bool=$User.IsInWater, True Float=1, False Float=0)
    Spawn Time=$System.Delay
    Use Spawn Probability=true
    UseLoopCountLimit=false
   ) {module=SpawnBurst_Instantaneous001}
  ]
  particleSpawnScript [
   &InitializeParticle(
    Color=<rgba 3.2063 4.2281 5 1>
    Lifetime Max=1.4
    Lifetime Mode=::NewEnumerator1
    Sprite Rotation Mode=::NewEnumerator2
    Sprite Size Mode=::NewEnumerator1
    Uniform Sprite Size Max=60
    Uniform Sprite Size Min=40
    UsePositionOffset=false
    Write Material Random=true
   ) {module=InitializeParticle001}
   &SphereLocation(
    Hemisphere Z=true
    MaskSpawnGroup=false
    Override Seed=false
    Sphere Coordinate Space=::World
    Sphere Radius=30
    Surface Only=false
    UseNonUniformScale=false
    UseOffset=false
    UseSphereOrientation=false
   )
   &AddVelocityFromPoint(
    Module.Use Falloff Distance=false
    Origin Offset=<0 0 -12>
    Origin Offset Coordinate Space=::World
    Velocity Strength=&Multiply_Float(
     A=&UniformRangedFloat(Minimum=700, Maximum=1150)
     B=&MaskFloatBySpawnGroup(Masked Float=2, Passthrough Float=2.5)
    )
   )
  ]
  particleUpdateScript [
   &ParticleState()
   &AccelerationForce(Acceleration=&Multiply_VectorByFloat(Vector=&UniformRangedVector(Minimum=<0 0 -1000>, Maximum=<0 0 -1800>), Float=2.35))
   &SolveForcesAndVelocity(Clamp Velocity=false, Limit Acceleration=false)
   &ScaleSpriteSize(Scale Factor=&Vector2DFromFloat(Value=&FloatFromCurve(
      FloatCurve=DataInterface() {propertyValues {Curve {keys [
          {arriveTangent=12.5197, interpMode=RCIM_Cubic, leaveTangent=12.5197, tangentMode=RCTM_User, time=0, value=0.4376}
          {arriveTangent=0.8929, interpMode=RCIM_Cubic, leaveTangent=0.8929, tangentMode=RCTM_User, time=1, value=5}
         ]}}}
      Scale Curve=2
     )))
   &ScaleColor(
    Scale Alpha=&FloatFromCurve(FloatCurve=DataInterface() {propertyValues {Curve {keys [<at 0 -1>,<at 0.009 0.9316>,<at 0.7102 -0.1519>,<at 1 -1>]}}})
    Scale RGB=&VectorFromCurve(VectorCurve=DataInterface() {propertyValues {XCurve {keys [<at 0 2>,<at 1 0>]}, YCurve {keys [<at 0 2>,<at 1 0>]}, ZCurve {keys [<at 0 2>,<at 1 0>]}}})
    ScaleA=true
    ScaleRGB=true
   )
   &DynamicMaterialParameters(
    Index 0 Param 1=&FloatFromCurve(FloatCurve=DataInterface() {propertyValues {Curve {keys [
         {arriveTangent=-2.0663, interpMode=RCIM_Cubic, leaveTangent=-2.0663, tangentMode=RCTM_User, time=0.25, value=1}
         {arriveTangent=-1.3063, interpMode=RCIM_Cubic, leaveTangent=-1.3063, tangentMode=RCTM_User, time=1.0011, value=0}
        ]}}})
    Index 0 Param 2=&FloatFromCurve(FloatCurve=DataInterface() {propertyValues {Curve {keys [<at -0.0086 -1.3001>,<at 0.9945 1.9153>]}}})
    Index 0 Param 3=0.5
    Module.Param0WriteEnabled=true
    Module.Param1WriteEnabled=false
    Module.Param2WriteEnabled=true
    Module.Param3WriteEnabled=true
   ) {module=DynamicMaterialParameters001}
   &SpriteRotationRate(Rotation Rate=&Lerp_Float(A=-150, B=150, Alpha=$Particles.Initial.MaterialRandom))
  ]
  rendererClasses [&NiagaraSpriteRendererProperties]
  renderers [{rendererClass=&NiagaraSpriteRendererProperties}]
 }
 SmokeyFingers001 {
  emitterUpdateScript [
   &EmitterState()
   &SpawnRate(SpawnGroup=1, SpawnRate=10, Use Spawn Probability=false) {enabled=false}
   &SpawnBurst_Instantaneous(
    Spawn Count=6
    Spawn Group=1
    Spawn Time=&Add_Float(A=$System.Delay, B=0)
    Use Spawn Probability=false
    UseLoopCountLimit=false
   ) {module=SpawnBurst_Instantaneous001}
   &SpawnBurst_Instantaneous(Spawn Count=12, Spawn Time=&Add_Float(A=$System.Delay, B=0.02), Use Spawn Probability=false)
  ]
  particleSpawnScript [
   /Niagara/Modules/Spawn/Initialization/InitializeParticle(
    Color=<rgba 0 1 1 0.85>
    Lifetime=&UniformRangedFloat(Minimum=0.2, Maximum=0.25)
    Mesh Scale=&Multiply_VectorByFloat(
     Vector=&UniformRangedVector(Minimum=<4 25 25>, Maximum=<6 25 25>)
     Float=&Multiply_Float(A=0.004, B=$User.FXScale)
    )
    Module.Write Color=true
    Module.Write Lifetime=true
    Module.Write Mass=true
    Module.Write Material Random=true
    Module.Write Position=true
    Module.Write Scale=true
    Module.Write SpriteRotation=false
    Module.Write SpriteSize=false
    Position=/Niagara/Functions/Localspace/SimulationPosition()
    Sprite Size=<25 25>
   ) {enabled=false}
   /Niagara/Modules/Spawn/Initialization/V2/InitializeParticle(
    Color=<rgba 0 1 1 0.85>
    Lifetime=&UniformRangedFloat(Minimum=0.3, Maximum=0.25)
    Mesh Scale=&Multiply_VectorByFloat(
     Vector=&UniformRangedVector(Minimum=<4 25 25>, Maximum=<6 25 25>)
     Float=&Multiply_Float(A=0.004, B=$User.FXScale)
    )
    Mesh Scale Mode=::NewEnumerator4
    Position Offset=&Multiply_VectorByFloat(Vector=<0 0 1>, Float=50)
    UsePositionOffset=true
    Write Material Random=true
   ) {module=InitializeParticle001}
   /Niagara/Modules/Spawn/Orientation/V3/InitialMeshOrientation(
    Mesh Orientation Mode=::NewEnumerator2
    Orientation Vector=&MaskVectorBySpawnGroup(
     Masked Vector=&RandomVector(Vector Scale=1)
     Passthrough Vector=&self:ScratchDynamicInput_02(
      Default:Vector3f=<0 0 -0.1>
      Angle:NiagaraFloat=&self:ScratchDynamicInput_03(
       Vector001:Vector3f=&ReturnCameraVector(Camera Vectors=::NewEnumerator1, Coordinate Space=::Simulation)
       Offset:NiagaraFloat=&Add_Float(
        A=-0.25
        B=&Multiply_Float(A=0.2, B=&RandomRangeFloat(Minimum=-1, Maximum=1))
       )
      )
     )
    )
    Rotation=&MaskVectorBySpawnGroup(Masked Vector=&RandomVector(Vector Scale=1), Passthrough Vector=<0 0 0>)
    Rotation Coordinate Space=::NewEnumerator2
    Use Rotation Vector=true
   ) {module=InitialMeshOrientation001}
   &CameraOffset(Camera Offset Amount=-25)
  ]
  particleUpdateScript [
   &GravityForce(Gravity=<0 0 -0.01>)
   &SolveForcesAndVelocity(Clamp Velocity=false, Limit Acceleration=false)
   &ParticleState()
   &ScaleMeshSize(Scale Factor=&VectorFromFloat(Value=&Lerp_Float(A=0.4, B=1, Alpha=$System.Cam Adjust))) {module=ScaleMeshSize001}
   &ScaleColor(
    Scale Alpha=&FloatFromCurve(
     FloatCurve=DataInterface() {propertyValues {Curve {keys [
         {arriveTangent=200.5769, interpMode=RCIM_Cubic, leaveTangent=200.5752, tangentMode=RCTM_User, time=0, value=0}
         {arriveTangent=0.0668, interpMode=RCIM_Cubic, leaveTangent=0.0668, tangentMode=RCTM_User, time=0.01, value=1}
         <at 1 0>
        ]}}}
     Scale Curve=&Lerp_Float(
      A=&Lerp_Float(A=0.5, B=1, Alpha=$System.Cam Adjust)
      B=&Lerp_Float(A=0.35, B=1, Alpha=$System.Cam Adjust)
      Alpha=$System.IsPlayerIndoors
     )
    )
    Scale RGB=&VectorFromCurve(VectorCurve=DataInterface() {propertyValues {
       bUseLUT=false
       XCurve {keys [
         <at -0.0009 0.9906 RCIM_Cubic>
         {arriveTangent=-1.9223, interpMode=RCIM_Cubic, leaveTangent=-1.0067, tangentMode=RCTM_Break, time=0.3779, value=0.4}
         <at 1 0 RCIM_Cubic>
        ]}
       YCurve {keys [
         {arriveTangent=-3.6896, interpMode=RCIM_Cubic, leaveTangent=-3.6896, tangentMode=RCTM_User, time=-0.0011, value=1.0067}
         {arriveTangent=-0.3437, interpMode=RCIM_Cubic, leaveTangent=-0.3437, tangentMode=RCTM_User, time=0.3551, value=0.2247}
         <at 1 0 RCIM_Cubic>
        ]}
       ZCurve {keys [
         <at -0.0011 0 RCIM_Cubic>
         {arriveTangent=-0.223, interpMode=RCIM_Cubic, leaveTangent=-0.223, tangentMode=RCTM_User, time=0.9965, value=0.7415}
        ]}
      }})
    ScaleA=true
    ScaleRGB=true
   )
   &ScaleMeshSize(Scale Factor=&VectorFromFloat(Value=&FloatFromCurve(FloatCurve=DataInterface() {propertyValues {Curve {keys [
          {arriveTangent=4.641, interpMode=RCIM_Cubic, leaveTangent=4.641, tangentMode=RCTM_User, time=0, value=0}
          {arriveTangent=0.8503, interpMode=RCIM_Cubic, leaveTangent=0.8503, time=0.1368, value=0.6768}
          <at 1 0.8503 RCIM_Cubic>
         ]}}})))
   &DynamicMaterialParameters(
    Index 0 Param 1=&FloatFromCurve(FloatCurve=DataInterface() {propertyValues {Curve {keys [
         {arriveTangent=3.1529, interpMode=RCIM_Cubic, leaveTangent=3.1529, tangentMode=RCTM_User, time=0, value=0.5}
         {arriveTangent=0.4946, interpMode=RCIM_Cubic, leaveTangent=0.4946, tangentMode=RCTM_User, time=0.1931, value=0.9256}
         <at 1.001 1.1573>
        ]}}})
    Index 0 Param 2=&FloatFromCurve(FloatCurve=DataInterface() {propertyValues {Curve {keys [<at 0.505 0.9841>,<at 0.7325 0.0093>,<at 1.0011 0.002>]}}})
    Index 0 Param 3=&FloatFromCurve(
     FloatCurve=DataInterface() {propertyValues {Curve {keys [
         {arriveTangent=3.7271, interpMode=RCIM_Cubic, leaveTangent=3.7271, tangentMode=RCTM_User, time=-0.002, value=1}
         {arriveTangent=2.1936, interpMode=RCIM_Cubic, leaveTangent=2.1936, time=0.4642, value=2.7053}
         <at 0.9979 3.1936>
        ]}}}
     Scale Curve=&Lerp_Float(
      A=&Multiply_Float(A=0.004, B=$User.FXScale)
      B=&Multiply_Float(
       A=&Multiply_Float(A=0.004, B=$User.FXScale)
       B=&Lerp_Float(
        A=0.15
        B=1
        Alpha=&FloatFromCurve(
         FloatCurve=DataInterface() {propertyValues {Curve {keys [
             {arriveTangent=2.7997, interpMode=RCIM_Cubic, leaveTangent=2.7997, tangentMode=RCTM_User, time=0, value=0}
             {
              arriveTangent=1.1557
              arriveTangentWeight=0.8079
              interpMode=RCIM_Cubic
              leaveTangent=1.1557
              leaveTangentWeight=0.8079
              tangentMode=RCTM_User
              time=1
              value=1
             }
            ]}}}
         CurveIndex=$System.Cam Adjust
        )
       )
      )
      Alpha=$System.IsPlayerIndoors
     )
    )
    Module.Param0WriteEnabled=true
    Module.Param1WriteEnabled=true
    Module.Param2WriteEnabled=true
    Module.Param3WriteEnabled=true
   )
  ]
  rendererClasses [&NiagaraMeshRendererProperties]
  renderers [{rendererClass=&NiagaraMeshRendererProperties, rendererIndex=0}]
 }
 Sparks {
  emitterSpawnScript [/SetVariables(Emitter.NumSource:NiagaraInt32=&RandomRangeInt(Minimum=4, Maximum=7)) {isSetParameters=true, module=SetVariables_#7}]
  emitterUpdateScript [
   &EmitterState(
    Enable Distance Culling=true
    MaxDistance=15000
    Module.bUseMaxDistance=true
    Module.bUseMinDistance=false
    Scalability Mode=::NewEnumerator1
   )
   &SpawnBurst_Instantaneous(
    Spawn Count=&RandomRangeInt(Minimum=50, Maximum=35)
    Spawn Group=1
    Spawn Time=$System.Delay
    Use Spawn Probability=false
    UseLoopCountLimit=false
   ) {module=SpawnBurst_Instantaneous001}
  ]
  particleSpawnScript [
   &InitializeParticle(
    Color=<rgba 1 0.3214 0 1>
    Lifetime Min=1.5
    Lifetime Mode=::NewEnumerator1
    Material Random:NiagaraFloat=&MaskFloat(Mask Parameter=&SpawnGroup(), Mask=0, Masked Float=0, Passthrough Float=1)
    Sprite Size=&MaskVector2dBySpawnGroup(
     Masked Vector2D=&RandomRangeVector2D(Minimum=<5 5>, Maximum=<6 6>)
     Passthrough Vector2D=&RandomRangeVector2D(Minimum=<2 2>, Maximum=<1.5 4>)
     Spawn Group Mask=0
    )
    Sprite Size Mode=::NewEnumerator4
    UsePositionOffset=false
    Write Material Random=true
   )
   &ShapeLocation(
    Hemisphere Distribution=<1 0.5>
    MaskSpawnGroup=false
    Offset Mode=::NewEnumerator0
    Override Seed=false
    Sphere Radius=&Multiply_Float(A=0, B=$User.FXScale)
    Sphere Surface Distribution=0.75
    UseOwnerScale=true
   )
   &AddVelocity(
    Cone Angle=135
    Cone Axis=<0 0 1>
    Inner Cone Angle=60
    Module.UseConeFalloff=false
    Module.UseVelDistribution=false
    Override Seed=false
    Velocity Mode=::NewEnumerator2
    Velocity Speed=&Multiply_Float(A=&RandomRangeFloat(Minimum=600, Maximum=3500), B=0.6)
   ) {module=AddVelocity001}
   /SetVariables(Particles.Position=&AddVectorToPosition(Position=$Particles.Position, Vector=<0 0 -100>)) {isSetParameters=true, module=SetVariables_#8}
  ]
  particleUpdateScript [
   &ParticleState()
   &Drag(Drag=2, Use Linear Drag=true, Use Rotational Drag=false)
   &GravityForce(Gravity=&Multiply_VectorByFloat(
     Vector=<0 0 -980>
     Float=&FloatFromCurve(FloatCurve=DataInterface() {propertyValues {Curve {keys [<at 0 1 RCIM_Cubic>, {arriveTangent=-3.1991, interpMode=RCIM_Cubic, leaveTangent=-3.1991, tangentMode=RCTM_User, time=1, value=0}]}}})
    ))
   &SolveForcesAndVelocity(Clamp Velocity=false, Limit Acceleration=false)
   &ScaleSpriteSize(
    Uniform Curve Scale=3
    Uniform Curve Sprite Scale=DataInterface() {propertyValues {Curve {keys [
        {interpMode=RCIM_Cubic, tangentMode=RCTM_Break, time=0, value=1}
        {arriveTangent=-3.2052, interpMode=RCIM_Cubic, leaveTangent=-2.3701, tangentMode=RCTM_Break, time=0.9, value=0.125}
        <at 1 0 RCIM_Cubic>
       ]}}}
   )
   &ScaleSpriteSize(Scale Sprite Size Mode=::NewEnumerator0, Uniform Scale Factor=&Lerp_Float(A=0.5, B=1, Alpha=$System.Cam Adjust)) {module=ScaleSpriteSize002}
   &ScaleSpriteSizeBySpeed(Max Scale Factor=<1 3>, Min Scale Factor=<0.8 0.5>)
   &ScaleColor(
    Scale Alpha=&FloatFromCurve(FloatCurve=DataInterface() {propertyValues {Curve {keys [
         {interpMode=RCIM_Cubic, tangentMode=RCTM_User, time=0.0048, value=0.438}
         {arriveTangent=0.3021, interpMode=RCIM_Cubic, leaveTangent=0.3021, tangentMode=RCTM_User, time=0.0983, value=1.5996}
         {arriveTangent=-1.0514, interpMode=RCIM_Cubic, leaveTangent=-1.0514, tangentMode=RCTM_User, time=0.3368, value=0.5383}
         {arriveTangent=-0.8117, interpMode=RCIM_Cubic, leaveTangent=-0.8117, tangentMode=RCTM_User, time=0.5811, value=1.092}
         <at 1 0>
        ]}}})
    Scale RGB=&VectorFromCurve(VectorCurve=DataInterface() {propertyValues {
       XCurve {keys [<at 0 16 RCIM_Cubic>,<at 0.8 1 RCIM_Cubic>]}
       YCurve {keys [<at 0 16 RCIM_Cubic>,<at 0.8 0.1249 RCIM_Cubic>]}
       ZCurve {keys [<at 0 16 RCIM_Cubic>,<at 0.8 0.2547 RCIM_Cubic>]}
      }})
    ScaleA=true
    ScaleRGB=true
   )
   &ScaleColor(
    Scale RGB=&VectorFromFloat(Value=&Lerp_Float(
      A=1
      B=30
      Alpha=&FloatFromCurve(
       FloatCurve=DataInterface() {propertyValues {Curve {keys [
           {interpMode=RCIM_Cubic, tangentMode=RCTM_Break, time=0.4, value=0}
           {arriveTangentWeight=0.8501, interpMode=RCIM_Cubic, tangentMode=RCTM_Break, time=1, value=1}
          ]}}}
       CurveIndex=&RandomRangeFloat(Minimum=0, Maximum=1)
      )
     ))
    ScaleA=false
    ScaleRGB=true
   ) {module=ScaleColor001}
   &ScaleSpriteSize(
    Scale Sprite Size Mode=::NewEnumerator0
    Uniform Scale Factor=/Game/Effects/Niagara/DynamicInputs/Scalability/FN_ScalabilityFloatModifier(
     Value at DistanceFraction 0:NiagaraFloat=1
     Value at DistanceFraction 1:NiagaraFloat=3
     DistanceRamp:NiagaraDataInterfaceCurve=DataInterface() {propertyValues {Curve {keys [<at 0.1 0>,<at 0.5 1>]}}}
    )
   ) {module=ScaleSpriteSize001}
  ]
  rendererClasses [&NiagaraSpriteRendererProperties]
  renderers [{rendererClass=&NiagaraSpriteRendererProperties}]
 }
 Sparks001 {
  emitterSpawnScript [/SetVariables(Emitter.NumSource:NiagaraInt32=&RandomRangeInt(Minimum=4, Maximum=7)) {isSetParameters=true, module=SetVariables_#7}]
  emitterUpdateScript [
   &EmitterState(
    Enable Distance Culling=true
    MaxDistance=15000
    Module.bUseMaxDistance=true
    Module.bUseMinDistance=false
    Scalability Mode=::NewEnumerator1
   )
   &SpawnBurst_Instantaneous(Spawn Count=$Emitter.NumSource, Spawn Time=$System.Delay, Use Spawn Probability=true, UseLoopCountLimit=false)
   &SpawnBurst_Instantaneous(
    Spawn Count=&RandomRangeInt(Minimum=200, Maximum=400)
    Spawn Group=1
    Spawn Time=$System.Delay
    Use Spawn Probability=false
    UseLoopCountLimit=false
   ) {module=SpawnBurst_Instantaneous001}
  ]
  particleSpawnScript [
   &InitializeParticle(
    Color=<rgba 1 0.3214 0 1>
    Lifetime Max=0.5
    Lifetime Min=0.75
    Lifetime Mode=::NewEnumerator1
    Material Random:NiagaraFloat=&MaskFloat(Mask Parameter=&SpawnGroup(), Mask=0, Masked Float=0, Passthrough Float=1)
    Sprite Size=&MaskVector2dBySpawnGroup(
     Masked Vector2D=&RandomRangeVector2D(Minimum=<3 3>, Maximum=<4 4>)
     Passthrough Vector2D=&RandomRangeVector2D(Minimum=<2 2>, Maximum=<1.5 4>)
     Spawn Group Mask=0
    )
    Sprite Size Mode=::NewEnumerator4
    UsePositionOffset=false
    Write Material Random=true
   )
   &ShapeLocation(
    Hemisphere Distribution=<1 0.5>
    MaskSpawnGroup=false
    Offset Mode=::NewEnumerator0
    Override Seed=false
    Sphere Radius=&Multiply_Float(A=&MaskFloatBySpawnGroup(Masked Float=0.05, Passthrough Float=0.35), B=$User.FXScale)
    Sphere Surface Distribution=0.75
    UseOwnerScale=true
   )
   &AddVelocity(
    Velocity=&MaskVectorBySpawnGroup(
     Masked Vector=&RandomVector(Vector Scale=1, Constrain to Hemisphere=true, Hemisphere Constraint Vector=<0 0 1>)
     Passthrough Vector=&Multiply_Vector(
      A=&Lerp_Vector(
       A=<800 800 800>
       B=<-800 -800 2500>
       Alpha=&FloatFromCurve(
        FloatCurve=DataInterface() {propertyValues {Curve {keys [
            {interpMode=RCIM_Cubic, tangentMode=RCTM_User, time=0, value=0}
            {arriveTangent=3.1561, interpMode=RCIM_Cubic, leaveTangent=3.1561, tangentMode=RCTM_User, time=1, value=1}
           ]}}}
        CurveIndex=&RandomRangeFloat(Minimum=0, Maximum=1)
       )
      )
      B=<1 1 1>
     )
    )
    Velocity Speed Scale=&MaskFloatBySpawnGroup(Masked Float=&RandomRangeFloat(Minimum=500, Maximum=5000), Passthrough Float=1)
   ) {module=AddVelocity001}
   &AddVelocity(
    Module.Use Falloff Distance=false
    Origin Offset=<0 0 200>
    Override Seed=false
    Velocity Mode=::NewEnumerator1
    Velocity Speed=&MaskFloatBySpawnGroup(
     Masked Float=0
     Passthrough Float=&Lerp_Float(
      A=500
      B=3000
      Alpha=&FloatFromCurve(
       FloatCurve=DataInterface() {propertyValues {Curve {keys [
           {interpMode=RCIM_Cubic, tangentMode=RCTM_User, time=0, value=0}
           {arriveTangent=3.1561, interpMode=RCIM_Cubic, leaveTangent=3.1561, tangentMode=RCTM_User, time=1, value=1}
          ]}}}
       CurveIndex=&RandomRangeFloat(Minimum=0, Maximum=1)
      )
     )
    )
   ) {module=AddVelocity002}
   /SetVariables(Particles.Position=&AddVectorToPosition(Position=$Particles.Position, Vector=<0 0 -100>)) {isSetParameters=true, module=SetVariables_#8}
  ]
  particleUpdateScript [
   &ParticleState()
   &AccelerationForce(Acceleration=&RandomRangeVector(Minimum=<0 0 -800>, Maximum=<0 0 -1100>))
   /Niagara/Modules/Update/Forces/V2/CurlNoiseForce(
    Noise Frequency=120
    Noise Strength=&FloatFromCurve(
     FloatCurve=DataInterface() {propertyValues {Curve {keys [{interpMode=RCIM_Cubic, tangentMode=RCTM_Break, time=0, value=1}, {interpMode=RCIM_Cubic, tangentMode=RCTM_Break, time=1, value=0}]}}}
     Scale Curve=1000
    )
    Pan Noise=false
   )
   &Drag(
    Drag=&FloatFromCurve(
     FloatCurve=DataInterface() {propertyValues {Curve {keys [
         {interpMode=RCIM_Cubic, leaveTangentWeight=0.5018, tangentMode=RCTM_Break, time=0, value=1}
         {arriveTangent=-12.4068, interpMode=RCIM_Cubic, leaveTangent=-12.4068, tangentMode=RCTM_User, time=0.2, value=0}
        ]}}}
     Scale Curve=5
    )
    Use Linear Drag=true
    Use Rotational Drag=false
   )
   &SolveForcesAndVelocity(Clamp Velocity=false, Limit Acceleration=false)
   &ScaleSpriteSize(
    Uniform Curve Scale=3
    Uniform Curve Sprite Scale=DataInterface() {propertyValues {Curve {keys [
        {interpMode=RCIM_Cubic, tangentMode=RCTM_Break, time=0, value=1}
        {arriveTangent=-3.2052, interpMode=RCIM_Cubic, leaveTangent=-2.3701, tangentMode=RCTM_Break, time=0.9, value=0.125}
        <at 1 0 RCIM_Cubic>
       ]}}}
   )
   &ScaleSpriteSizeBySpeed(Max Scale Factor=<1 3>)
   &ScaleColor(
    Scale Alpha=&FloatFromCurve(FloatCurve=DataInterface() {propertyValues {Curve {keys [
         {interpMode=RCIM_Cubic, tangentMode=RCTM_User, time=0.0048, value=0.438}
         {arriveTangent=0.3021, interpMode=RCIM_Cubic, leaveTangent=0.3021, tangentMode=RCTM_User, time=0.0983, value=1.5996}
         {arriveTangent=-1.0514, interpMode=RCIM_Cubic, leaveTangent=-1.0514, tangentMode=RCTM_User, time=0.3368, value=0.5383}
         {arriveTangent=-0.8117, interpMode=RCIM_Cubic, leaveTangent=-0.8117, tangentMode=RCTM_User, time=0.5811, value=1.092}
         <at 1 0>
        ]}}})
    Scale RGB=&VectorFromCurve(VectorCurve=DataInterface() {propertyValues {
       XCurve {keys [<at 0 16 RCIM_Cubic>,<at 0.8 1 RCIM_Cubic>]}
       YCurve {keys [<at 0 16 RCIM_Cubic>,<at 0.8 0.1249 RCIM_Cubic>]}
       ZCurve {keys [<at 0 16 RCIM_Cubic>,<at 0.8 0.2547 RCIM_Cubic>]}
      }})
    ScaleA=true
    ScaleRGB=true
   )
   &ScaleColor(
    Scale RGB=&VectorFromFloat(Value=&Lerp_Float(
      A=1
      B=30
      Alpha=&FloatFromCurve(
       FloatCurve=DataInterface() {propertyValues {Curve {keys [
           {interpMode=RCIM_Cubic, tangentMode=RCTM_Break, time=0.4, value=0}
           {arriveTangentWeight=0.8501, interpMode=RCIM_Cubic, tangentMode=RCTM_Break, time=1, value=1}
          ]}}}
       CurveIndex=&RandomRangeFloat(Minimum=0, Maximum=1)
      )
     ))
    ScaleA=false
    ScaleRGB=true
   ) {module=ScaleColor001}
   &ScaleSpriteSize(
    Scale Sprite Size Mode=::NewEnumerator0
    Uniform Scale Factor=&FN_ScalabilityFloatModifier(
     Value at DistanceFraction 0:NiagaraFloat=1
     Value at DistanceFraction 1:NiagaraFloat=3
     DistanceRamp:NiagaraDataInterfaceCurve=DataInterface() {propertyValues {Curve {keys [<at 0.1 0>,<at 0.5 1>]}}}
    )
   ) {module=ScaleSpriteSize001}
   &ScaleSpriteSize(
    Uniform Curve Index=&DistanceBetweenPositions(
     A=/Niagara/DynamicInputs/Helpers/SimulationPosition()
     B=&ReturnCameraPosition(Coordinate Space:ENiagaraCoordinateSpace=::Simulation, Camera Query:NiagaraDataInterfaceCamera=DataInterface())
    )
    Uniform Curve Sprite Scale=DataInterface() {propertyValues {Curve {keys [{tangentMode=RCTM_User, time=684.4967, value=0.2}, {tangentMode=RCTM_User, time=1184.4967, value=0.7}]}}}
   ) {module=ScaleSpriteSize002}
  ]
  rendererClasses [&NiagaraSpriteRendererProperties]
  renderers [{rendererClass=&NiagaraSpriteRendererProperties}]
 }
 Sparks002 {
  simTarget=GPUComputeSim
  emitterSpawnScript [/SetVariables(Emitter.NumSource:NiagaraInt32=&RandomRangeInt(Minimum=4, Maximum=7)) {isSetParameters=true, module=SetVariables_#7}]
  emitterUpdateScript [
   &EmitterState(
    Enable Distance Culling=true
    MaxDistance=15000
    Module.bUseMaxDistance=true
    Module.bUseMinDistance=false
    Scalability Mode=::NewEnumerator1
   )
   &SpawnBurst_Instantaneous(
    Spawn Count=&RandomRangeInt(Minimum=500, Maximum=750)
    Spawn Group=1
    Spawn Time=$System.Delay
    Use Spawn Probability=false
    UseLoopCountLimit=false
   ) {module=SpawnBurst_Instantaneous001}
  ]
  particleSpawnScript [
   &InitializeParticle(
    Color=<rgba 1 0.3214 0 1>
    Lifetime Max=1.255
    Lifetime Min=0.65
    Lifetime Mode=::NewEnumerator1
    Material Random:NiagaraFloat=&MaskFloat(Mask Parameter=&SpawnGroup(), Mask=0, Masked Float=0, Passthrough Float=1)
    Sprite Size=&MaskVector2dBySpawnGroup(
     Masked Vector2D=&RandomRangeVector2D(Minimum=<5 5>, Maximum=<6 6>)
     Passthrough Vector2D=&RandomRangeVector2D(Minimum=<2 2>, Maximum=<1.5 4>)
     Spawn Group Mask=0
    )
    Sprite Size Mode=::NewEnumerator4
    UsePositionOffset=true
    Write Material Random=true
   )
   &ShapeLocation(
    Hemisphere Distribution=<1 0.5>
    MaskSpawnGroup=false
    Offset Mode=::NewEnumerator0
    Override Seed=false
    Sphere Radius=&Multiply_Float(A=0, B=$User.FXScale)
    Sphere Surface Distribution=0.75
    UseOwnerScale=true
   )
   &AddVelocity(
    Cone Angle=190
    Cone Axis=<0 0 1>
    Inner Cone Angle=100
    Module.UseConeFalloff=false
    Module.UseVelDistribution=false
    Override Seed=false
    Velocity Mode=::NewEnumerator2
    Velocity Speed=&Multiply_Float(A=&RandomRangeFloat(Minimum=600, Maximum=3500), B=0.75)
   ) {module=AddVelocity001}
   /SetVariables(Particles.Position=&AddVectorToPosition(Position=$Particles.Position, Vector=<0 0 0>)) {isSetParameters=true, module=SetVariables_#8}
  ]
  particleUpdateScript [
   &ParticleState()
   &Drag(Drag=2, Use Linear Drag=true, Use Rotational Drag=false)
   &GravityForce(Gravity=&Multiply_VectorByFloat(
     Vector=<0 0 -980>
     Float=&FloatFromCurve(FloatCurve=DataInterface() {propertyValues {Curve {keys [<at 0 1 RCIM_Cubic>, {arriveTangent=-3.1991, interpMode=RCIM_Cubic, leaveTangent=-3.1991, tangentMode=RCTM_User, time=1, value=0}]}}})
    ))
   &SolveForcesAndVelocity(Clamp Velocity=false, Limit Acceleration=false)
   &ScaleSpriteSize(
    Uniform Curve Scale=3
    Uniform Curve Sprite Scale=DataInterface() {propertyValues {Curve {keys [
        {interpMode=RCIM_Cubic, tangentMode=RCTM_Break, time=0, value=1}
        {arriveTangent=-3.2052, interpMode=RCIM_Cubic, leaveTangent=-2.3701, tangentMode=RCTM_Break, time=0.9, value=0.125}
        <at 1 0 RCIM_Cubic>
       ]}}}
   )
   &ScaleSpriteSize(
    Scale Sprite Size Mode=::NewEnumerator0
    Uniform Scale Factor=&Lerp_Float(A=0.5, B=0.65, Alpha=$System.Cam Adjust)
   ) {module=ScaleSpriteSize002}
   &ScaleSpriteSizeBySpeed(Max Scale Factor=<1 3>, Min Scale Factor=<1 0.5>)
   &ScaleColor(
    Scale Alpha=&FloatFromCurve(FloatCurve=DataInterface() {propertyValues {Curve {keys [
         {interpMode=RCIM_Cubic, tangentMode=RCTM_User, time=0.0048, value=0.438}
         {arriveTangent=0.3021, interpMode=RCIM_Cubic, leaveTangent=0.3021, tangentMode=RCTM_User, time=0.0983, value=1.5996}
         {arriveTangent=-1.0514, interpMode=RCIM_Cubic, leaveTangent=-1.0514, tangentMode=RCTM_User, time=0.3368, value=0.5383}
         {arriveTangent=-0.8117, interpMode=RCIM_Cubic, leaveTangent=-0.8117, tangentMode=RCTM_User, time=0.5811, value=1.092}
         <at 1 0>
        ]}}})
    Scale RGB=&VectorFromCurve(VectorCurve=DataInterface() {propertyValues {
       XCurve {keys [<at 0 16 RCIM_Cubic>,<at 0.8 1 RCIM_Cubic>]}
       YCurve {keys [<at 0 16 RCIM_Cubic>,<at 0.8 0.1249 RCIM_Cubic>]}
       ZCurve {keys [<at 0 16 RCIM_Cubic>,<at 0.8 0.2547 RCIM_Cubic>]}
      }})
    ScaleA=true
    ScaleRGB=true
   )
   &ScaleColor(
    Scale RGB=&VectorFromFloat(Value=&Lerp_Float(
      A=1
      B=30
      Alpha=&FloatFromCurve(
       FloatCurve=DataInterface() {propertyValues {Curve {keys [
           {interpMode=RCIM_Cubic, tangentMode=RCTM_Break, time=0.4, value=0}
           {arriveTangentWeight=0.8501, interpMode=RCIM_Cubic, tangentMode=RCTM_Break, time=1, value=1}
          ]}}}
       CurveIndex=&RandomRangeFloat(Minimum=0, Maximum=1)
      )
     ))
    ScaleA=false
    ScaleRGB=true
   ) {module=ScaleColor001}
   &ScaleColor(
    Scale Alpha=&Lerp_Float(
     A=1
     B=&FloatFromCurve(
      FloatCurve=DataInterface() {propertyValues {Curve {
         defaultValue=1
         keys [
          {arriveTangent=9.0725, interpMode=RCIM_Cubic, time=0, value=0}
          <at 0.25 0.9847 RCIM_Cubic>
          {arriveTangent=-0.0405, interpMode=RCIM_Cubic, leaveTangent=-0.0405, tangentMode=RCTM_User, time=1, value=0}
         ]
        }}}
      Scale Curve=0.1
     )
     Alpha=$System.IsPlayerIndoors
    )
    ScaleA=true
    ScaleRGB=true
   ) {module=ScaleColor002}
   &ScaleSpriteSize(
    Scale Sprite Size Mode=::NewEnumerator0
    Uniform Scale Factor=&FN_ScalabilityFloatModifier(
     Value at DistanceFraction 0:NiagaraFloat=1
     Value at DistanceFraction 1:NiagaraFloat=3
     DistanceRamp:NiagaraDataInterfaceCurve=DataInterface() {propertyValues {Curve {keys [<at 0.1 0>,<at 0.5 1>]}}}
    )
   ) {module=ScaleSpriteSize001}
   &ScaleColor(
    Scale Alpha=&Lerp_Float(
     A=0
     B=1
     Alpha=&FloatFromCurve(
      FloatCurve=DataInterface() {propertyValues {Curve {keys [{arriveTangent=1.6665, interpMode=RCIM_Cubic, leaveTangent=1.6665, tangentMode=RCTM_User, time=0, value=0.05}, <at 1 1 RCIM_Cubic>]}}}
      CurveIndex=$System.Cam Adjust
     )
    )
    ScaleA=true
    ScaleRGB=true
   ) {module=ScaleColor003}
  ]
  rendererClasses [&NiagaraSpriteRendererProperties]
  renderers [{rendererClass=&NiagaraSpriteRendererProperties}]
 }
 Sparks_fromOther {
  emitterSpawnScript [
   /SetVariables(Emitter.emitterInitRand:NiagaraInt32=&RandomRangeInt(Minimum=-1000000, Maximum=1000000)) {isSetParameters=true}
   /SetVariables(Emitter.NumSource:NiagaraInt32=&RandomRangeInt(Minimum=4, Maximum=7)) {isSetParameters=true, module=SetVariables_#7}
  ]
  emitterUpdateScript [
   &EmitterState(
    Enable Distance Culling=true
    MaxDistance=15000
    Module.bUseMaxDistance=true
    Module.bUseMinDistance=false
    Scalability Mode=::NewEnumerator1
   )
   &SpawnParticlesFromOtherEmitter(
    Attribute Reader=DataInterface() {propertyValues {EmitterBinding {bindingMode=Other, emitterName=Sparks001}}}
    Spawn Rate=90
    Use Maximum Age=false
    Use Minimum Age=false
   )
  ]
  particleSpawnScript [
   &InitializeParticle(
    Color=<rgba 1 0.3214 0 1>
    Lifetime Max=0.75
    Lifetime Min=0.5
    Lifetime Mode=::NewEnumerator1
    Sprite Size Max=<6 6>
    Sprite Size Min=<2 2>
    Sprite Size Mode=::NewEnumerator2
    UsePositionOffset=false
    Write Material Random=true
   )
   &SampleParticlesFromOtherEmitter(Material Random Sampling:ENiagara_DirectReadSamplingMode=::NewEnumerator0)
   &KillParticles(Kill Particles=&SetBoolByFloatComparison(A=$Particles.MaterialRandom, B=0.5))
   &AddVelocity(
    Velocity=&RandomVector(Vector Scale=1)
    Velocity Speed Scale=&RandomRangeFloat(Minimum=0, Maximum=500)
   )
  ]
  particleUpdateScript [
   &ParticleState()
   &AccelerationForce(Acceleration=&RandomRangeVector(Minimum=<0 0 -800>, Maximum=<0 0 -1100>))
   &CurlNoiseForce(
    Noise Frequency=120
    Noise Strength=&FloatFromCurve(
     FloatCurve=DataInterface() {propertyValues {Curve {keys [{interpMode=RCIM_Cubic, tangentMode=RCTM_Break, time=0, value=1}, {interpMode=RCIM_Cubic, tangentMode=RCTM_Break, time=1, value=0}]}}}
     Scale Curve=1000
    )
    Pan Noise=false
   )
   &Drag(Drag=3, Use Linear Drag=true, Use Rotational Drag=false)
   &SolveForcesAndVelocity(Clamp Velocity=false, Limit Acceleration=false)
   &ScaleSpriteSize(
    Uniform Curve Scale=2
    Uniform Curve Sprite Scale=DataInterface() {propertyValues {Curve {keys [{interpMode=RCIM_Cubic, tangentMode=RCTM_Break, time=0, value=1}, {interpMode=RCIM_Cubic, tangentMode=RCTM_Break, time=1, value=0.2}]}}}
   )
   &ScaleSpriteSizeBySpeed(Max Scale Factor=<1 3>)
   &ScaleColor(
    Scale Alpha=&FloatFromCurve(FloatCurve=DataInterface() {propertyValues {Curve {keys [
         {interpMode=RCIM_Cubic, tangentMode=RCTM_User, time=0.0048, value=0.438}
         {arriveTangent=0.3021, interpMode=RCIM_Cubic, leaveTangent=0.3021, tangentMode=RCTM_User, time=0.0983, value=1.5996}
         {arriveTangent=-1.0514, interpMode=RCIM_Cubic, leaveTangent=-1.0514, tangentMode=RCTM_User, time=0.3368, value=0.5383}
         {arriveTangent=-0.8117, interpMode=RCIM_Cubic, leaveTangent=-0.8117, tangentMode=RCTM_User, time=0.5811, value=1.092}
         <at 1 0>
        ]}}})
    Scale RGB=&VectorFromCurve(VectorCurve=DataInterface() {propertyValues {
       XCurve {keys [<at 0 16 RCIM_Cubic>,<at 0.8 1 RCIM_Cubic>]}
       YCurve {keys [<at 0 16 RCIM_Cubic>,<at 0.8 0.1249 RCIM_Cubic>]}
       ZCurve {keys [<at 0 16 RCIM_Cubic>,<at 0.8 0.2547 RCIM_Cubic>]}
      }})
    ScaleA=true
    ScaleRGB=true
   )
   &ScaleColor(
    Scale RGB=&VectorFromFloat(Value=&Lerp_Float(
      A=1
      B=30
      Alpha=&FloatFromCurve(
       FloatCurve=DataInterface() {propertyValues {Curve {keys [
           {interpMode=RCIM_Cubic, tangentMode=RCTM_Break, time=0.4, value=0}
           {arriveTangentWeight=0.8501, interpMode=RCIM_Cubic, tangentMode=RCTM_Break, time=1, value=1}
          ]}}}
       CurveIndex=&RandomRangeFloat(Minimum=0, Maximum=1)
      )
     ))
    ScaleA=false
    ScaleRGB=true
   ) {module=ScaleColor001}
   &ScaleSpriteSize(
    Scale Sprite Size Mode=::NewEnumerator0
    Uniform Scale Factor=&FN_ScalabilityFloatModifier(
     Value at DistanceFraction 0:NiagaraFloat=1
     Value at DistanceFraction 1:NiagaraFloat=3
     DistanceRamp:NiagaraDataInterfaceCurve=DataInterface() {propertyValues {Curve {keys [<at 0.1 0>,<at 0.5 1>]}}}
    )
   ) {module=ScaleSpriteSize001}
   &ScaleSpriteSize(
    Uniform Curve Index=&DistanceBetweenPositions(
     A=&SimulationPosition()
     B=&ReturnCameraPosition(Coordinate Space:ENiagaraCoordinateSpace=::Simulation, Camera Query:NiagaraDataInterfaceCamera=DataInterface())
    )
    Uniform Curve Sprite Scale=DataInterface() {propertyValues {Curve {keys [{tangentMode=RCTM_User, time=684.4967, value=0.2}, {tangentMode=RCTM_User, time=1184.4967, value=0.7}]}}}
   ) {module=ScaleSpriteSize002}
  ]
  rendererClasses [&NiagaraSpriteRendererProperties]
  renderers [{rendererClass=&NiagaraSpriteRendererProperties}]
 }
 Sphere_glow_LW {rendererClasses [], renderers []}
}
summary {
 systemName=NS_Jethro_CloudyMuffin_Exp
 emitters [
  {emitterName=CenterFlash, rendererClasses [&NiagaraMeshRendererProperties]}
  {emitterName=Rocks, rendererClasses [&NiagaraMeshRendererProperties, &NiagaraSpriteRendererProperties]}
  {emitterName=SmokeyFingers001, rendererClasses [&NiagaraMeshRendererProperties]}
  {emitterName=EnergyWave, rendererClasses [&NiagaraMeshRendererProperties, &NiagaraLightRendererProperties]}
  {emitterName=FireSmoke, rendererClasses [&NiagaraSpriteRendererProperties]}
  {emitterName=Sparks001, rendererClasses [&NiagaraSpriteRendererProperties]}
  {emitterName=Sparks_fromOther, rendererClasses [&NiagaraSpriteRendererProperties]}
  {emitterName=Combined_Streaks, rendererClasses [&NiagaraSpriteRendererProperties]}
  {emitterName=FN_Mesh001_water, rendererClasses [&NiagaraMeshRendererProperties]}
  {emitterName=Secondary_Gloops, rendererClasses [&NiagaraSpriteRendererProperties]}
  {emitterName=Sparks, rendererClasses [&NiagaraSpriteRendererProperties]}
  {emitterName=Energy, rendererClasses [&NiagaraMeshRendererProperties, &NiagaraSpriteRendererProperties]}
  {emitterName=Sparks002, simTarget=GPUComputeSim, rendererClasses [&NiagaraSpriteRendererProperties]}
  {emitterName=Corona_LW, rendererClasses []}
  {emitterName=Pressure_waveLW, rendererClasses []}
  {emitterName=Sphere_glow_LW, rendererClasses []}
 ]
 userVariables [
  User.FXDuration_Initial {type=NiagaraFloat, defaultValue=0, description=""}
  User.FXScale {type=NiagaraFloat, defaultValue=500, description=""}
  User.Global Color {type=LinearColor, description="", defaultValue {struct=LinearColor, value=<rgba 4 0.5231 0 1>}}
  User.Global Scale {type=NiagaraFloat, defaultValue=1, description=""}
  User.IsInWater {type=NiagaraBool}
 ]
}
