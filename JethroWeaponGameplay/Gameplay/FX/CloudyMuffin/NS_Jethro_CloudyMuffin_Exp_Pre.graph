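# Niagara system for the Jethro CloudyMuffin weapon, apparently its pre-explosion effect, scaled and timed by FXScale and FXDuration_Initial user parameters.
# A center flash and a light-emitting energy wave use spawn groups for orange/cyan variants, alongside converging streak sprites and a delayed shrinking mesh.
# The main explosion emitter switches motion, scale and color by spawn group, driving dissolve-style material parameters with curves and time-of-day scaling.
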NiagaraSystem /JethroWeaponGameplay/Gameplay/FX/CloudyMuffin/NS_Jethro_CloudyMuffin_Exp_Pre
dependencies {
 usedDataInterfaces [
  &NiagaraDataInterfaceCurve
  &NiagaraDataInterfacePlatformSet
  &NiagaraDataInterfaceColorCurve
  &NiagaraDataInterfaceVectorCurve
  &NiagaraDataInterfaceArrayFloat
  &NiagaraDataInterfaceVector2DCurve
 ]
 usedDynamicInputs [&ReturnNormalizedExecIndex]
 usedModules [
  &self:SystemSpawnScript.SystemScriptSource.SystemScriptGraph.NiagaraNodeAssignment_0.SetVariables_#1
  &self:SystemSpawnScript.SystemScriptSource.SystemScriptGraph.NiagaraNodeAssignment_1.SetVariables_#2
  &SystemState
 ]
}
emitters {
 CenterFlash {
  emitterUpdateScript [
   &EmitterState()
   &SpawnBurst_Instantaneous(Spawn Group=1, Use Spawn Probability=false, UseLoopCountLimit=false) {module=SpawnBurst_Instantaneous001}
  ]
  particleSpawnScript [
   &InitializeParticle(
    Color=&MaskLinearColorBySpawnGroup(Masked Linear Color=<rgba 12 2.087 0.6504 1>, Passthrough Linear Color=<rgba 0.0542 0.8803 1 0.7>)
    Lifetime=&MaskFloatBySpawnGroup(
     Masked Float=&RandomRangeFloat(Minimum=0.1, Maximum=0.15)
     Passthrough Float=&RandomRangeFloat(Minimum=0.125, Maximum=0.125)
    )
    Mesh Scale Mode=::NewEnumerator3
    Mesh Uniform Scale=&MaskFloatBySpawnGroup(
     Masked Float=&Multiply_Float(A=0.0016, B=$User.FXScale)
     Passthrough Float=&Multiply_Float(A=0.0005, B=$User.FXScale)
    )
    UsePositionOffset=false
    Write Material Random=true
   )
   &InitialMeshOrientation(Mesh Orientation Mode=::NewEnumerator2, Orientation Vector=<0 0 1>, Use Rotation Vector=false)
   &DynamicMaterialParameters(
    Index 0 Param 1=&UniformRangedFloat(Minimum=0, Maximum=20)
    Module.Param0WriteEnabled=true
    Module.Param1WriteEnabled=false
    Module.Param2WriteEnabled=false
    Module.Param3WriteEnabled=false
   )
   &CameraOffset(Camera Offset Amount=100)
  ]
  particleUpdateScript [
   &ParticleState()
   &SolveForcesAndVelocity(Clamp Velocity=false, Limit Acceleration=false)
   &ScaleMeshSize(Scale Factor=&VectorFromFloat(Value=&FloatFromCurve(
      FloatCurve=DataInterface() {propertyValues {Curve {keys [{arriveTangent=9.901, interpMode=RCIM_Cubic, leaveTangent=9.901, tangentMode=RCTM_User, time=0, value=0.4}, <at 1 6 RCIM_Cubic>]}}}
      Scale Curve=1.5
     )))
   &ScaleColor(
    Scale Alpha=&FloatFromCurve(FloatCurve=DataInterface() {propertyValues {Curve {keys [<at 0.4 1 RCIM_Cubic>,<at 1 0>]}}})
    ScaleA=true
    ScaleRGB=true
   )
  ]
  rendererClasses [&NiagaraMeshRendererProperties, &NiagaraSpriteRendererProperties]
  renderers [
   {rendererClass=&NiagaraMeshRendererProperties, rendererIndex=0}
   {rendererClass=&NiagaraSpriteRendererProperties, rendererIndex=1}
  ]
 }
 Empty003 {
  emitterUpdateScript [
   /SetVariables(Emitter.saturation=1) {isSetParameters=true}
   &EmitterState()
   &SpawnBurst_Instantaneous(Spawn Count=1, Spawn Time=0.38, Use Spawn Probability=false)
  ]
  particleSpawnScript [
   &InitializeParticle(
    Color=<rgba 1 1 1 0.7>
    Lifetime=0.2
    Mesh Scale Mode=::NewEnumerator3
    Mesh Uniform Scale=1.5
    UsePositionOffset=false
    Write Material Random=true
   )
   &InitialMeshOrientation(Use Rotation Vector=false)
  ]
  particleUpdateScript [
   &ParticleState()
   &ScaleMeshSize(Scale Factor=&VectorFromFloat(Value=&FloatFromCurve(FloatCurve=DataInterface() {propertyValues {Curve {keys [
          {arriveTangent=-1.9728, interpMode=RCIM_Cubic, time=0, value=1}
          {arriveTangent=-1.2046, interpMode=RCIM_Cubic, leaveTangent=-1.2046, tangentMode=RCTM_User, time=1, value=0}
         ]}}})))
   &ScaleColor(
    Color Value To Scale=&MultiplyLinearColor(
     Linear Color=$Particles.Initial.Color
     Scale Factor=&MakeVector4FromVectorAndFloat("Vector (RGB)"=&VectorFromFloat(Value=10), "Float (Alpha)"=1)
    )
    Linear Color Curve=DataInterface() {propertyValues {
      AlphaCurve {keys [<at 0 1>,<at 1 0>]}
      BlueCurve {keys [<at 0.0674 1>,<at 0.3516 0.5089>,<at 0.3793 0>]}
      GreenCurve {keys [<at 0.0674 0.4925>,<at 0.3516 0.5045>,<at 0.3793 0.0896>]}
      RedCurve {keys [<at 0.0674 0>,<at 0.3516 0.5085>,<at 0.3793 1>]}
     }}
    Scale Mode=::NewEnumerator2
   )
  ]
  rendererClasses [&NiagaraMeshRendererProperties]
  renderers [{rendererClass=&NiagaraMeshRendererProperties, rendererIndex=0}]
 }
 EnergyWave {
  emitterUpdateScript [
   &EmitterState()
   &SpawnBurst_Instantaneous(Spawn Group=1, Use Spawn Probability=false, UseLoopCountLimit=false) {module=SpawnBurst_Instantaneous001}
   &SpawnBurst_Instantaneous(Spawn Count=1, Spawn Time=$System.Delay, Use Spawn Probability=false) {enabled=false}
  ]
  particleSpawnScript [
   &InitializeParticle(
    Color Mode=::NewEnumerator0
    Lifetime=0.35
    Mesh Scale Mode=::NewEnumerator3
    Mesh Uniform Scale=&MaskFloatBySpawnGroup(
     Masked Float=&Multiply_Float(A=0.011, B=$User.FXScale)
     Passthrough Float=&Multiply_Float(A=0.0072, B=$User.FXScale)
    )
    UsePositionOffset=false
    Write Material Random=true
   )
   &InitialMeshOrientation(Orientation Axis=<0 0 1>, Orientation Vector=<0 0 1>, Rotation=<0 0.25 0>, Use Orientation Vector=true, Use Rotation Vector=true)
   &CameraOffset(Camera Offset Amount=30)
   &ShapeLocation(MaskSpawnGroup=false, Offset Mode=::NewEnumerator0, Override Seed=false, Sphere Radius=1, UseOwnerScale=true)
   /SetVariables(Particles.LightPosOffset:NiagaraPosition=<0 0 50>) {isSetParameters=true}
  ]
  particleUpdateScript [
   &ParticleState()
   &ScaleColor(
    Scale Mode=::NewEnumerator1
    Scale RGBA=&MakeVector4_FromLinearColor(LinearColor=&MaskLinearColorBySpawnGroup(
      Masked Linear Color=&ColorFromCurve(ColorCurve=DataInterface() {propertyValues {
         AlphaCurve {keys [<at 0 1>,<at 0.3576 1>,<at 1 0>]}
         BlueCurve {keys [<at 0 0>,<at 0.1684 0>,<at 0.3654 0>,<at 1 0>]}
         GreenCurve {keys [<at 0 0.4051>,<at 0.1684 0.325>,<at 0.3654 0.2679>,<at 1 0.1355>]}
         RedCurve {keys [<at 0 1>,<at 0.1684 1>,<at 0.3654 1>,<at 1 1>]}
        }})
      Passthrough Linear Color=&ColorFromCurve(ColorCurve=DataInterface() {propertyValues {
         AlphaCurve {keys [{arriveTangent=-2.6592, interpMode=RCIM_Cubic, leaveTangent=-2.6592, tangentMode=RCTM_User, time=0, value=1}, <at 1 0 RCIM_Cubic>]}
         BlueCurve {keys [<at 0 1>,<at 0.1684 1>,<at 0.3654 1>,<at 1 1>]}
         GreenCurve {keys [<at 0 0.5949>,<at 0.1684 0.675>,<at 0.3654 0.7321>,<at 1 0.8645>]}
         RedCurve {keys [<at 0 0>,<at 0.1684 0>,<at 0.3654 0>,<at 1 0>]}
        }})
     ))
    ScaleRGBA=true
   )
   &DynamicMaterialParameters(
    Index 0 Param 1=&FloatFromCurve(FloatCurve=DataInterface() {propertyValues {Curve {keys [{arriveTangent=-2.1015, interpMode=RCIM_Cubic, leaveTangent=-2.1015, tangentMode=RCTM_User, time=0, value=0.5}, <at 1 -0.15>]}}})
    Index 0 Param 2=&FloatFromCurve(FloatCurve=DataInterface() {propertyValues {Curve {keys [{arriveTangent=-0.3464, interpMode=RCIM_Cubic, leaveTangent=-0.3464, tangentMode=RCTM_User, time=0, value=1}, <at 1 0.9>]}}})
    Index 0 Param 3=&FloatFromCurve(FloatCurve=DataInterface() {propertyValues {Curve {keys [
         {interpMode=RCIM_Cubic, tangentMode=RCTM_Break, time=0, value=0}
         {arriveTangent=0.4478, interpMode=RCIM_Cubic, leaveTangent=0.4478, tangentMode=RCTM_User, time=0.15, value=1}
         {interpMode=RCIM_Cubic, tangentMode=RCTM_Break, time=1, value=1.2}
        ]}}})
    Module.Param0WriteEnabled=true
    Module.Param1WriteEnabled=true
    Module.Param2WriteEnabled=true
    Module.Param3WriteEnabled=false
   )
   &ScaleColor(
    Scale RGB=&VectorFromFloat(Value=&FloatFromCurve(
      FloatCurve=DataInterface() {propertyValues {Curve {keys [
          {interpMode=RCIM_Cubic, leaveTangent=-40.039, tangentMode=RCTM_Break, time=0, value=10}
          {arriveTangent=-6.6072, interpMode=RCIM_Cubic, leaveTangent=-6.6072, tangentMode=RCTM_User, time=0.15, value=3}
          {interpMode=RCIM_Cubic, tangentMode=RCTM_Break, time=1, value=1}
         ]}}}
      Scale Curve=10
     ))
    ScaleA=false
    ScaleRGB=true
   ) {module=ScaleColor001}
   &ScaleMeshSize(Scale Factor=&VectorFromFloat(Value=&FloatFromCurve(
      FloatCurve=DataInterface() {propertyValues {Curve {keys [
          {interpMode=RCIM_Cubic, tangentMode=RCTM_Break, time=0, value=0.2}
          {arriveTangent=2.9365, interpMode=RCIM_Cubic, leaveTangent=2.9365, tangentMode=RCTM_User, time=0.1, value=1}
          {interpMode=RCIM_Cubic, tangentMode=RCTM_Break, time=1, value=2}
         ]}}}
      Scale Curve=1.35
     )))
   &FN_Light_Attributes(
    Color=&MaskLinearColorBySpawnGroup(Masked Linear Color=<rgba 1 0.1939 0 1>, Passthrough Linear Color=<rgba 0 0.0806 0.1 1>)
    Create Color Attribute=true
    Create Exponent Attribute=true
    Create Position Attribute=true
    Create Radius Attribute=true
    Position=&Add_Vector(A=Linked() {linkedVariable {name=Particles.Initial.Position, type=NiagaraPosition}}, B=<0 0 25>)
    Radius=&Multiply_Float(A=3, B=$User.FXScale)
    Scale Color=&FloatFromCurve(
     FloatCurve=DataInterface() {propertyValues {Curve {keys [
         {interpMode=RCIM_Cubic, leaveTangentWeight=0.9776, tangentMode=RCTM_Break, time=0.1, value=1}
         {arriveTangent=-2.4571, interpMode=RCIM_Cubic, tangentMode=RCTM_Break, time=0.75, value=0}
        ]}}}
     Scale Curve=10
    )
   )
   /SetVariables(Particles.LightPos:NiagaraPosition=&Add_Vector(
     A=Linked() {linkedVariable {name=Particles.Position, type=NiagaraPosition}}
     B=Linked() {linkedVariable {name=Particles.LightPosOffset, type=NiagaraPosition}}
    )) {isSetParameters=true}
  ]
  rendererClasses [&NiagaraMeshRendererProperties, &NiagaraLightRendererProperties]
  renderers [
   {rendererClass=&NiagaraMeshRendererProperties, rendererIndex=0}
   {rendererClass=&NiagaraLightRendererProperties, rendererIndex=1}
  ]
 }
 Lines {
  emitterUpdateScript [
   &EmitterState()
   &SpawnBurst_Instantaneous(
    Spawn Count=35
    Spawn Time=&Multiply_Float(A=$User.FXDuration_Initial, B=0.6)
    Use Spawn Probability=false
    UseLoopCountLimit=false
   )
  ]
  particleSpawnScript [&InitializeParticle(
    Color Maximum=<rgba 0.6323 0.8628 1 0.1>
    Color Minimum=<rgba 1 0.6678 0.4245 0.05>
    Color Mode=::NewEnumerator2
    Lifetime=0.3
    Sprite Size Max=<10 100>
    Sprite Size Min=<35 200>
    Sprite Size Mode=::NewEnumerator2
    UsePositionOffset=false
    Write Material Random=true
   )]
  particleUpdateScript [
   &ShapeLocation(
    MaskSpawnGroup=false
    Override Seed=false
    Sphere Radius=&FloatFromCurve(FloatCurve=DataInterface() {propertyValues {Curve {keys [<at 0 1>,<at 0.9 0>]}}}, Scale Curve=1000)
    UseOwnerScale=true
   )
   &SolveForcesAndVelocity(Clamp Velocity=false, Limit Acceleration=false)
   &ParticleState()
   &UpdateMeshOrientation(
    Facing Interpolation=false
    Orientation Method=::NewEnumerator2
    Override Quat=false
    TransformFacingOrigin=false
    TransformFacingPosition=false
    TransformUp=false
   )
   &AlignSpriteToMeshOrientation(Mesh Orientation Relative Sprite Alignment Vector=<1 0 0>, SetAlignment=true, SetFacing=true)
   &ScaleColor(
    Scale Alpha=&FloatFromCurve(FloatCurve=DataInterface() {propertyValues {Curve {
        defaultValue=0
        keys [
         {interpMode=RCIM_Cubic, tangentMode=RCTM_User, time=0, value=0}
         {arriveTangentWeight=0.1667, interpMode=RCIM_Cubic, leaveTangentWeight=1.0284, time=0.5, value=1}
         {tangentMode=RCTM_User, time=1, value=0}
        ]
       }}})
    ScaleA=true
    ScaleRGB=true
   )
  ]
  rendererClasses [&NiagaraSpriteRendererProperties]
  renderers [{rendererClass=&NiagaraSpriteRendererProperties}]
 }
 NE_Explosion_Main001 {
  emitterSpawnScript [
   /SetVariables(Emitter.Global Scale=1.5) {isSetParameters=true}
   /SetVariables(Emitter.Dissolve Noise Scale:LinearColor=&MultiplyLinearColor(
     Linear Color=/Niagara/DynamicInputs/TypeConversions/MakeLinearColorFromFloat(Value:NiagaraFloat=&FloatFromCurve(FloatCurve=DataInterface() {propertyValues {Curve {keys [<at 0 2>,<at 1 1>,<at 2 0.5>]}}}, CurveIndex=$Emitter.Global Scale))
     Scale Factor=/Niagara/DynamicInputs/Multiply/Multiply_Vector4_ByFloat(Vector4:Vector4f=<0.0008 0.0008 0.0009 1>, Float:NiagaraFloat=1)
    )) {isSetParameters=true}
   /SetVariables(Emitter.Dissolve Scale Adjsutment=500) {enabled=false, isSetParameters=true}
  ]
  emitterUpdateScript [
   &EmitterState()
   &SpawnBurst_Instantaneous(Spawn Count=8, Use Spawn Probability=false, UseLoopCountLimit=false) {module=SpawnBurst_Instantaneous001}
   &SpawnBurst_Instantaneous(Spawn Group=1, Use Spawn Probability=false, UseLoopCountLimit=false)
  ]
  particleSpawnScript [
   &InitializeParticle(
    Color=&self:color_sel_4(
     Linear Color:LinearColor=<rgba 1 1 1 0.35>
     Linear Color001:LinearColor=<rgba 1 1 1 1>
     Linear Color002:LinearColor=<rgba 1 1 1 0.5>
     Linear Color003:LinearColor=<rgba 1 1 1 0.5>
     int32:NiagaraInt32=&SpawnGroup()
    )
    Lifetime=&self:float_sel_4(
     float:NiagaraFloat=$User.FXDuration_Initial
     float001:NiagaraFloat=0.25
     float002:NiagaraFloat=&SelectFloatFromArray(
      Array Sampling Mode=::NewEnumerator0
      Direct Array Index=?
      Float Selection Array=DataInterface() {propertyValues {FloatData [1.5,1.95]}}
      Random Seed=0
     )
     float003:NiagaraFloat=2.1
     int32:NiagaraInt32=&SpawnGroup()
    )
    Mesh Scale=&Multiply_VectorByFloat(
     Vector=&self:vector_sel_4(
      Default:Vector3f=&Multiply_VectorByFloat(Vector=&RandomRangeVector(Minimum=<3 2.8 2.8>, Maximum=<3 3.5 3.5>), Float=0.5)
      Vector:Vector3f=<1.5 1.5 1.5>
      Vector001:Vector3f=&RandomRangeVector(Minimum=<3 2.8 2.8>, Maximum=<3 3.5 3.5>)
      Vector002:Vector3f=<5 5 5>
      int32:NiagaraInt32=&SpawnGroup()
     )
     Float=&self:float_sel_4(
      float:NiagaraFloat=&FloatFromCurve(
       FloatCurve=DataInterface() {propertyValues {Curve {keys [<at 0 0.9 RCIM_Constant>,<at 0.5 1 RCIM_Constant>,<at 1 1>]}}}
       CurveIndex=&RandomRangeFloat(Minimum=0, Maximum=1)
      )
      float001:NiagaraFloat=1
      float002:NiagaraFloat=&FloatFromCurve(
       FloatCurve=DataInterface() {propertyValues {Curve {keys [<at 0 0.9>,<at 1 1 RCIM_Cubic>]}}}
       CurveIndex=&RandomRangeFloat(Minimum=0, Maximum=1)
       Scale Curve=0.8
      )
      float003:NiagaraFloat=1
      int32:NiagaraInt32=&SpawnGroup()
     )
    )
    Mesh Scale Mode=::NewEnumerator4
    UsePositionOffset=false
    Write Material Random=true
   )
   &ShapeLocation(
    Apply To Particle Position=::NewEnumerator1
    MaskSpawnGroup=false
    Override Seed=false
    Sphere Radius=&self:float_sel_4(
     float:NiagaraFloat=&Multiply_Float(A=100, B=$Emitter.Global Scale)
     float001:NiagaraFloat=0
     float002:NiagaraFloat=&Multiply_Float(A=100, B=$Emitter.Global Scale)
     float003:NiagaraFloat=0
     int32:NiagaraInt32=&SpawnGroup()
    )
    Sphere Surface Distribution=0.25
    UseOwnerScale=true
   )
   &ShapeLocation(
    Apply To Particle Position=::NewEnumerator1
    Cylinder Height=0
    Cylinder Radius=50
    MaskSpawnGroup=false
    Override Seed=false
    Shape Primitive=::NewEnumerator1
    Surface Only=false
    UseOwnerScale=true
   ) {module=ShapeLocation001}
   /SetVariables(Particles.Position=&self:Pos_Sel_4(
     Position:NiagaraPosition=$Particles.ShapeLocation001.ShapePosition
     Position001:NiagaraPosition=/Niagara/DynamicInputs/Helpers/SimulationPosition()
     Position002:NiagaraPosition=$Particles.ShapeLocation.ShapePosition
     Position003:NiagaraPosition=&SimulationPosition()
     int32:NiagaraInt32=&SpawnGroup()
    )) {isSetParameters=true}
   &AddVelocity(
    Module.Use Falloff Distance=false
    Origin Offset=&Multiply_Vector(
     A=&self:Pos_Sel_4(
      Position:NiagaraPosition=<0 0 0>
      Position001:NiagaraPosition=<0 0 -10>
      Position002:NiagaraPosition=<0 0 0>
      Position003:NiagaraPosition=<0 0 -10>
      int32:NiagaraInt32=&SpawnGroup()
     )
     B=&VectorFromFloat(Value=$Emitter.Global Scale)
    )
    Override Seed=false
    Velocity Mode=::NewEnumerator1
    Velocity Speed=&self:float_sel_4(
     float:NiagaraFloat=&Multiply_Float(
      A=&Multiply_Float(
       A=&RandomRangeFloat(Minimum=900, Maximum=1100)
       B=&RandomRangeFloat(Minimum=0.75, Maximum=2.5)
      )
      B=$Emitter.Global Scale
     )
     float001:NiagaraFloat=0
     float002:NiagaraFloat=&Multiply_Float(A=&RandomRangeFloat(Minimum=2000, Maximum=750), B=$Emitter.Global Scale)
     float003:NiagaraFloat=0
     int32:NiagaraInt32=&SpawnGroup()
    )
   )
   &AddVelocity(
    Velocity=<0 0 1000>
    Velocity Speed Scale=&self:float_sel_4(
     float:NiagaraFloat=&RandomRangeFloat(Minimum=0, Maximum=3)
     float001:NiagaraFloat=0
     float002:NiagaraFloat=0
     float003:NiagaraFloat=0
     int32:NiagaraInt32=&SpawnGroup()
    )
   ) {module=AddVelocity001}
   &CameraOffset(Camera Offset Amount=&Multiply_Float(
     A=&self:float_sel_4(
      float:NiagaraFloat=40
      float001:NiagaraFloat=70
      float002:NiagaraFloat=40
      float003:NiagaraFloat=70
      int32:NiagaraInt32=&SpawnGroup()
     )
     B=$Emitter.Global Scale
    ))
   /SetVariables(Particles.Scale=&Multiply_VectorByFloat(Vector=$Particles.Scale, Float=$Emitter.Global Scale)) {isSetParameters=true}
   /SetVariables(Particles.VisibilityTag=&self:int_sel_4(
     int32:NiagaraInt32=2
     int32001:NiagaraInt32=1
     int32002:NiagaraInt32=0
     int32003:NiagaraInt32=0
     int32004:NiagaraInt32=&SpawnGroup()
    )) {isSetParameters=true}
  ]
  particleUpdateScript [
   &ParticleState()
   &ScaleMeshSize(Scale Factor=&self:Pos_Sel_4(
     Position:NiagaraPosition=&Multiply_VectorByFloat(
      Vector=<1.5 1 1>
      Float=&FloatFromCurve(FloatCurve=DataInterface() {propertyValues {Curve {keys [<at 0.7 1 RCIM_Cubic>, {arriveTangent=-2.0053, interpMode=RCIM_Cubic, leaveTangent=-2.0053, tangentMode=RCTM_User, time=1, value=0.75}]}}})
     )
     Position001:NiagaraPosition=<1.5 1 1>
     Position002:NiagaraPosition=<1.5 1 1>
     Position003:NiagaraPosition=<1.5 1 1>
     int32:NiagaraInt32=&SpawnGroup()
    )) {module=ScaleMeshSize002}
   &GravityForce(Gravity=&Multiply_VectorByFloat(
     Vector=&self:vector_sel_4(
      Default:Vector3f=&RandomRangeVector(Minimum=<0 0 500>, Maximum=<1 1 1500>)
      Vector:Vector3f=<0 0 0>
      Vector001:Vector3f=&RandomRangeVector(Minimum=<0 0 500>, Maximum=<1 1 1500>)
      Vector002:Vector3f=<0 0 1600>
      int32:NiagaraInt32=&SpawnGroup()
     )
     Float=$Emitter.Global Scale
    ))
   &Drag(Drag=&RandomRangeFloat(Minimum=9, Maximum=10), Use Linear Drag=true, Use Rotational Drag=false)
   &DynamicMaterialParameters(
    Index 0 Param 1=&FloatFromCurve(
     FloatCurve=DataInterface() {propertyValues {Curve {keys [{arriveTangent=0.9809, interpMode=RCIM_Cubic, leaveTangent=0.9809, tangentMode=RCTM_User, time=0.2, value=0.8}, <at 1 0.5>]}}}
     Scale Curve=&Multiply_Float(A=-150, B=$Emitter.Global Scale)
    )
    Index 0 Param 2=&FloatFromCurve(FloatCurve=DataInterface() {propertyValues {Curve {keys [
         {tangentMode=RCTM_User, time=0, value=1}
         {arriveTangent=-0.0405, interpMode=RCIM_Cubic, leaveTangent=-0.0405, tangentMode=RCTM_User, time=0.2, value=1}
         {arriveTangent=-0.0002, interpMode=RCIM_Cubic, leaveTangent=-0.0002, tangentMode=RCTM_User, time=0.9, value=0.6}
        ]}}})
    Index 0 Param 3=&FloatFromCurve(
     FloatCurve=DataInterface() {propertyValues {Curve {keys [{arriveTangent=1.4725, time=0, value=0}, {tangentMode=RCTM_User, time=1, value=1}]}}}
     Scale Curve=&Multiply_Float(
      A=&self:float_sel_4(
       float:NiagaraFloat=&Multiply_Float(A=&RandomRangeFloat(Minimum=250, Maximum=300), B=1)
       float001:NiagaraFloat=250
       float002:NiagaraFloat=&RandomRangeFloat(Minimum=250, Maximum=300)
       float003:NiagaraFloat=250
       int32:NiagaraInt32=&SpawnGroup()
      )
      B=$Emitter.Global Scale
     )
    )
    Index 0 Param 4=&self:float_sel_4(
     float:NiagaraFloat=&FloatFromCurve(
      FloatCurve=DataInterface() {propertyValues {Curve {keys [
          {arriveTangent=0.1175, interpMode=RCIM_Cubic, leaveTangent=1.7315, tangentMode=RCTM_Break, time=0.9, value=-3}
          {arriveTangent=35.5342, interpMode=RCIM_Cubic, leaveTangent=14.4841, tangentMode=RCTM_Break, time=1, value=9}
         ]}}}
      Scale Curve=0.2
     )
     float001:NiagaraFloat=&FloatFromCurve(
      FloatCurve=DataInterface() {propertyValues {Curve {keys [
          {arriveTangent=0.1175, interpMode=RCIM_Cubic, leaveTangent=0.6205, tangentMode=RCTM_Break, time=0.22, value=-2}
          {arriveTangent=6.3488, interpMode=RCIM_Cubic, leaveTangent=6.3488, tangentMode=RCTM_User, time=0.5987, value=-0.8048}
          {arriveTangent=42.4007, interpMode=RCIM_Cubic, leaveTangent=14.4841, tangentMode=RCTM_Break, time=1, value=7.3}
         ]}}}
      Scale Curve=0.2
     )
     float002:NiagaraFloat=&FloatFromCurve(
      FloatCurve=DataInterface() {propertyValues {Curve {keys [
          {arriveTangent=0.1175, interpMode=RCIM_Cubic, leaveTangent=1.7315, tangentMode=RCTM_Break, time=0.1, value=-3}
          {arriveTangent=35.5342, interpMode=RCIM_Cubic, leaveTangent=14.4841, tangentMode=RCTM_Break, time=1, value=9}
         ]}}}
      Scale Curve=0.2
     )
     float003:NiagaraFloat=&FloatFromCurve(
      FloatCurve=DataInterface() {propertyValues {Curve {keys [
          {arriveTangent=0.1175, interpMode=RCIM_Cubic, leaveTangent=0.6205, tangentMode=RCTM_Break, time=0.22, value=-2}
          {arriveTangent=6.3488, interpMode=RCIM_Cubic, leaveTangent=6.3488, tangentMode=RCTM_User, time=0.5987, value=-0.8048}
          {arriveTangent=42.4007, interpMode=RCIM_Cubic, leaveTangent=14.4841, tangentMode=RCTM_Break, time=1, value=7.3}
         ]}}}
      Scale Curve=0.2
     )
     int32:NiagaraInt32=&SpawnGroup()
    )
    Index 1 Param 1=&self:float_sel_4(
     float:NiagaraFloat=&FloatFromCurve(
      FloatCurve=DataInterface() {propertyValues {Curve {keys [
          {arriveTangent=1.457, interpMode=RCIM_Cubic, leaveTangent=1.457, tangentMode=RCTM_User, time=0, value=1}
          {arriveTangent=85.7775, interpMode=RCIM_Cubic, leaveTangent=85.7775, tangentMode=RCTM_User, time=0.08, value=10}
         ]}}}
      Scale Curve=&RandomRangeFloat(Minimum=111, Maximum=111)
     )
     float001:NiagaraFloat=&FloatFromCurve(
      FloatCurve=DataInterface() {propertyValues {Curve {
         defaultValue=0.01
         keys [
          {arriveTangent=0.5348, interpMode=RCIM_Cubic, leaveTangent=0.5348, leaveTangentWeight=0.2333, tangentMode=RCTM_User, time=0, value=0.01}
          {
           arriveTangent=1.457
           arriveTangentWeight=0.0133
           interpMode=RCIM_Cubic
           leaveTangent=1.457
           leaveTangentWeight=0.22
           tangentMode=RCTM_User
           time=0.04
           value=0.05
          }
          {
           arriveTangent=85.7775
           arriveTangentWeight=0.0133
           interpMode=RCIM_Cubic
           leaveTangent=85.7775
           leaveTangentWeight=0.2067
           tangentMode=RCTM_User
           time=0.35
           value=10
          }
         ]
        }}}
      Scale Curve=10
     )
     float002:NiagaraFloat=&FloatFromCurve(
      FloatCurve=DataInterface() {propertyValues {Curve {keys [
          {arriveTangent=0.5348, interpMode=RCIM_Cubic, leaveTangent=0.5348, tangentMode=RCTM_User, time=0, value=0.01}
          {arriveTangent=1.457, interpMode=RCIM_Cubic, leaveTangent=1.457, tangentMode=RCTM_User, time=0.04, value=0.05}
          {arriveTangent=85.7775, interpMode=RCIM_Cubic, leaveTangent=85.7775, tangentMode=RCTM_User, time=0.35, value=10}
         ]}}}
      Scale Curve=0
     )
     float003:NiagaraFloat=&FloatFromCurve(
      FloatCurve=DataInterface() {propertyValues {Curve {keys [
          {arriveTangent=0.5348, interpMode=RCIM_Cubic, leaveTangent=0.5348, tangentMode=RCTM_User, time=0, value=0.01}
          {arriveTangent=1.457, interpMode=RCIM_Cubic, leaveTangent=1.457, tangentMode=RCTM_User, time=0.04, value=0.05}
          {arriveTangent=85.7775, interpMode=RCIM_Cubic, leaveTangent=85.7775, tangentMode=RCTM_User, time=0.08, value=10}
         ]}}}
      Scale Curve=20
     )
     int32:NiagaraInt32=&SpawnGroup()
    )
    Index 1 Param 2=&Divide_Float(
     A=&self:float_sel_4(
      float:NiagaraFloat=&RandomRangeFloat(Minimum=0.012, Maximum=0.014)
      float001:NiagaraFloat=0.014
      float002:NiagaraFloat=&RandomRangeFloat(Minimum=0.012, Maximum=0.014)
      float003:NiagaraFloat=0.014
      int32:NiagaraInt32=&SpawnGroup()
     )
     B=$Emitter.Global Scale
    )
    Index 1 Param 3=&FloatFromCurve(
     FloatCurve=DataInterface() {propertyValues {Curve {keys [{interpMode=RCIM_Cubic, tangentMode=RCTM_User, time=0.4, value=1}, <at 0.8 1>]}}}
     Scale Curve=1.5
    )
    Index 1 Param 4=&FloatFromCurve(FloatCurve=DataInterface() {propertyValues {Curve {keys [{interpMode=RCIM_Cubic, tangentMode=RCTM_User, time=0.05, value=1}, {interpMode=RCIM_Cubic, tangentMode=RCTM_User, time=1, value=6}]}}})
    Index 2 Param 1=&FloatFromCurve(
     FloatCurve=DataInterface() {propertyValues {Curve {keys [
         {arriveTangent=0.9553, interpMode=RCIM_Cubic, leaveTangent=0.9553, tangentMode=RCTM_User, time=0, value=0}
         {arriveTangent=1.5886, interpMode=RCIM_Cubic, leaveTangent=1.5886, tangentMode=RCTM_User, time=1, value=1.5}
        ]}}}
     Scale Curve=1.2
    )
    Index 2 Param 2=&FloatFromCurve(
     FloatCurve=DataInterface() {propertyValues {Curve {keys [
         {interpMode=RCIM_Cubic, tangentMode=RCTM_User, time=0.3, value=0}
         {arriveTangent=2.4333, interpMode=RCIM_Cubic, leaveTangent=2.4333, tangentMode=RCTM_User, time=1, value=1}
        ]}}}
     Scale Curve=&Multiply_Float(A=0.02, B=$Emitter.Global Scale)
    )
    Index 2 Param 3=&self:float_sel_4(
     float:NiagaraFloat=&FloatFromCurve(
      FloatCurve=DataInterface() {propertyValues {Curve {keys [
          {arriveTangent=-3.1996, interpMode=RCIM_Cubic, leaveTangent=-3.1996, tangentMode=RCTM_User, time=0, value=1}
          {interpMode=RCIM_Cubic, leaveTangent=-0.7527, time=1, value=0}
         ]}}}
      Scale Curve=20
     )
     float001:NiagaraFloat=&FloatFromCurve(
      FloatCurve=DataInterface() {propertyValues {Curve {keys [
          {arriveTangent=-17.9351, interpMode=RCIM_Cubic, leaveTangent=-17.9351, leaveTangentWeight=0.3333, tangentMode=RCTM_User, time=0, value=1}
          {
           arriveTangent=-0.7527
           arriveTangentWeight=0.0333
           interpMode=RCIM_Cubic
           leaveTangent=-0.7527
           leaveTangentWeight=0.3
           tangentMode=RCTM_User
           time=0.1
           value=0
          }
         ]}}}
      Scale Curve=100
     )
     float002:NiagaraFloat=&FloatFromCurve(
      FloatCurve=DataInterface() {propertyValues {Curve {keys [
          {arriveTangent=-12.7706, interpMode=RCIM_Cubic, leaveTangent=-12.7706, tangentMode=RCTM_User, time=0, value=1}
          {arriveTangent=-0.7527, interpMode=RCIM_Cubic, leaveTangent=-0.7527, tangentMode=RCTM_User, time=0.2, value=0}
         ]}}}
      Scale Curve=5000
     )
     float003:NiagaraFloat=&FloatFromCurve(
      FloatCurve=DataInterface() {propertyValues {Curve {keys [
          {arriveTangent=-17.9351, interpMode=RCIM_Cubic, leaveTangent=-17.9351, tangentMode=RCTM_User, time=0, value=1}
          {arriveTangent=-0.7527, interpMode=RCIM_Cubic, leaveTangent=-0.7527, tangentMode=RCTM_User, time=0.1, value=0}
         ]}}}
      Scale Curve=100
     )
     int32:NiagaraInt32=&SpawnGroup()
    )
    Index 2 Param 4=$Emitter.Global Scale
    Index 3 Param 1=&FloatFromCurve(
     FloatCurve=DataInterface() {propertyValues {Curve {keys [
         {arriveTangent=2.2353, interpMode=RCIM_Cubic, leaveTangent=2.2353, tangentMode=RCTM_User, time=0, value=0.2}
         {arriveTangent=1.4416, interpMode=RCIM_Cubic, leaveTangent=1.4416, tangentMode=RCTM_User, time=0.2, value=0.5}
         {interpMode=RCIM_Cubic, tangentMode=RCTM_User, time=0.8, value=1.25}
        ]}}}
     Scale Curve=&Multiply_Float(A=-2.5, B=$Emitter.Global Scale)
    )
    Index 3 Param 2=&FloatFromCurve(
     FloatCurve=DataInterface() {propertyValues {Curve {keys [
         {arriveTangent=5.1563, interpMode=RCIM_Cubic, leaveTangent=5.1563, tangentMode=RCTM_User, time=0, value=0}
         {arriveTangent=0.1309, interpMode=RCIM_Cubic, leaveTangent=0.1309, tangentMode=RCTM_User, time=0.1699, value=0.4}
         {arriveTangent=0.8545, interpMode=RCIM_Cubic, leaveTangent=0.8545, tangentMode=RCTM_User, time=1, value=1}
        ]}}}
     Scale Curve=&Multiply_Float(A=-0.4, B=$Emitter.Global Scale)
    )
    Index 3 Param 3=&self:float_sel_4(
     float:NiagaraFloat=&FloatFromCurve(
      FloatCurve=DataInterface() {propertyValues {Curve {keys [<at 0.0014 200 RCIM_Cubic>, <at 0.1 200 RCIM_Cubic>, {interpMode=RCIM_Cubic, leaveTangent=-5.9757, time=0.15, value=200}]}}}
      Scale Curve=&FNTimeOfDayScalar(Night Scale=0.6, Morning Scale=0.6, Day Scale=100000, Evening Scale=0.6)
     )
     float001:NiagaraFloat=&FloatFromCurve(
      FloatCurve=DataInterface() {propertyValues {Curve {keys [
          {interpMode=RCIM_Cubic, leaveTangentWeight=0.0495, tangentMode=RCTM_User, time=0.0014, value=200}
          {arriveTangent=-5.9757, arriveTangentWeight=0.101, interpMode=RCIM_Cubic, leaveTangent=-5.9757, tangentMode=RCTM_User, time=1, value=0}
         ]}}}
      Scale Curve=&FNTimeOfDayScalar(Night Scale=0.6, Morning Scale=0.6, Evening Scale=0.6)
     )
     float002:NiagaraFloat=&FloatFromCurve(
      FloatCurve=DataInterface() {propertyValues {Curve {keys [<at 0.0014 200 RCIM_Cubic>, <at 0.1 200 RCIM_Cubic>, {interpMode=RCIM_Cubic, leaveTangent=-5.9757, time=0.15, value=200}]}}}
      Scale Curve=&FNTimeOfDayScalar(Night Scale=0.6, Morning Scale=0.6, Day Scale=0, Evening Scale=0.6)
     )
     float003:NiagaraFloat=&FloatFromCurve(
      FloatCurve=DataInterface() {propertyValues {Curve {
         defaultValue=200
         keys [
          {interpMode=RCIM_Cubic, leaveTangentWeight=0.3329, tangentMode=RCTM_User, time=0.0014, value=200}
          {arriveTangent=-5.9757, interpMode=RCIM_Cubic, leaveTangent=-5.9757, tangentMode=RCTM_User, time=1, value=0}
         ]
        }}}
      Scale Curve=&FNTimeOfDayScalar(Night Scale=0.6, Morning Scale=0.6, Evening Scale=0.6)
     )
     int32:NiagaraInt32=&SpawnGroup()
    )
    Index 3 Param 4=&FloatFromCurve(
     FloatCurve=DataInterface() {propertyValues {Curve {keys [{tangentMode=RCTM_User, time=0, value=0}, {tangentMode=RCTM_User, time=1, value=1}]}}}
     Scale Curve=&Multiply_Float(A=-100, B=$Emitter.Global Scale)
    )
    Module.Index0Param0Write=true
    Module.Index0Param1Write=true
    Module.Index0Param2Write=true
    Module.Index0Param3Write=true
    Module.Index1Param0Write=true
    Module.Index1Param1Write=true
    Module.Index1Param2Write=true
    Module.Index1Param3Write=true
    Module.Index2Param0Write=true
    Module.Index2Param1Write=true
    Module.Index2Param2Write=true
    Module.Index2Param3Write=true
    Module.Index3Param0Write=true
    Module.Index3Param1Write=true
    Module.Index3Param2Write=true
    Module.Index3Param3Write=true
    Write Parameter Index 1=true
    Write Parameter Index 2=true
    Write Parameter Index 3=true
   ) {module=DynamicMaterialParameters001}
   &ScaleColor(
    Scale Mode=::NewEnumerator1
    Scale RGBA=&MakeVector4_FromLinearColor(LinearColor=&self:color_sel_4(
      Linear Color:LinearColor=&LerpLinearColors(
       StartColor=<rgba 1 1 1 1>
       EndColor=&ColorFromCurve(ColorCurve=DataInterface() {propertyValues {
          AlphaCurve {keys [<at 0 1>,<at 0.9236 1>,<at 1 1>]}
          BlueCurve {keys [<at 0.7 1>,<at 0.9 0>]}
          GreenCurve {keys [<at 0.7 0.4925>,<at 0.9 0.0896>]}
          RedCurve {keys [<at 0.7 0>,<at 0.9 1>]}
         }})
       LerpFactor=0.7
      )
      Linear Color001:LinearColor=&ColorFromCurve(ColorCurve=DataInterface() {propertyValues {
         AlphaCurve {keys [<at 0 1>,<at 0.0732 1>,<at 0.9894 0>]}
         BlueCurve {keys [<at 0.1888 0.6042>,<at 0.4977 0.4497>]}
         GreenCurve {keys [<at 0.1888 0.6042>,<at 0.4977 0.4497>]}
         RedCurve {keys [<at 0.1888 0.6042>,<at 0.4977 0.4497>]}
        }})
      Linear Color002:LinearColor=&ColorFromCurve(ColorCurve=DataInterface() {propertyValues {
         AlphaCurve {keys [<at 0 1>,<at 0.8027 1>,<at 0.9894 0>]}
         BlueCurve {keys [<at 0.1888 0.6042>,<at 0.4977 0.4497>]}
         GreenCurve {keys [<at 0.1888 0.6042>,<at 0.4977 0.4497>]}
         RedCurve {keys [<at 0.1888 0.6042>,<at 0.4977 0.4497>]}
        }})
      Linear Color003:LinearColor=&ColorFromCurve(ColorCurve=DataInterface() {propertyValues {
         AlphaCurve {keys [<at 0 1>,<at 0.0393 1>,<at 0.1071 0>]}
         BlueCurve {keys [<at 0.1888 0.6042>,<at 0.4977 0.4497>]}
         GreenCurve {keys [<at 0.1888 0.6042>,<at 0.4977 0.4497>]}
         RedCurve {keys [<at 0.1888 0.6042>,<at 0.4977 0.4497>]}
        }})
      int32:NiagaraInt32=&SpawnGroup()
     ))
    ScaleRGBA=true
   )
   &CurlNoiseForce(
    Noise Frequency=&Divide_Float(A=1, B=$Emitter.Global Scale)
    Noise Quality / Cost=::NewEnumerator0
    Noise Strength=&Multiply_Float(
     A=&self:float_sel_4(
      float:NiagaraFloat=&FloatFromCurve(
       FloatCurve=DataInterface() {propertyValues {Curve {keys [
           {arriveTangent=3.7261, interpMode=RCIM_Cubic, leaveTangent=3.7261, tangentMode=RCTM_User, time=0.05, value=0}
           {interpMode=RCIM_Cubic, tangentMode=RCTM_User, time=0.5, value=1}
           <at 0.8 0>
          ]}}}
       Scale Curve=1000
      )
      float001:NiagaraFloat=0
      float002:NiagaraFloat=&FloatFromCurve(
       FloatCurve=DataInterface() {propertyValues {Curve {keys [
           {arriveTangent=3.7261, interpMode=RCIM_Cubic, leaveTangent=3.7261, tangentMode=RCTM_User, time=0.05, value=0}
           {interpMode=RCIM_Cubic, tangentMode=RCTM_User, time=0.5, value=1}
           <at 0.8 0>
          ]}}}
       Scale Curve=1000
      )
      float003:NiagaraFloat=&FloatFromCurve(
       FloatCurve=DataInterface() {propertyValues {Curve {keys [
           {arriveTangent=3.7261, interpMode=RCIM_Cubic, leaveTangent=3.7261, tangentMode=RCTM_User, time=0.05, value=0}
           {interpMode=RCIM_Cubic, tangentMode=RCTM_User, time=0.5, value=1}
           <at 0.8 0>
          ]}}}
       Scale Curve=1000
      )
      int32:NiagaraInt32=&SpawnGroup()
     )
     B=$Emitter.Global Scale
    )
    Pan Noise=true
    Pan Noise Field=&RandomRangeVector(Minimum=<0 0 0>, Maximum=<1 1 1>)
   )
   &SolveForcesAndVelocity(Clamp Velocity=false, Limit Acceleration=false)
   &ScaleMeshSize(Scale Factor=&VectorFromCurve(
     VectorCurve=DataInterface() {propertyValues {
       XCurve {keys [
         {arriveTangent=1.1946, interpMode=RCIM_Cubic, leaveTangent=1.1946, tangentMode=RCTM_User, time=0, value=0.35}
         {arriveTangent=0.581, interpMode=RCIM_Cubic, leaveTangent=0.581, tangentMode=RCTM_User, time=0.2, value=0.8875}
         {arriveTangent=0.0837, interpMode=RCIM_Cubic, leaveTangent=0.0837, tangentMode=RCTM_User, time=1, value=1.1}
        ]}
       YCurve {keys [
         {arriveTangent=1.1946, interpMode=RCIM_Cubic, leaveTangent=1.1946, tangentMode=RCTM_User, time=0, value=0.35}
         {arriveTangent=0.581, interpMode=RCIM_Cubic, leaveTangent=0.581, tangentMode=RCTM_User, time=0.2, value=0.8875}
         {arriveTangent=0.0837, interpMode=RCIM_Cubic, leaveTangent=0.0837, tangentMode=RCTM_User, time=1, value=1.1}
        ]}
       ZCurve {keys [
         {arriveTangent=0.9745, interpMode=RCIM_Cubic, leaveTangent=0.9745, tangentMode=RCTM_User, time=0, value=0.35}
         {arriveTangent=0.8594, interpMode=RCIM_Cubic, leaveTangent=0.8594, tangentMode=RCTM_User, time=0.2, value=0.8875}
         {arriveTangent=0.0837, interpMode=RCIM_Cubic, leaveTangent=0.0837, tangentMode=RCTM_User, time=1, value=1.1}
        ]}
      }}
     Scale Curve=&Multiply_VectorByFloat(
      Vector=<1.1 1.1 1.1>
      Float=&FloatFromCurve(
       FloatCurve=DataInterface() {propertyValues {Curve {keys [<at 0.5 0.8 RCIM_Constant>,<at 0.7 0.9 RCIM_Constant>]}}}
       CurveIndex=&RandomRangeFloat(Minimum=0.4, Maximum=1)
       Scale Curve=&self:float_sel_4(
        float:NiagaraFloat=1
        float001:NiagaraFloat=&FloatFromCurve(
         FloatCurve=DataInterface() {propertyValues {Curve {keys [
             {arriveTangent=0.8977, interpMode=RCIM_Cubic, leaveTangent=0.8977, tangentMode=RCTM_User, time=0, value=0.6}
             {arriveTangent=0.4659, interpMode=RCIM_Cubic, leaveTangent=0.4659, tangentMode=RCTM_User, time=0.7, value=1}
            ]}}}
         Scale Curve=0.88
        )
        float002:NiagaraFloat=1
        float003:NiagaraFloat=&FloatFromCurve(
         FloatCurve=DataInterface() {propertyValues {Curve {keys [
             {arriveTangent=0.8977, interpMode=RCIM_Cubic, leaveTangent=0.8977, tangentMode=RCTM_User, time=0, value=0.6}
             {arriveTangent=0.4659, interpMode=RCIM_Cubic, leaveTangent=0.4659, tangentMode=RCTM_User, time=0.7, value=1}
            ]}}}
         Scale Curve=0.88
        )
        int32:NiagaraInt32=&SpawnGroup()
       )
      )
     )
    )) {enabled=false}
   &ScaleMeshSize(Scale Factor=&VectorFromFloat(Value=&self:float_sel_4(
      float:NiagaraFloat=&FloatFromCurve(FloatCurve=DataInterface() {propertyValues {Curve {
          defaultValue=1
          keys [
           {arriveTangent=1.1946, interpMode=RCIM_Cubic, leaveTangent=1.1946, tangentMode=RCTM_User, time=0, value=0.35}
           {arriveTangent=0.581, arriveTangentWeight=0.0667, interpMode=RCIM_Cubic, leaveTangent=0.581, tangentMode=RCTM_User, time=0.2, value=0.8875}
           {arriveTangent=0.0837, arriveTangentWeight=0.2667, interpMode=RCIM_Cubic, leaveTangent=0.0837, tangentMode=RCTM_User, time=1, value=1.1}
          ]
         }}})
      float001:NiagaraFloat=&FloatFromCurve(FloatCurve=DataInterface() {propertyValues {Curve {
          defaultValue=1
          keys [
           {arriveTangent=1.1946, interpMode=RCIM_Cubic, leaveTangent=1.1946, tangentMode=RCTM_User, time=0, value=0.35}
           {arriveTangent=0.581, arriveTangentWeight=0.0667, interpMode=RCIM_Cubic, leaveTangent=0.581, tangentMode=RCTM_User, time=0.2, value=0.8875}
           {arriveTangent=0.0837, arriveTangentWeight=0.2667, interpMode=RCIM_Cubic, leaveTangent=0.0837, tangentMode=RCTM_User, time=1, value=1.1}
          ]
         }}})
      float002:NiagaraFloat=&FloatFromCurve(FloatCurve=DataInterface() {propertyValues {Curve {
          defaultValue=1
          keys [
           {arriveTangent=1.1946, interpMode=RCIM_Cubic, leaveTangent=1.1946, tangentMode=RCTM_User, time=0, value=0.35}
           {arriveTangent=0.581, arriveTangentWeight=0.0667, interpMode=RCIM_Cubic, leaveTangent=0.581, tangentMode=RCTM_User, time=0.2, value=0.8875}
           {arriveTangent=0.0837, arriveTangentWeight=0.2667, interpMode=RCIM_Cubic, leaveTangent=0.0837, tangentMode=RCTM_User, time=1, value=1.1}
          ]
         }}})
      float003:NiagaraFloat=&FloatFromCurve(FloatCurve=DataInterface() {propertyValues {Curve {
          defaultValue=1
          keys [
           {arriveTangent=1.1946, interpMode=RCIM_Cubic, leaveTangent=1.1946, tangentMode=RCTM_User, time=0, value=0.35}
           {arriveTangent=0.581, arriveTangentWeight=0.0667, interpMode=RCIM_Cubic, leaveTangent=0.581, tangentMode=RCTM_User, time=0.2, value=0.8875}
           {arriveTangent=0.0837, arriveTangentWeight=0.2667, interpMode=RCIM_Cubic, leaveTangent=0.0837, tangentMode=RCTM_User, time=1, value=1.1}
          ]
         }}})
      int32:NiagaraInt32=&SpawnGroup()
     ))) {module=ScaleMeshSize001}
   /SetVariables(Particles.LerpPosition:NiagaraPosition=&Lerp_Vector(
     A=Linked() {linkedVariable {name=Particles.Position, type=NiagaraPosition}}
     B=&SimulationPosition()
     Alpha=&FloatFromCurve(FloatCurve=DataInterface() {propertyValues {Curve {keys [
          {arriveTangent=1.4308, interpMode=RCIM_Cubic, leaveTangent=1.4308, tangentMode=RCTM_User, time=0.6, value=0}
          {arriveTangent=4.0045, interpMode=RCIM_Cubic, leaveTangent=4.0045, tangentMode=RCTM_User, time=0.95, value=1}
         ]}}})
    )) {isSetParameters=true}
  ]
  rendererClasses [&NiagaraMeshRendererProperties]
  renderers [{rendererClass=&NiagaraMeshRendererProperties, rendererIndex=0}]
 }
 Secondary_Gloops {
  emitterUpdateScript [
   &EmitterState(
    Enable Distance Culling=true
    MaxDistance=12000
    Module.bUseMaxDistance=true
    Module.bUseMinDistance=false
    Scalability Mode=::NewEnumerator1
   )
   &SpawnBurst_Instantaneous(
    Spawn Count=&RandomRangeInt(Minimum=45, Maximum=20)
    Spawn Group=1
    Spawn Probability=&MakeCustomFloatFromBool(Bool=$User.IsInWater, True Float=1, False Float=0)
    Spawn Time=$System.Delay
    Use Spawn Probability=true
    UseLoopCountLimit=false
   ) {enabled=false, module=SpawnBurst_Instantaneous001}
   &SpawnBurst_Instantaneous(
    Spawn Count=&UniformRangedInt(Minimum=12, Maximum=15)
    Spawn Probability=&MakeCustomFloatFromBool(Bool=$User.IsInWater, True Float=1, False Float=0)
    Use Spawn Probability=true
   )
  ]
  particleSpawnScript [
   /Niagara/Modules/Spawn/Initialization/InitializeParticle(
    Color=<rgba 3.2063 4.2281 5 1>
    Lifetime=&RandomRangeFloat(Minimum=1, Maximum=1.4)
    Module.Write Color=true
    Module.Write Lifetime=true
    Module.Write Mass=true
    Module.Write Material Random=true
    Module.Write Position=false
    Module.Write Scale=false
    Module.Write SpriteRotation=true
    Module.Write SpriteSize=true
    Position=/Niagara/Functions/Localspace/SimulationPosition()
    Sprite Rotation=&UniformRangedFloat(Minimum=0, Maximum=360)
    Sprite Size=&Vector2DFromFloat(Value=&UniformRangedFloat(Minimum=50, Maximum=35))
   )
   &SphereLocation(
    Hemisphere Z=true
    MaskSpawnGroup=false
    Override Seed=false
    Sphere Coordinate Space=::World
    Sphere Radius=30
    Surface Only=false
    UseNonUniformScale=false
    UseOffset=false
    UseSphereOrientation=false
   )
   &AddVelocityFromPoint(
    Module.Use Falloff Distance=false
    Origin Offset=<0 0 -12>
    Origin Offset Coordinate Space=::World
    Velocity Strength=&Multiply_Float(
     A=&UniformRangedFloat(Minimum=700, Maximum=1150)
     B=&MaskFloatBySpawnGroup(Masked Float=2, Passthrough Float=2.5)
    )
   )
  ]
  particleUpdateScript [
   &ParticleState()
   &AccelerationForce(Acceleration=&Multiply_VectorByFloat(Vector=&UniformRangedVector(Minimum=<0 0 -1600>, Maximum=<0 0 -1800>), Float=2))
   &Drag(Use Linear Drag=true, Use Rotational Drag=false) {enabled=false}
   &SolveForcesAndVelocity(Clamp Velocity=false, Limit Acceleration=false)
   &ScaleSpriteSize(Scale Factor=&Vector2DFromFloat(Value=&FloatFromCurve(
      FloatCurve=DataInterface() {propertyValues {Curve {keys [
          {arriveTangent=12.5197, interpMode=RCIM_Cubic, leaveTangent=12.5197, tangentMode=RCTM_User, time=0, value=0.4376}
          {arriveTangent=0.8929, interpMode=RCIM_Cubic, leaveTangent=0.8929, tangentMode=RCTM_User, time=1, value=5}
         ]}}}
      Scale Curve=2
     )))
   &ScaleColor(
    Scale Alpha=&FloatFromCurve(FloatCurve=DataInterface() {propertyValues {Curve {keys [<at 0 -1>,<at 0.009 0.9316>,<at 0.7102 -0.1519>,<at 1 -1>]}}})
    Scale RGB=&VectorFromCurve(VectorCurve=DataInterface() {propertyValues {XCurve {keys [<at 0 2>,<at 1 0>]}, YCurve {keys [<at 0 2>,<at 1 0>]}, ZCurve {keys [<at 0 2>,<at 1 0>]}}})
    ScaleA=true
    ScaleRGB=true
   )
   &DynamicMaterialParameters(
    Index 0 Param 1=&FloatFromCurve(FloatCurve=DataInterface() {propertyValues {Curve {keys [
         {interpMode=RCIM_Cubic, tangentMode=RCTM_Break, time=0, value=1}
         {arriveTangent=-3.3696, interpMode=RCIM_Cubic, leaveTangent=-3.3696, tangentMode=RCTM_User, time=1, value=0}
        ]}}})
    Index 0 Param 2=&FloatFromCurve(FloatCurve=DataInterface() {propertyValues {Curve {keys [{arriveTangent=0.4361, interpMode=RCIM_Cubic, leaveTangent=0.4361, tangentMode=RCTM_User, time=0, value=0.05}, <at 1 0.25>]}}})
    Module.Param0WriteEnabled=true
    Module.Param1WriteEnabled=true
    Module.Param2WriteEnabled=true
    Module.Param3WriteEnabled=true
   ) {enabled=false}
   &DynamicMaterialParameters(
    Index 0 Param 1=&FloatFromCurve(FloatCurve=DataInterface() {propertyValues {Curve {keys [
         {arriveTangent=-2.0663, interpMode=RCIM_Cubic, leaveTangent=-2.0663, tangentMode=RCTM_User, time=0.25, value=1}
         {arriveTangent=-1.3063, interpMode=RCIM_Cubic, leaveTangent=-1.3063, tangentMode=RCTM_User, time=1.0011, value=0}
        ]}}})
    Index 0 Param 2=&FloatFromCurve(FloatCurve=DataInterface() {propertyValues {Curve {keys [<at -0.0086 -1.3001>,<at 0.9945 1.9153>]}}})
    Index 0 Param 3=0.5
    Module.Param0WriteEnabled=true
    Module.Param1WriteEnabled=false
    Module.Param2WriteEnabled=true
    Module.Param3WriteEnabled=true
   ) {module=DynamicMaterialParameters001}
   &CameraOffset(Camera Offset Amount=&FloatFromCurve(FloatCurve=DataInterface() {propertyValues {Curve {keys [
         {interpMode=RCIM_Cubic, leaveTangent=1080.3575, tangentMode=RCTM_Break, time=0, value=15}
         {arriveTangent=135, interpMode=RCIM_Cubic, leaveTangent=135, tangentMode=RCTM_Break, time=0.12, value=94}
         <at 1 150>
        ]}}})) {enabled=false}
   &SpriteRotationRate(Rotation Rate=&Lerp_Float(A=-150, B=150, Alpha=$Particles.Initial.MaterialRandom))
  ]
  rendererClasses [&NiagaraSpriteRendererProperties]
  renderers [{rendererClass=&NiagaraSpriteRendererProperties}]
 }
 SnowAirStreaks {
  emitterSpawnScript [/SetVariables(Emitter.NumToSpawn=7) {isSetParameters=true}]
  emitterUpdateScript [
   &EmitterState()
   &SpawnBurst_Instantaneous(Spawn Count=$Emitter.NumToSpawn, Use Spawn Probability=false, UseLoopCountLimit=false)
   &SpawnRate(
    SpawnRate=&MakeCustomFloatFromBool(Bool=$System.IsSynced, True Float=20, False Float=15)
    Use Spawn Probability=false
   ) {enabled=false}
  ]
  particleSpawnScript [
   /Niagara/Modules/Spawn/Initialization/V2/InitializeParticle(
    Color Maximum=<rgba 0.3 0.4551 0.6 0.8>
    Color Minimum=<rgba 0.385 0.5215 0.7 0.6>
    Color Mode=::NewEnumerator2
    Lifetime=0.35
    Sprite Size Max=<80 120>
    Sprite Size Min=<30 300>
    Sprite Size Mode=::NewEnumerator2
    UsePositionOffset=false
    Write Material Random=true
   ) {module=InitializeParticle001}
   &SubUVAnimation(End Frame=3, SubUV Animation Mode:ENiagara_SubUVLookupMode=::NewEnumerator2)
   &CameraOffset(Camera Offset Amount=25)
   &PartitionParticles(Number Of Partitions=1, Particles Per Partition:NiagaraInt32=$Emitter.NumToSpawn, Partition Mode:ENiagara_PartitionMode_v2=::NewEnumerator1)
   &ShapeLocation(
    Cylinder Distribution=::NewEnumerator2
    Cylinder Height=0.01
    Cylinder Radius=10
    MaskSpawnGroup=false
    Shape Primitive=::NewEnumerator1
    U Position=&Add_Float(
     A=$Particles.PartitionParticles.Normalized Index In Partition
     B=&Multiply_Float(A=0.1, B=&RandomRangeFloat(Minimum=-1, Maximum=1))
    )
    UseOwnerScale=true
    V Position=&RandomRangeFloat(Minimum=0, Maximum=1)
   )
   &AddVelocity(
    Velocity=&Add_Vector(A=&Multiply_VectorByFloat(Vector=$Particles.ShapeLocation.ShapeNormal, Float=1.5), B=<0 0 0.2>)
    Velocity Speed Scale=2000
   )
  ]
  particleUpdateScript [
   &ParticleState()
   &ScaleSpriteSize(Scale Sprite Size Mode=::NewEnumerator0, Uniform Scale Factor=5) {module=ScaleSpriteSize001}
   &Drag(Drag=0, Use Linear Drag=true, Use Rotational Drag=false)
   &ScaleVelocity(Velocity Scale=&VectorFromFloat(Value=&FloatFromCurve(FloatCurve=DataInterface() {propertyValues {Curve {keys [
          {arriveTangent=-3.6709, interpMode=RCIM_Cubic, leaveTangent=-3.6709, tangentMode=RCTM_User, time=0, value=1}
          {interpMode=RCIM_Cubic, tangentMode=RCTM_User, time=1, value=0}
         ]}}}))) {enabled=false}
   &SolveForcesAndVelocity(Clamp Velocity=false, Limit Acceleration=false)
   &ScaleSpriteSize(Scale Factor=&Vector2DFromCurve(
     Vector2Curve=DataInterface() {propertyValues {
       XCurve {keys [
         {arriveTangent=2.5428, interpMode=RCIM_Cubic, leaveTangent=2.5428, tangentMode=RCTM_User, time=0, value=1}
         {arriveTangent=0.1281, interpMode=RCIM_Cubic, leaveTangent=0.1281, tangentMode=RCTM_User, time=1, value=2}
        ]}
       YCurve {keys [{arriveTangent=2.8823, time=0, value=1}, {arriveTangent=2.0297, leaveTangent=2.0297, time=1, value=0.5}]}
      }}
     ScaleCurve=&Multiply_Vector2D_ByFloat(
      Float=&RandomRangeFloat(Minimum=0.6, Maximum=1)
      Vector2D=&RandomRangeVector2D(Minimum=<0.6 0.75>, Maximum=<0.8 1>)
     )
    ))
   &ScaleColor(
    Scale Alpha=&FloatFromCurve(
     FloatCurve=DataInterface() {propertyValues {Curve {keys [<at 0 1.5 RCIM_Cubic>, {arriveTangent=-2.8468, interpMode=RCIM_Cubic, leaveTangent=-2.8468, tangentMode=RCTM_User, time=1, value=0}]}}}
     Scale Curve=&Multiply_Float(
      A=&Multiply_Float(
       A=&FNTimeOfDayScalar(Night Scale=0.25, Morning Scale=0.5, Evening Scale=0.4)
       B=&RandomRangeFloat(Minimum=1, Maximum=1.25)
      )
      B=&MakeCustomFloatFromBool(Bool=$System.IsSynced, True Float=0.4, False Float=0.75)
     )
    )
    Scale RGB=&Multiply_Vector(
     A=&VectorFromCurve(
      VectorCurve=DataInterface() {propertyValues {
        XCurve {keys [
          {arriveTangent=-2.3067, interpMode=RCIM_Cubic, time=0, value=2}
          {arriveTangent=-2.5804, interpMode=RCIM_Cubic, leaveTangent=-2.5804, tangentMode=RCTM_User, time=1, value=0.5}
         ]}
        YCurve {keys [
          {arriveTangent=-2.3067, interpMode=RCIM_Cubic, time=0, value=2}
          {arriveTangent=-2.5804, interpMode=RCIM_Cubic, leaveTangent=-2.5804, tangentMode=RCTM_User, time=1, value=0.5}
         ]}
        ZCurve {keys [
          {arriveTangent=-2.3067, interpMode=RCIM_Cubic, time=0, value=2}
          {arriveTangent=-2.5804, interpMode=RCIM_Cubic, leaveTangent=-2.5804, tangentMode=RCTM_User, time=1, value=0.5}
         ]}
       }}
      Scale Curve=&VectorFromFloat(Value=&MakeCustomFloatFromBool(Bool=$System.IsSynced, True Float=2, False Float=0.75))
     )
     B=&MakeVectorFromLinearColorRGB(Color=&FN_Select_Color_from_Time_Of_Day(Morning=<rgba 1 0.8463 0.7189 1>, Day=<rgba 1 1 1 1>, Evening=<rgba 1 0.8371 0.7936 1>, Night=<rgba 0.6111 0.6842 1 1>))
    )
    ScaleA=true
    ScaleRGB=true
   )
   &DynamicMaterialParameters(
    Index 0 Param 1=20
    Index 0 Param 2=1
    Index 0 Param 3=1
    Index 0 Param 4=&FloatFromCurve(FloatCurve=DataInterface() {propertyValues {Curve {keys [
         {interpMode=RCIM_Cubic, tangentMode=RCTM_User, time=0, value=0}
         {arriveTangent=2.7922, interpMode=RCIM_Cubic, leaveTangent=2.7922, tangentMode=RCTM_User, time=1, value=1}
        ]}}})
    Index 2 Param 1=2
    Index 2 Param 2=&FloatFromCurve(
     FloatCurve=DataInterface() {propertyValues {Curve {keys [
         {arriveTangent=0.0308, interpMode=RCIM_Cubic, leaveTangent=0.0308, tangentMode=RCTM_User, time=0, value=1}
         {arriveTangent=-2.7663, interpMode=RCIM_Cubic, leaveTangent=-2.7663, tangentMode=RCTM_User, time=1, value=0}
        ]}}}
     Scale Curve=&MakeCustomFloatFromBool(Bool=$System.IsSynced, True Float=0.775, False Float=0.75)
    )
    Module.Param0WriteEnabled=true
    Module.Param0WriteEnabled2=true
    Module.Param1WriteEnabled=true
    Module.Param1WriteEnabled2=true
    Module.Param2WriteEnabled=true
    Module.Param2WriteEnabled2=true
    Module.Param3WriteEnabled=true
    Module.Param3WriteEnabled2=true
    Write Parameter Index 2=true
   )
  ]
  rendererClasses [&NiagaraSpriteRendererProperties]
  renderers [{rendererClass=&NiagaraSpriteRendererProperties}]
 }
 SphereElement_Initial {
  emitterUpdateScript [
   &EmitterState()
   &SpawnBurst_Instantaneous(Spawn Count=1, Use Spawn Probability=true)
   /SetVariables(Emitter.PanSpeed=0.4, Emitter.Tile_U=10, Emitter.UV_Distortion=5) {isSetParameters=true}
   /SetVariables(Emitter.PlasmaAmount=0.5, Emitter.UVScale:Vector4f=<0.2 0.2 0 0>) {isSetParameters=true}
  ]
  particleSpawnScript [
   &InitializeParticle(
    Color=<rgba 1 1 1 0.035>
    Lifetime=$User.FXDuration_Initial
    Mesh Scale Mode=::NewEnumerator3
    Mesh Uniform Scale=&Multiply_Float(A=0.026, B=$User.FXScale)
    Sprite Size Mode=::NewEnumerator3
    Uniform Sprite Size=3000
    UsePositionOffset=false
    Write Material Random=true
   )
   &InitialMeshOrientation(
    Mesh Axis To Orient=<0 0 1>
    Mesh Orientation Mode=::NewEnumerator2
    Rotation=&RandomRangeVector(Minimum=<0 0 0>, Maximum=<0 0 1>)
    Use Rotation Vector=true
   )
  ]
  particleUpdateScript [
   &ParticleState()
   &ScaleMeshSize(Scale Factor=&VectorFromFloat(Value=&Lerp_Float(
      A=0.1
      B=1
      Alpha=&FloatFromCurve(FloatCurve=DataInterface() {propertyValues {Curve {keys [
           {arriveTangent=2.1741, interpMode=RCIM_Cubic, leaveTangent=2.1741, tangentMode=RCTM_User, time=0, value=0.5}
           {
            arriveTangent=-0.0343
            arriveTangentWeight=0.3049
            interpMode=RCIM_Cubic
            leaveTangent=-0.0343
            leaveTangentWeight=0.3049
            tangentMode=RCTM_User
            time=0.6
            value=1
           }
           {arriveTangent=-5.2478, interpMode=RCIM_Cubic, leaveTangent=-5.2478, tangentMode=RCTM_User, time=1, value=0}
          ]}}})
     ))) {module=ScaleMeshSize001}
   &ScaleColor(
    Scale RGB=&MakeVectorFromLinearColorRGB(Color=&LerpLinearColors(
      StartColor=<rgba 0 0.4925 1 1>
      EndColor=<rgba 1 0.0896 0 1>
      LerpFactor=&FloatFromCurve(FloatCurve=DataInterface() {propertyValues {Curve {keys [
           <at 0 0>
           {arriveTangent=1.1667, interpMode=RCIM_Cubic, leaveTangent=1.1667, tangentMode=RCTM_User, time=0.7, value=0}
           {interpMode=RCIM_Cubic, tangentMode=RCTM_User, time=0.9, value=1.05}
          ]}}})
     ))
    ScaleA=false
    ScaleRGB=true
   ) {module=ScaleColor001}
   &ScaleColor(
    Scale Alpha=&FloatFromCurve(FloatCurve=DataInterface() {propertyValues {Curve {keys [
         {arriveTangent=6.3568, interpMode=RCIM_Cubic, leaveTangent=6.3568, tangentMode=RCTM_User, time=0, value=0}
         {
          arriveTangent=0.1702
          arriveTangentWeight=0.2056
          interpMode=RCIM_Cubic
          leaveTangent=0.1702
          leaveTangentWeight=0.2367
          tangentMode=RCTM_Break
          tangentWeightMode=RCTWM_WeightedBoth
          time=0.3
          value=0.6
         }
         {interpMode=RCIM_Cubic, tangentMode=RCTM_User, time=1, value=1}
        ]}}})
    Scale RGB=&VectorFromFloat(Value=&Lerp_Float(
      A=0
      B=30
      Alpha=&Multiply_Float(
       A=4
       B=&FloatFromCurve(FloatCurve=DataInterface() {propertyValues {Curve {keys [
            {interpMode=RCIM_Cubic, leaveTangentWeight=0.939, tangentMode=RCTM_User, time=0, value=0}
            {arriveTangent=2.368, interpMode=RCIM_Cubic, leaveTangent=2.368, tangentMode=RCTM_User, time=1, value=1}
           ]}}})
      )
     ))
    ScaleA=true
    ScaleRGB=true
   )
   &DynamicMaterialParameters(
    Index 0 Param 1=&FloatFromCurve(
     FloatCurve=DataInterface() {propertyValues {Curve {keys [
         {arriveTangent=-0.742, interpMode=RCIM_Cubic, leaveTangent=-0.742, leaveTangentWeight=0.1747, tangentMode=RCTM_User, time=0, value=1}
         {arriveTangent=-0.8, interpMode=RCIM_Cubic, leaveTangent=-0.8, time=0.9, value=0.8}
         {arriveTangent=-8.0918, arriveTangentWeight=0.6932, interpMode=RCIM_Cubic, leaveTangent=-8.0918, tangentMode=RCTM_User, time=1, value=0.2}
        ]}}}
     Scale Curve=0.2
    )
    Index 0 Param 3=&Lerp_Float(
     A=2
     B=4
     Alpha=&FloatFromCurve(FloatCurve=DataInterface() {propertyValues {Curve {keys [
          {interpMode=RCIM_Cubic, tangentMode=RCTM_User, time=0, value=1}
          {arriveTangent=-2.8954, interpMode=RCIM_Cubic, leaveTangent=-2.8954, tangentMode=RCTM_User, time=1, value=0}
         ]}}})
    )
    Module.Index0Param0Write=true
    Module.Index0Param1Write=true
    Module.Index0Param2Write=true
    Module.Index0Param3Write=true
   )
   /SetVariables(Particles.Rign Color:LinearColor=&MultiplyLinearColor(
     Linear Color=$Particles.Color
     Scale Factor=&MakeVector4FromVectorAndFloat("Vector (RGB)"=&VectorFromFloat(Value=0.3), "Float (Alpha)"=1)
    )) {isSetParameters=true}
   &FN_Light_Attributes(
    Color=$Particles.Color
    Create Color Attribute=true
    Create Exponent Attribute=true
    Create Position Attribute=true
    Create Radius Attribute=true
    Exponent=2
    Position=&AddVectorToPosition(Position=/Niagara/DynamicInputs/Helpers/SimulationPosition(), Vector=<0 0 100>)
    Radius=&Lerp_Float(
     A=0.1
     B=1
     Alpha=&FloatFromCurve(
      FloatCurve=DataInterface() {propertyValues {Curve {keys [
          {arriveTangent=2.8634, interpMode=RCIM_Cubic, leaveTangent=2.8634, tangentMode=RCTM_User, time=0, value=0.2}
          {
           arriveTangent=1.0052
           arriveTangentWeight=0.2917
           interpMode=RCIM_Cubic
           leaveTangentWeight=0.0417
           tangentMode=RCTM_Break
           time=0.8
           value=1
          }
          {
           arriveTangentWeight=0.0533
           interpMode=RCIM_Cubic
           leaveTangent=-18.1489
           leaveTangentWeight=0.0667
           tangentMode=RCTM_Break
           time=0.9
           value=1
          }
          {arriveTangent=-3.9622, interpMode=RCIM_Cubic, leaveTangent=-3.9622, tangentMode=RCTM_User, time=1, value=0}
         ]}}}
      Scale Curve=1500
     )
    )
    Scale Color=3
   )
   /SetVariables(Particles.Vector 4:Vector4f=&MakeVector4(
     X=&FloatFromCurve(FloatCurve=DataInterface() {propertyValues {Curve {keys [
          {arriveTangent=-3.1256, interpMode=RCIM_Cubic, leaveTangent=-3.1256, leaveTangentWeight=0.1747, tangentMode=RCTM_User, time=0, value=1}
          {arriveTangent=-2.4168, arriveTangentWeight=0.6932, interpMode=RCIM_Cubic, leaveTangent=-2.4168, tangentMode=RCTM_User, time=1, value=0}
         ]}}})
     Y=0
     Z=0
     W=0
    )) {isSetParameters=true}
   /SetVariables(Particles.Vector 4 Dynam1:Vector4f=&MakeVector4(X=500, Y=0, Z=0, W=0)) {isSetParameters=true}
   /SetVariables(Particles.Cam Offset Sprite:NiagaraFloat=&Lerp_Float(
     A=250
     B=750
     Alpha=&FloatFromCurve(FloatCurve=DataInterface() {propertyValues {Curve {
         defaultValue=0
         keys [
          {arriveTangent=2.1741, interpMode=RCIM_Cubic, leaveTangent=2.1741, tangentMode=RCTM_User, time=0, value=0.5}
          {arriveTangent=-0.0343, arriveTangentWeight=0.2, interpMode=RCIM_Cubic, leaveTangent=-0.0343, tangentMode=RCTM_User, time=0.6, value=1}
          {arriveTangent=-5.2478, arriveTangentWeight=0.1333, interpMode=RCIM_Cubic, leaveTangent=-5.2478, tangentMode=RCTM_User, time=1, value=0}
         ]
        }}})
    )) {isSetParameters=true}
   /SetVariables(Particles.Linear Color=&MultiplyLinearColor(
     Linear Color=$Particles.Color
     Scale Factor=&MakeVector4FromVectorAndFloat("Vector (RGB)"=&VectorFromFloat(Value=2), "Float (Alpha)"=1)
    )) {isSetParameters=true}
   &ScaleColor(
    Scale Alpha=&MakeCustomFloatFromBool(
     Bool=&IsPlatformSetActive(New Platform Set=DataInterface() {propertyValues {Platforms {qualityLevelMask=-4}}})
     True Float=1
     False Float=2
    )
    Scale RGB=&MakeCustomVectorFromBool(
     Bool=&IsPlatformSetActive(New Platform Set=DataInterface() {propertyValues {Platforms {qualityLevelMask=-4}}})
     True Vector=&VectorFromFloat(Value=1)
     False Vector=&VectorFromFloat(Value=10)
    )
    ScaleA=true
    ScaleRGB=true
   ) {module=ScaleColor002}
  ]
  rendererClasses [
   &NiagaraMeshRendererProperties
   &NiagaraLightRendererProperties
   &NiagaraSpriteRendererProperties
  ]
  renderers [
   {rendererClass=&NiagaraMeshRendererProperties, rendererIndex=0}
   {rendererClass=&NiagaraMeshRendererProperties, rendererIndex=1}
   {rendererClass=&NiagaraLightRendererProperties, rendererIndex=2}
   {rendererClass=&NiagaraMeshRendererProperties, rendererIndex=3}
   {rendererClass=&NiagaraMeshRendererProperties, rendererIndex=4}
   {rendererClass=&NiagaraSpriteRendererProperties, rendererIndex=5}
  ]
 }
 WaterFN_Mesh001 {
  emitterUpdateScript [
   &EmitterState()
   &SpawnBurst_Instantaneous(
    Spawn Count=&MakeCustomIntFromBool(Bool=$User.IsInWater, True INT=1, False INT=0)
    Spawn Group=1
    Spawn Time=$System.Delay
    Use Spawn Probability=false
    UseLoopCountLimit=false
   ) {enabled=false, module=SpawnBurst_Instantaneous001}
   &SpawnBurst_Instantaneous(
    Spawn Count=&MakeCustomIntFromBool(Bool=$User.IsInWater, True INT=1, False INT=0)
    Use Spawn Probability=false
   )
  ]
  particleSpawnScript [
   &InitializeParticle(
    Color=<rgba 1 1 1 0.025>
    Lifetime=&UniformRangedFloat(Minimum=1.9, Maximum=2.4)
    Mesh Scale=&Multiply_VectorByFloat(
     Vector=&MaskVectorBySpawnGroup(Masked Vector=<5 5 2>, Passthrough Vector=<5 5 1>)
     Float=&MaskFloatBySpawnGroup(Masked Float=1.2, Passthrough Float=2)
    )
    Mesh Scale Mode=::NewEnumerator4
    UsePositionOffset=false
    Write Material Random=true
   ) {module=InitializeParticle001}
   &InitialMeshOrientation(
    Mesh Orientation Mode=::NewEnumerator7
    Rotation=&RandomRangeVector(Minimum=<0 0 0>, Maximum=<0 0 1>)
    Use Rotation Vector=true
   ) {module=InitialMeshOrientation001}
  ]
  particleUpdateScript [
   &ParticleState()
   &ScaleMeshSize(Scale Factor=&VectorFromCurve(VectorCurve=DataInterface() {propertyValues {
       XCurve {keys [
         {interpMode=RCIM_Cubic, leaveTangent=7.6613, tangentMode=RCTM_Break, time=0, value=0}
         {arriveTangent=2.4936, interpMode=RCIM_Cubic, leaveTangent=2.4936, tangentMode=RCTM_User, time=0.2392, value=1.3545}
         <at 1 2>
        ]}
       YCurve {keys [
         {interpMode=RCIM_Cubic, leaveTangent=7.7916, tangentMode=RCTM_Break, time=0, value=0}
         {arriveTangent=2.6017, interpMode=RCIM_Cubic, leaveTangent=2.6017, tangentMode=RCTM_User, time=0.2392, value=1.3545}
         <at 1 2>
        ]}
       ZCurve {keys [<at 0 3>,<at 1 1>]}
      }}))
   &DynamicMaterialParameters(
    Index 0 Param 1=&FloatFromCurve(FloatCurve=DataInterface() {propertyValues {Curve {keys [<at 0 0>,<at 1 1.5>]}}})
    Index 0 Param 2=0.25
    Index 0 Param 3=50
    Index 0 Param 4=1
    Module.Param0WriteEnabled=true
    Module.Param1WriteEnabled=true
    Module.Param2WriteEnabled=true
    Module.Param3WriteEnabled=true
   )
   &CameraOffset(Camera Offset Amount=50)
   &ScaleColor(
    Scale Alpha=&FloatFromCurve(FloatCurve=DataInterface() {propertyValues {Curve {keys [
         <at 0 0>
         {arriveTangent=5.1071, interpMode=RCIM_Cubic, leaveTangent=5.1071, tangentMode=RCTM_Break, time=0.1294, value=2.8567}
         {arriveTangent=-8.2025, interpMode=RCIM_Cubic, leaveTangent=-8.2025, tangentMode=RCTM_User, time=0.3115, value=1.8534}
         <at 0.4715 0.0687>
         <at 1 0>
        ]}}})
    Scale RGB=&VectorFromFloat(Value=&FloatFromCurve(FloatCurve=DataInterface() {propertyValues {Curve {keys [<at 0 2>,<at 1 0.2>]}}}))
    ScaleA=true
    ScaleRGB=true
   )
  ]
  rendererClasses [&NiagaraMeshRendererProperties]
  renderers [{rendererClass=&NiagaraMeshRendererProperties, rendererIndex=0}]
 }
}
summary {
 systemName=NS_Jethro_CloudyMuffin_Exp_Pre
 emitters [
  {
   emitterName=SphereElement_Initial
   rendererClasses [
    &NiagaraMeshRendererProperties
    &NiagaraLightRendererProperties
    &NiagaraSpriteRendererProperties
   ]
  }
  {emitterName=CenterFlash, rendererClasses [&NiagaraMeshRendererProperties, &NiagaraSpriteRendererProperties]}
  {emitterName=EnergyWave, rendererClasses [&NiagaraMeshRendererProperties, &NiagaraLightRendererProperties]}
  {emitterName=WaterFN_Mesh001, rendererClasses [&NiagaraMeshRendererProperties]}
  {emitterName=Secondary_Gloops, rendererClasses [&NiagaraSpriteRendererProperties]}
  {emitterName=Lines, rendererClasses [&NiagaraSpriteRendererProperties]}
  {emitterName=NE_Explosion_Main001, rendererClasses [&NiagaraMeshRendererProperties]}
  {emitterName=Empty003, rendererClasses [&NiagaraMeshRendererProperties]}
  {emitterName=SnowAirStreaks, rendererClasses [&NiagaraSpriteRendererProperties]}
 ]
 userVariables [
  User.FXDuration_Initial {type=NiagaraFloat, defaultValue=0.6, description=""}
  User.FXScale {type=NiagaraFloat, defaultValue=500, description=""}
  User.Global Color {type=LinearColor, description="", defaultValue {struct=LinearColor, value=<rgba 4 0.5231 0 1>}}
  User.Global Scale {type=NiagaraFloat, defaultValue=1, description=""}
  User.IsInWater {type=NiagaraBool}
 ]
}
