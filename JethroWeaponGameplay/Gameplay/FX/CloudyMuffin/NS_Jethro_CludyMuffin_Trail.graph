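# Niagara trail and countdown effect for Jethro's CloudyMuffin weapon, with user parameters for the explosion timer.
# Ribbon emitter spawns by distance moved, leaving a short-lived glowing trail that cools from orange-yellow to red.
# ETimer bursts a red sprite with a light; system Timeline speeds its local Pulse module, brightens both, reveals a horizontal flare in the last quarter, then kills it.
# Camera distance enlarges the timer sprite to aid readability.
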NiagaraSystem /JethroWeaponGameplay/Gameplay/FX/CloudyMuffin/NS_Jethro_CludyMuffin_Trail
dependencies {
 usedDataInterfaces [&NiagaraDataInterfaceColorCurve, &NiagaraDataInterfaceCurve, &NiagaraDataInterfaceCamera]
 usedDynamicInputs []
 usedModules [&SystemState, &Timeline]
}
emitters {
 ETimer {
  emitterUpdateScript [
   &EmitterState()
   &SpawnBurst_Instantaneous(Use Spawn Probability=false, UseLoopCountLimit=true)
  ]
  particleSpawnScript [
   &InitializeParticle(
    Color=<rgba 1 0.0167 0 0.1>
    Lifetime=10
    Sprite Size Mode=::NewEnumerator3
    Uniform Sprite Size=250
    UsePositionOffset=false
    Write Material Random=true
   )
   &CameraOffset()
   /SetVariables(Particles.Hflare Cam offse=25) {isSetParameters=true}
   /SetVariables(Particles.HFlare Position:NiagaraPosition=&AddVectorToPosition(Position=$Particles.Position, Vector=<0 0 0>)) {isSetParameters=true}
  ]
  particleUpdateScript [
   &ParticleState(Kill Particles When Lifetime Has Elapsed=false, Let Infinitely Lived Particles Die When Emitter Deactivates=true)
   /SetVariables(Particles.DistCurve:NiagaraFloat=&FloatFromCurve(
     FloatCurve=DataInterface() {propertyValues {Curve {defaultValue=1, keys [{tangentMode=RCTM_User, time=0.2, value=0}, {tangentMode=RCTM_User, time=1, value=1}]}}}
     CurveIndex=&Divide_Float(
      A=&DistanceBetweenPositions(
       A=&ReturnCameraPosition(Coordinate Space:ENiagaraCoordinateSpace=::Simulation, Camera Query:NiagaraDataInterfaceCamera=DataInterface())
       B=/Niagara/DynamicInputs/Helpers/SimulationPosition()
      )
      B=&Multiply_Float(A=5, B=1000)
     )
    )) {isSetParameters=true}
   &ScaleSpriteSize(Scale Sprite Size Mode=::NewEnumerator0, Uniform Scale Factor=&Lerp_Float(A=1, B=2, Alpha=$Particles.DistCurve))
   &self:Pulse(
    Alpha Scale Max:NiagaraFloat=&Lerp_Float(A=1, B=1.5, Alpha=$Particles.DistCurve)
    ColorScale:Vector4f=&MakeVector4FromVectorAndFloat(
     "Vector (RGB)"=&VectorFromFloat(Value=&Lerp_Float(
       A=1
       B=3
       Alpha=&FloatFromCurve(
        FloatCurve=DataInterface() {propertyValues {Curve {
           defaultValue=1
           keys [<at 0 0 RCIM_Cubic>, {arriveTangent=1.9786, interpMode=RCIM_Cubic, leaveTangent=1.9786, tangentMode=RCTM_User, time=1, value=1}]
          }}}
        CurveIndex=$System.Timeline.NormalizedTime
       )
      ))
     "Float (Alpha)"=1
    )
    DT Mul:NiagaraFloat=&Lerp_Float(
     A=1
     B=8
     Alpha=&FloatFromCurve(
      FloatCurve=DataInterface() {propertyValues {Curve {
         defaultValue=1
         keys [
          {
           interpMode=RCIM_Cubic
           leaveTangent=0.2304
           leaveTangentWeight=0.5868
           tangentMode=RCTM_Break
           tangentWeightMode=RCTWM_WeightedBoth
           time=0
           value=0
          }
          {
           arriveTangent=4.5131
           arriveTangentWeight=0.9532
           interpMode=RCIM_Cubic
           leaveTangent=3.7331
           tangentMode=RCTM_Break
           tangentWeightMode=RCTWM_WeightedBoth
           time=1
           value=1
          }
         ]
        }}}
      CurveIndex=$System.Timeline.NormalizedTime
     )
    )
   ) {module=Pulse}
   &FN_Light_Attributes(
    Color=$Particles.Color
    Create Color Attribute=true
    Create Exponent Attribute=true
    Create Radius Attribute=true
    Exponent=1
    Radius=&Lerp_Float(
     A=100
     B=120
     Alpha=&FloatFromCurve(
      FloatCurve=DataInterface() {propertyValues {Curve {defaultValue=1, keys [{tangentMode=RCTM_User, time=0, value=0}, {tangentMode=RCTM_User, time=1, value=1}]}}}
      CurveIndex=$System.Timeline.NormalizedTime
     )
    )
    Scale Color=&Lerp_Float(
     A=8
     B=12
     Alpha=&FloatFromCurve(
      FloatCurve=DataInterface() {propertyValues {Curve {
         defaultValue=1
         keys [<at 0 0 RCIM_Cubic>, {arriveTangent=2.7732, interpMode=RCIM_Cubic, leaveTangent=2.7732, tangentMode=RCTM_User, time=1, value=1}]
        }}}
      CurveIndex=$System.Timeline.NormalizedTime
     )
    )
   )
   /SetVariables(
    Particles.Hflare Color:LinearColor=&MultiplyLinearColor(
     Linear Color=$Particles.Color
     Scale Factor=&MakeVector4FromVectorAndFloat(
      "Vector (RGB)"=&VectorFromFloat(Value=&FloatFromCurve(
        FloatCurve=DataInterface() {propertyValues {Curve {keys [<at 0.75 0 RCIM_Cubic>,<at 0.9 1 RCIM_Cubic>]}}}
        CurveIndex=$System.Timeline.NormalizedTime
        Scale Curve=5
       ))
      "Float (Alpha)"=1
     )
    )
    Particles.Hflare Size:Vector2f=&Multiply_Vector2D(
     A=$Particles.SpriteSize
     B=&Multiply_Vector2D_ByFloat(
      Float=&FloatFromCurve(
       FloatCurve=DataInterface() {propertyValues {Curve {keys [<at 0 0 RCIM_Constant>,<at 0.75 4>]}}}
       CurveIndex=$System.Timeline.NormalizedTime
       Scale Curve=2
      )
      Vector2D=<1.35 0.08>
     )
    )
   ) {isSetParameters=true}
   &KillParticles(Kill Particles=&SetBoolByFloatComparison(A=$System.Timeline.NormalizedTime, B=0.999))
  ]
  rendererClasses [&NiagaraSpriteRendererProperties, &NiagaraLightRendererProperties]
  renderers [
   {rendererClass=&NiagaraSpriteRendererProperties}
   {rendererClass=&NiagaraLightRendererProperties, rendererIndex=1}
   {rendererClass=&NiagaraSpriteRendererProperties, rendererIndex=2}
  ]
 }
 Ribbon {
  emitterSpawnScript [/SetVariables(Emitter.Static bool=true) {isSetParameters=true}]
  emitterUpdateScript [
   &EmitterState(Life Cycle Mode=::NewEnumerator1, Loop Duration=1, UseLoopDelay=true)
   &SpawnRate(SpawnRate=30, Use Spawn Probability=false) {enabled=false}
   &SpawnPerUnit(
    Module.Use Max Movement Threshold=true
    Module.Use Movement Tolerance=true
    Movement Tolerance=0.3
    Spawn Spacing=10
    Use Spawn Probability=false
   )
  ]
  particleSpawnScript [&InitializeParticle(
    Color=<rgba 1 1 1 0.1>
    Lifetime=0.35
    Ribbon Width=15
    Ribbon Width Mode=::NewEnumerator1
    UsePositionOffset=false
    Write Material Random=true
   )]
  particleUpdateScript [
   &ParticleState()
   &Color(
    Color=&ColorFromCurve(ColorCurve=DataInterface() {propertyValues {
       AlphaCurve {keys [<at 0 1 RCIM_Cubic>,<at 1 0 RCIM_Cubic>]}
       BlueCurve {keys [<at 0 0 RCIM_Cubic>,<at 0.6185 0 RCIM_Cubic>]}
       GreenCurve {keys [<at 0 1.3333 RCIM_Cubic>,<at 0.6185 0.1667 RCIM_Cubic>]}
       RedCurve {keys [<at 0 4 RCIM_Cubic>,<at 0.6185 1 RCIM_Cubic>]}
      }})
    Scale Alpha=&Multiply_Float(
     A=&Multiply_Float(
      A=&Multiply_Float(
       A=&FloatFromCurve(FloatCurve=DataInterface() {propertyValues {Curve {
           defaultValue=1
           keys [
            <at 0 0 RCIM_Cubic>
            {interpMode=RCIM_Cubic, leaveTangentWeight=0.1333, tangentMode=RCTM_User, time=0.1, value=1}
            {interpMode=RCIM_Cubic, tangentMode=RCTM_User, time=0.5, value=1}
            <at 1 0 RCIM_Cubic>
           ]
          }}})
       B=&FloatFromCurve(
        FloatCurve=DataInterface() {propertyValues {Curve {
           defaultValue=1
           keys [
            <at 0 0 RCIM_Cubic>
            {interpMode=RCIM_Cubic, leaveTangentWeight=0.1333, tangentMode=RCTM_User, time=0.1, value=1}
            {interpMode=RCIM_Cubic, tangentMode=RCTM_User, time=0.9, value=1}
           ]
          }}}
        CurveIndex=&ReturnNormalizedExecIndex()
       )
      )
      B=&Lerp_Float(
       A=&FloatFromCurve(
        FloatCurve=DataInterface() {propertyValues {Curve {
           defaultValue=1
           keys [
            {arriveTangent=0.4704, interpMode=RCIM_Cubic, leaveTangent=0.4704, tangentMode=RCTM_User, time=0, value=0}
            {arriveTangent=2.9573, interpMode=RCIM_Cubic, leaveTangent=2.9573, tangentMode=RCTM_User, time=1, value=1}
           ]
          }}}
        CurveIndex=&ReturnNormalizedExecIndex()
       )
       B=1
       Alpha=&FloatFromCurve(
        FloatCurve=DataInterface() {propertyValues {Curve {
           defaultValue=1
           keys [<at 0 0 RCIM_Cubic>, {arriveTangent=2.8054, interpMode=RCIM_Cubic, leaveTangent=2.8054, tangentMode=RCTM_User, time=1, value=1}]
          }}}
        CurveIndex=$Emitter.Age
       )
      )
     )
     B=0.75
    )
    Scale Color=&VectorFromFloat(Value=2)
   )
  ]
  rendererClasses [&NiagaraRibbonRendererProperties]
  renderers [{rendererClass=&NiagaraRibbonRendererProperties}]
 }
}
summary {
 systemName=NS_Jethro_CludyMuffin_Trail
 emitters [
  {emitterName=Ribbon, rendererClasses [&NiagaraRibbonRendererProperties]}
  {emitterName=ETimer, rendererClasses [&NiagaraSpriteRendererProperties, &NiagaraLightRendererProperties]}
 ]
 userVariables [
  User.ExplosionTimer {type=NiagaraFloat, defaultValue=2.75, description=""}
  User.PlayTimer {type=NiagaraBool, defaultValue=true}
 ]
}
